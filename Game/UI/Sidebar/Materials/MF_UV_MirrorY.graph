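# MF_UV_MirrorY is a UI sidebar material function that flips UVs horizontally, mirroring them across the vertical axis.
# It takes a Vector2 UV input, inverts and saturates the U channel, then appends the untouched V channel.
# The saturate clamps the flipped U to the 0-1 range, so U does not tile beyond it.
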MaterialFunction /Game/UI/Sidebar/Materials/MF_UV_MirrorY
graph {nodes [
  n1 = FunctionInput(Preview=TextureCoordinate()) {inputType=FunctionInput_Vector2}
  n2 = OneMinus(None=ComponentMask(None=@n1) {r=true})
  FunctionOutput(None=AppendVector(A=Saturate(None=@n2), B=ComponentMask(None=@n1) {g=true}))
 ]}
