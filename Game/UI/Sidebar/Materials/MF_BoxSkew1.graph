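# Material function for the sidebar UI that builds a skewed-box gradient with independently angled left, right, top and bottom edges.
# Center-scaled UVs, optionally mirrored by a static Flip switch, are dotted against per-edge slope vectors, with left and right offsets.
# Slopes are aspect-corrected from texture coordinate set 3, with MinWidth adjusting the top and bottom ratios.
# Nested Min nodes intersect the four edges into one unthresholded output; trailing SmoothStep and Step nodes are unused.
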MaterialFunction /Game/UI/Sidebar/Materials/MF_BoxSkew1
graph {nodes [
  n1 = MaterialFunctionCall(
   UVs=TextureCoordinate()
   Texture Scale=FunctionInput(Preview=Constant() {r=1}) {bUsePreviewValueAsDefault=true, inputName=Scale, inputType=FunctionInput_Vector2, sortPriority=4}
  ) {materialFunction=&ScaleUVsByCenter}
  n2 = FunctionInput(Preview=StaticBool()) {inputName=Flip, inputType=FunctionInput_StaticBool, sortPriority=10}
  NamedRerouteDeclaration(None=StaticSwitch(True=MaterialFunctionCall!(In=@n1.UVs) {materialFunction=&/MF_UV_MirrorY}.Result, False=@n1.UVs, Value=@n2)) {name=UV Mirror Y, nodeColor=<rgba 0 0 0 1>}
  NamedRerouteDeclaration(None=StaticSwitch(True=MaterialFunctionCall!(In=@n1.UVs) {materialFunction=&/MF_UV_MirrorX}.Result, False=@n1.UVs, Value=@n2)) {name=UV Mirror X, nodeColor=<rgba 0 0 0 1>}
  n3 = OneMinus(None=@n1.UVs)
  n4 = MaterialFunctionCall(In=@n3) {materialFunction=&MF_UV_MirrorY}
  NamedRerouteDeclaration(None=StaticSwitch(True=@n4.Result, False=@n3, Value=@n2)) {name=UV Mirror Y Inverted, nodeColor=<rgba 0 0 0 1>}
  NamedRerouteDeclaration(None=StaticSwitch(True=MaterialFunctionCall(In=@n3) {materialFunction=&MF_UV_MirrorX}.Result, False=@n3, Value=@n2)) {name=UV Mirror X Inverted, nodeColor=<rgba 0 0 0 1>}
  NamedRerouteDeclaration(None=@n4.Result) {name=Flip Bottom, nodeColor=<rgba 0 0 0 1>}
  NamedRerouteUsage()
  n5 = Constant()
  n6 = TextureCoordinate() {coordinateIndex=3}
  n7 = Divide(A=ComponentMask(None=@n6) {g=true}, B=ComponentMask(None=@n6) {r=true})
  n8 = Multiply(
   A=FunctionInput(Preview=@n5) {bUsePreviewValueAsDefault=true, inputName="Offset  Left", inputType=FunctionInput_Scalar, sortPriority=4}
   B=@n7
  )
  n9 = Multiply(
   A=@n7
   B=FunctionInput(Preview=Constant()) {bUsePreviewValueAsDefault=true, inputName=Angle Left, inputType=FunctionInput_Scalar, sortPriority=4}
  )
  n10 = DotProduct(A=Add(A=NamedRerouteUsage(), B=AppendVector(A=@n8, B=@n5)), B=AppendVector(A=Constant() {r=1}, B=Multiply(A=@n9) {constB=-1}))
  n11 = MaterialFunctionCall(XY=Abs(None=OneMinus(None=@n6))) {materialFunction=&Swizzle}
  n12 = FunctionInput(Preview=Constant()) {bUsePreviewValueAsDefault=true, inputName=MinWidth, inputType=FunctionInput_Scalar}
  n13 = Constant()
  n14 = Max(A=Add(A=@n11.YX, B=AppendVector(A=@n12, B=@n13)), B=@n11.YX)
  n15 = Multiply(
   A=Divide(A=ComponentMask(None=@n14) {g=true}, B=ComponentMask(None=@n14) {r=true})
   B=FunctionInput(Preview=Constant()) {bUsePreviewValueAsDefault=true, inputName=Angle Top, inputType=FunctionInput_Scalar, sortPriority=1}
  )
  n16 = DotProduct(A=NamedRerouteUsage(), B=AppendVector(A=Multiply(A=@n15) {constB=-1}, B=Constant() {r=1}))
  n17 = Constant()
  n18 = Multiply(
   A=@n7
   B=FunctionInput(Preview=@n17) {bUsePreviewValueAsDefault=true, inputName="Offset  Right", inputType=FunctionInput_Scalar, sortPriority=4}
  )
  n19 = Multiply(
   A=@n7
   B=FunctionInput(Preview=Constant() {r=0.4}) {bUsePreviewValueAsDefault=true, inputName=Angle Right, inputType=FunctionInput_Scalar, sortPriority=2}
  )
  n20 = DotProduct(A=Add(A=NamedRerouteUsage(), B=AppendVector(A=@n18, B=@n17)), B=AppendVector(A=Constant() {r=1}, B=Multiply(A=@n19) {constB=-1}))
  n21 = MaterialFunctionCall(XY=@n6) {materialFunction=&Swizzle}
  n22 = Constant()
  n23 = Max(A=Add(A=@n21.YX, B=AppendVector(A=@n12, B=@n22)), B=@n21.YX)
  n24 = Multiply(
   A=Divide(A=ComponentMask(None=@n23) {g=true}, B=ComponentMask(None=@n23) {r=true})
   B=FunctionInput(Preview=Constant()) {bUsePreviewValueAsDefault=true, inputName=Angle Bottom, inputType=FunctionInput_Scalar, sortPriority=3}
  )
  n25 = DotProduct(A=NamedRerouteUsage(), B=AppendVector(A=Multiply(A=@n24) {constB=-1}, B=Constant() {r=1}))
  n26 = FunctionOutput(None=Min(A=Min(A=@n10, B=@n16), B=Min(A=@n20, B=@n25)))
  SmoothStep(Min=@n26, Value=@n26)
  Step(Y=@n26) {constX=0.01}
 ]}
