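# Translucent UI master material for sidebar boxes, drawing a skewed fill and border that blend to "Ending" colors and opacities through IsHovered.
# Two skew-box functions are smoothstepped and subtracted to form the border, with inner scale corrected by aspect ratios from UV channel 3.
# A custom rhombus SDF builds a triangular fishtail mask on the right edge, apparently subtracted from opacity through a named reroute to cut a notch.
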Material /Game/UI/Sidebar/Materials/BoxSkew_Master
graph {
 properties {
  bUsedWithStaticMesh=false
  blendMode=BLEND_Translucent
  materialDomain=MD_UI
  pixelDepthOffsetMode=PDOM_Legacy
  usedShadingModels=Default Lit
 }
 nodes [
  n1 = VectorParameter() {parameterName=EndingOuterColor, defaultValue=<rgba 0.625 0.625 0.625 1>}
  n2 = VectorParameter() {parameterName=OuterColor, defaultValue=<rgba 1 1 1 1>}
  n3 = ScalarParameter() {group=Hover Animation, parameterName=IsHovered, sliderMax=1, sliderMin=-1, sortPriority=33}
  n4 = Constant()
  n5 = Multiply(A=ScalarParameter() {defaultValue=0.01, parameterName=Smoothness}, B=NamedRerouteUsage())
  n6 = ScalarParameter() {parameterName=MinWidth}
  n7 = ScalarParameter() {group=Inner Skew Box, parameterName=I_Angle_Top}
  n8 = ScalarParameter() {group=Inner Skew Box, parameterName=I_Angle_Right}
  n9 = ScalarParameter() {group=Inner Skew Box, parameterName=I_Angle_Bottom}
  n10 = ScalarParameter() {group=Inner Skew Box, parameterName=I_Angle_Left}
  n11 = ScalarParameter() {group=Inner Skew Box, parameterName=I_Offset_Right}
  n12 = ScalarParameter() {group=Inner Skew Box, parameterName=I_Offset_Left}
  n13 = TextureCoordinate() {coordinateIndex=3}
  n14 = ComponentMask(None=@n13) {g=true}
  n15 = ComponentMask(None=@n13) {r=true}
  n16 = Divide(A=@n14, B=@n15)
  n17 = Divide(A=@n15, B=@n14)
  n18 = AppendVector(A=@n16, B=@n17)
  n19 = OneMinus(None=Multiply(A=ScalarParameter() {defaultValue=0.4, group=Inner Skew Box, parameterName=I_SkewScale}, B=Saturate(None=@n18)))
  n20 = StaticBool()
  n21 = SmoothStep(
   Min=@n4
   Max=@n5
   Value=MaterialFunctionCall!(
    MinWidth=@n6
    Angle Top=ScalarParameter() {group=Outer Skew Box, parameterName=O_Angle_Top}
    Angle Right=ScalarParameter() {group=Outer Skew Box, parameterName=O_Angle_Right}
    Angle Bottom=ScalarParameter() {group=Outer Skew Box, parameterName=O_Angle_Bottom}
    Angle Left=ScalarParameter() {group=Outer Skew Box, parameterName=O_Angle_Left}
    Scale=ScalarParameter() {defaultValue=1, group=Outer Skew Box, parameterName=O_SkewScale}
    Flip=StaticBool()
   ) {materialFunction=&/MF_BoxSkew}.Result
  )
  n22 = Subtract(
   A=@n21
   B=SmoothStep(
    Min=@n4
    Max=@n5
    Value=MaterialFunctionCall!(
     MinWidth=@n6
     Angle Top=@n7
     Angle Right=@n8
     Angle Bottom=@n9
     Angle Left=@n10
     "Offset  Right"=@n11
     "Offset  Left"=@n12
     Scale=@n19
     Flip=@n20
    ) {materialFunction=&/MF_BoxSkew1}.Result
   )
  )
  n23 = VectorParameter() {parameterName=InnerColor, defaultValue=<rgba 0.0104 0.0104 0.0104 1>}
  n24 = VectorParameter() {parameterName=EndingInnerColor, defaultValue=<rgba 0.3372 0.3372 0.3372 1>}
  LinearInterpolate(
   A=Multiply(A=LinearInterpolate(A=@n23.RGB, B=@n24.RGB, Alpha=@n3) {constA=1, constB=1.2}, B=@n21)
   B=LinearInterpolate(A=@n2.RGB, B=@n1.RGB, Alpha=@n3) {constA=1, constB=1.2}
   Alpha=@n22
  )
  n25 = Constant()
  n26 = Multiply(A=ScalarParameter() {defaultValue=0.01, parameterName=Smoothness}, B=NamedRerouteUsage())
  MaterialExpressionSubtract_1 = Subtract(A=ConstantBiasScale(None=TextureCoordinate()) {bias=-0.5, scale=2}, B=AppendVector(A=Constant() {r=1}, B=Constant()))
  MaterialExpressionAppendVector_1 = AppendVector(A=Multiply(A=NamedRerouteUsage(), B=ScalarParameter() {defaultValue=0.25, parameterName=Fishtail Width}), B=Constant() {r=1})
  n27 = Custom!(p=@MaterialExpressionSubtract_1, b=@MaterialExpressionAppendVector_1) {
   code="float2 q = abs(p);\r\nfloat h = clamp((-2.0*(q.x*b.x - q.y*b.y)+(b.x*b.x - b.y*b.y))/dot(b,b),-1.0,1.0);\r\nfloat d = length( q - 0.5*b*float2(1.0-h,1.0+h) );\r\nreturn d * sign( q.x*b.y + q.y*b.x - b.x*b.y );"
   description=Rhombus
   inputs [
    {inputName=p, input {expression=&self:MaterialExpressionSubtract_1}}
    {inputName=b, input {expression=&self:MaterialExpressionAppendVector_1}}
   ]
  }
  NamedRerouteDeclaration(None=OneMinus(None=SmoothStep(Min=@n25, Max=@n26, Value=@n27))) {name=Isosceles Triangle, nodeColor=<rgba 0 0 0 1>}
  NamedRerouteDeclaration(None=@n16) {name=ScaleX, nodeColor=<rgba 0 0 0 1>}
  NamedRerouteDeclaration(None=@n17) {name=ScaleY, nodeColor=<rgba 0 0 0 1>}
  n28 = Multiply(A=LinearInterpolate(A=@n23.A, B=@n24.A, Alpha=@n3) {constA=1, constB=1.2}, B=@n21)
  n29 = LinearInterpolate(A=@n2.A, B=@n1.A, Alpha=@n3) {constA=1, constB=1.2}
  Saturate(None=Subtract(A=LinearInterpolate(A=@n28, B=@n29, Alpha=@n22), B=NamedRerouteUsage()))
  SmoothStep()
 ]
}
