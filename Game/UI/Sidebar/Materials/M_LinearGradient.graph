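# M_LinearGradient is a translucent UI material that produces a two-colour linear gradient, stored with the sidebar UI assets.
# UVs pass through CustomRotator, driven by a 0-1 Rotation parameter, and the red channel becomes the gradient axis.
# A SmoothStep with GradientClamp as its upper bound shapes that axis into a blend alpha.
# That alpha interpolates Color1 to Color2 separately for RGB and alpha, so transparency can follow the gradient.
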Material /Game/UI/Sidebar/Materials/M_LinearGradient
graph {
 properties {
  bUsedWithStaticMesh=false
  blendMode=BLEND_Translucent
  materialDomain=MD_UI
  pixelDepthOffsetMode=PDOM_Legacy
  usedShadingModels=Default Lit
 }
 nodes [
  n1 = VectorParameter() {parameterName=Color1, defaultValue=<rgba 0 0 0 1>}
  n2 = VectorParameter() {parameterName=Color2}
  n3 = TextureCoordinate()
  n4 = ScalarParameter() {parameterName=Rotation, sliderMax=1}
  n5 = SmoothStep(
   Max=ScalarParameter() {defaultValue=1, parameterName=GradientClamp, sliderMax=1}
   Value=ComponentMask(None=MaterialFunctionCall(UVs=@n3, "Rotation Angle (0-1)"=@n4) {materialFunction=&CustomRotator}.Rotated Values) {r=true}
  )
  LinearInterpolate(A=@n1.RGB, B=@n2.RGB, Alpha=@n5)
  LinearInterpolate(A=@n1.A, B=@n2.A, Alpha=@n5)
 ]
}
