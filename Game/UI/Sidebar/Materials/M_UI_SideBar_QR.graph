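# Translucent UI material for a QR code in the sidebar, defaulting to the social panel's add-friend QR texture.
# A swappable Texture parameter is sampled in linear colour using the normalized UV from GetUserInterfaceUV.
# A Color vector parameter, white by default, sits alongside the sample, with no links from either to each other or to material outputs shown.
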Material /Game/UI/Sidebar/Materials/M_UI_SideBar_QR
graph {
 properties {
  bUsedWithStaticMesh=false
  blendMode=BLEND_Translucent
  materialDomain=MD_UI
  pixelDepthOffsetMode=PDOM_Legacy
  usedShadingModels=Default Lit
 }
 nodes [
  n1 = TextureObjectParameter() {parameterName=Texture, samplerType=SAMPLERTYPE_LinearColor, sortPriority=0, texture=/Game/UI/SocialPanel/Art/Textures/T_UI_QR_AddFriend}
  TextureSample(UVs=MaterialFunctionCall() {materialFunction=&GetUserInterfaceUV}.Normalized UV, Tex=@n1)
  VectorParameter() {parameterName=Color, sortPriority=1, defaultValue=<rgba 1 1 1 1>}
 ]
}
