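# Translucent UI material for the sidebar that, by its name, draws lines for a QR element: thin vertical strokes at the four corners.
# Centered, mirrored UVs give a Thickness-wide edge band, intersected with corner regions set by NegativeSpacePct to leave gaps between strokes.
# The band adds the R channel to itself, so only left and right edges draw.
# Ceil makes the mask hard-edged; Opacity scales it, and a Color parameter provides tint.
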Material /Game/UI/Sidebar/Materials/M_UI_SideBar_QR_Lines
graph {
 properties {
  bUsedWithStaticMesh=false
  blendMode=BLEND_Translucent
  materialDomain=MD_UI
  pixelDepthOffsetMode=PDOM_Legacy
  usedShadingModels=Default Lit
 }
 nodes [
  n1 = MaterialFunctionCall() {materialFunction=&GetUserInterfaceUV}
  n2 = Multiply(A=Abs(None=Subtract(A=@n1.Normalized UV) {constB=0.5})) {constB=2}
  n3 = OneMinus(None=ScalarParameter() {defaultValue=0.02, parameterName=Thickness})
  n4 = MaterialFunctionCall(Float2=Saturate(None=Subtract(A=@n2, B=@n3))) {materialFunction=&BreakOutFloat2Components}
  n5 = Saturate(None=Add(A=@n4.R, B=@n4.R))
  n6 = ScalarParameter() {defaultValue=0.85, parameterName=NegativeSpacePct}
  n7 = Ceil(None=Saturate(None=Subtract(A=@n2, B=@n6) {constB=0.9}))
  n8 = ComponentMask(None=@n7) {r=true}
  n9 = ComponentMask(None=@n7) {g=true}
  Multiply(A=Min(A=Ceil(None=@n5), B=Min(A=@n8, B=@n9)), B=ScalarParameter() {defaultValue=0.2, parameterName=Opacity}) {constB=0.2}
  VectorParameter() {parameterName=Color, sortPriority=1, defaultValue=<rgba 1 1 1 1>}
 ]
}
