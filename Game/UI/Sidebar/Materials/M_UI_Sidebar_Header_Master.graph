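# Translucent UI master material for sidebar headers: an animated, patterned two-color background.
# Two color parameters each form an exponential radial gradient from a darkened shade, placed with screen-aligned UVs.
# A drifting triangle grid and a rotated, tiled, panning banner icon blend between these gradients in screen space, while scrolling noise washes the pattern toward the second.
# A static switch enables an ease-out cubic vertical fade from widget UVs.
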Material /Game/UI/Sidebar/Materials/M_UI_Sidebar_Header_Master
graph {
 properties {
  bUsedWithStaticMesh=false
  blendMode=BLEND_Translucent
  materialDomain=MD_UI
  pixelDepthOffsetMode=PDOM_Legacy
  usedShadingModels=Default Lit
 }
 nodes [
  n1 = VectorParameter() {parameterName=RGBA1, defaultValue=<rgba 0.1339 0.286 0.525 1>}
  n2 = Constant() {r=0.15}
  n3 = MaterialFunctionCall(
   UVs=MaterialFunctionCall() {materialFunction=&ScreenAlignedUVs}."X Scale to Ratio, Y 100%"
   CenterPosition=AppendVector(A=Constant() {r=0.5}, B=Constant() {r=1})
   Radius=Constant() {r=1.16}
   Density=Constant() {r=1.82}
  ) {materialFunction=&RadialGradientExponential}
  n4 = LinearInterpolate(A=Saturate(None=Multiply(A=@n1.RGB, B=@n2)), B=@n1.RGB, Alpha=@n3.RadialGradientExponential)
  n5 = LinearInterpolate(
   A=Saturate(None=Multiply(A=@n1.RGB) {constB=0.25})
   B=VectorParameter() {parameterName=RGBA0, defaultValue=<rgba 0.2279 0.3876 0.7991 1>}.RGB
   Alpha=@n3.RadialGradientExponential
  )
  n6 = NamedRerouteDeclaration(None=MaterialFunctionCall!(ScreenUV, TextureSize, ScalePivotPoint) {materialFunction=/FortUILibrary/MaterialFunctions/Basics/MF_UI_ScreenSpaceUVScaling}.Scaled to Fit) {name=ScreenSpace UV, nodeColor=<rgba 0 0 0 1>}
  n7 = Panner(
   Coordinate=Rotator(Coordinate=@n6, Time=ScalarParameter() {defaultValue=-0.0571, group=Triangles, parameterName=Tri Angle}) {speed=6.2832}
   Speed=AppendVector(A=ScalarParameter() {defaultValue=0.0025, group=Triangles, parameterName=PanSpeed}, B=Constant())
  )
  n8 = ScalarParameter() {defaultValue=5, group=Triangles, parameterName=Tri Tile Amount}
  n9 = Multiply(
   A=TextureSample(UVs=Multiply(A=@n7, B=@n8), Tex=TextureObject() {texture=/Game/Global/Textures/Default/Masks_and_Gradients/T-TriangleGrid}).R
   B=ScalarParameter() {defaultValue=0.2, group=Triangles, parameterName=TriangleDensity}
  )
  n10 = Panner(
   Coordinate=Multiply(A=@n6, B=ScalarParameter() {defaultValue=10, group=effect, parameterName=TileScreenSpace})
   Speed=AppendVector(A=ScalarParameter() {defaultValue=0.5, parameterName=Panner X speed}, B=ScalarParameter() {defaultValue=0.5, parameterName=Panner Y speed})
  ) {speedX=0.005, speedY=0.005}
  n11 = Constant() {r=0.25}
  n12 = Multiply(A=Frac(None=Rotator(Coordinate=@n10, Time=@n11) {speed=-1.3414}), B=Constant() {r=0.9})
  n13 = TextureObjectParameter() {parameterName=BannerIcon, texture=/Game/2dAssets/Banners/Season6/T-Banners-Icons-Skirmish-LuckyLlamas}
  n14 = Multiply(A=Time(), B=ScalarParameter() {defaultValue=0.025, parameterName=Noise-Mask-OffSetTime})
  n15 = Saturate(None=TextureSample(UVs=Panner(Coordinate=@n6, Time=@n14) {speedY=1}) {samplerType=SAMPLERTYPE_Grayscale, texture=&LowResBlurredNoise}.R)
  n16 = Constant() {r=1.5}
  LinearInterpolate(
   A=LinearInterpolate(A=LinearInterpolate(A=@n4, B=@n5, Alpha=@n9) {constB=0.1}, B=@n5, Alpha=TextureSample(UVs=@n12, Tex=@n13).R) {constB=0.1}
   B=@n5
   Alpha=Saturate(None=Multiply(A=@n15, B=@n16) {constB=1.5})
  )
  n17 = ComponentMask(None=MaterialFunctionCall() {materialFunction=&GetUserInterfaceUV}.Normalized UV) {g=true}
  StaticSwitchParameter(
   True=MaterialFunctionCall("Value (0-1)"=OneMinus(None=@n17)) {materialFunction=&MF_UI_EaseCurves}.EaseOutCubic
   False=Constant() {r=1}
  ) {group=Fade, parameterName=Enable Fade}
 ]
}
