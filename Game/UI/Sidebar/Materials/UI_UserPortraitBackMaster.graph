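# Translucent UI master material for the sidebar's user portrait background, with parameterized active, inactive, Kairos and empty-portrait looks.
# Radial-gradient color blends feed a UseColorBackground switch, then an optional radial stroke.
# A separate mask chain offers round or rotated, side-clipped shapes.
# An optional reveal combines three scaled shatter-hole decals and radial gradients as smoothstep thresholds on a triangle pattern, giving a patterned dissolve.
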Material /Game/UI/Sidebar/Materials/UI_UserPortraitBackMaster
graph {
 properties {
  bUsedWithStaticMesh=false
  blendMode=BLEND_Translucent
  materialDomain=MD_UI
  numCustomizedUVs=3
  pixelDepthOffsetMode=PDOM_Legacy
  usedShadingModels=Default Lit
 }
 nodes [
  n1 = ScalarParameter() {defaultValue=0.5, group=Stroke, parameterName=StrokeRadius}
  n2 = ScalarParameter() {defaultValue=10, group=Stroke, parameterName=StrokeDensity}
  n3 = VectorParameter() {parameterName=BackgroundRGB, defaultValue=<rgba 0.013 0.3613 0.6376 1>}
  n4 = Multiply(A=MaterialFunctionCall(CenterPosition=Constant2Vector() {g=0.75, r=0.5}, Radius=Constant() {r=0.75}, Density=Constant() {r=1.25}) {materialFunction=&RadialGradientExponential}.RadialGradientExponential) {constB=0.5}
  n5 = LinearInterpolate(A=@n3.RGB, B=VectorParameter() {parameterName=HighlightRGB, defaultValue=<rgba 0.4452 1 1 1>}.RGB, Alpha=@n4)
  n6 = LinearInterpolate(
   A=LinearInterpolate(A=VectorParameter() {parameterName=InactiveBackgroundRGB, defaultValue=<rgba 0 0.0999 0.3095 1>}.RGB, B=@n3.RGB, Alpha=@n4)
   B=LinearInterpolate(A=@n5, B=VectorParameter() {parameterName=InactiveRGBALerp, defaultValue=<rgba 0 0.0168 0.0887 0.5>}.RGB, Alpha=Constant() {r=0.75})
   Alpha=ScalarParameter() {parameterName=IsKairosPortrait}
  )
  n7 = ScalarParameter() {parameterName=IsInactive, sliderMax=1}
  n8 = LinearInterpolate(
   A=VectorParameter() {parameterName=NoPortraitFill, defaultValue=<rgba 0 0.0999 0.3095 1>}.RGB
   B=VectorParameter() {parameterName=NoPortraitHighlight, defaultValue=<rgba 0 0.3228 1 1>}.RGB
   Alpha=@n4
  )
  n9 = LinearInterpolate(A=ScalarParameter() {defaultValue=0.5, parameterName=DesaturationFactor}, B=Constant(), Alpha=ScalarParameter() {parameterName=IsEmpty})
  n10 = VectorParameter() {parameterName=InactiveRGBALerp, defaultValue=<rgba 0 0.0168 0.0887 0.5>}
  n11 = VectorParameter() {parameterName=PortraitFill}
  n12 = LinearInterpolate(
   A=LinearInterpolate(A=@n8, B=@n11.RGB, Alpha=@n11.RGB)
   B=LinearInterpolate(A=Desaturation(None=@n8, Fraction=@n9), B=@n10.RGB, Alpha=@n10.RGB)
   Alpha=@n7
  )
  n13 = StaticSwitchParameter(True=LinearInterpolate(A=LinearInterpolate(A=@n5, B=@n6, Alpha=@n7), B=@n12) {constAlpha=0}, False=@n12) {defaultValue=true, parameterName=UseColorBackground}
  LinearInterpolate(
   A=@n13
   B=LinearInterpolate(
    A=VectorParameter() {group=Stroke, parameterName=StrokeColor}.RGB
    B=@n13
    Alpha=MaterialFunctionCall(Radius=@n1, Density=@n2) {materialFunction=&RadialGradientExponential}.RadialGradientExponential
   ) {constAlpha=0}
   Alpha=ScalarParameter() {group=Stroke, parameterName=UseStroke}
  ) {constAlpha=0}
  n14 = VectorParameter() {group=MaskReveal, parameterName=RevealDecal1}
  n15 = MaterialFunctionCall!(Texture Scale=@n14.A, UVs=TextureCoordinate(), Offset XY Texturesize Z=@n14.A) {materialFunction=/Engine/Functions/Engine_MaterialFunctions01/Texturing/ScaleUVsAroundPoint}
  n16 = VectorParameter() {group=MaskReveal, parameterName=RevealDecal2}
  n17 = MaterialFunctionCall(Texture Scale=@n16.A, UVs=TextureCoordinate(), Offset XY Texturesize Z=@n16.A) {materialFunction=&ScaleUVsAroundPoint}
  n18 = VectorParameter() {group=MaskReveal, parameterName=RevealDecal3}
  n19 = TextureCoordinate()
  n20 = Max(
   A=Max(
    A=TextureSample(UVs=@n15.UVs) {samplerSource=SSM_Clamp_WorldGroupSettings, texture=/Game/Athena/HUD/Discoverability/materials/T_UI_Shatterhole}.R
    B=TextureSample(UVs=@n17.UVs) {samplerSource=SSM_Clamp_WorldGroupSettings, texture=&T_UI_Shatterhole}.R
   )
   B=TextureSample(UVs=MaterialFunctionCall(Texture Scale=@n18.A, UVs=@n19, Offset XY Texturesize Z=@n18.A) {materialFunction=&ScaleUVsAroundPoint}.UVs) {samplerSource=SSM_Clamp_WorldGroupSettings, texture=&T_UI_Shatterhole}.R
  )
  OneMinus(None=@n20)
  n21 = MaterialFunctionCall(
   Radius=ScalarParameter() {defaultValue=0.5, parameterName=MaskRevealRadius}
   Density=ScalarParameter() {defaultValue=10, parameterName=MaskRevealDensity}
  ) {materialFunction=&RadialGradientExponential}
  n22 = TextureCoordinate()
  n23 = ScalarParameter() {group=PortaitMask, parameterName=Mask Rotation}
  n24 = MaterialFunctionCall(
   UVs=StaticSwitchParameter(
    True=@n21.RadialGradientExponential
    False=MaterialFunctionCall(UVs=@n22, "Rotation Angle (0-1)"=@n23) {materialFunction=&CustomRotator}.Rotated Values
   ) {parameterName=UseRoundMask}
   Texture Scale=ScalarParameter() {defaultValue=1, group=PortaitMask, parameterName=Mask Size}
  ) {materialFunction=&ScaleUVsByCenter}
  n25 = AppendVector(A=ScalarParameter() {defaultValue=1, group=PortaitMask, parameterName=Mask Side Clip}, B=Constant() {r=1})
  n26 = Round(None=ComponentMask(None=TextureCoordinate()) {g=true})
  n27 = Max(
   A=Multiply(
    A=MaterialFunctionCall(UVs=@n22, Texture Scale=@n25) {materialFunction=&ScaleUVsByCenter}.0-1 mask
    B=OneMinus(None=@n26)
   )
   B=@n24.0-1 mask
  )
  n28 = TextureCoordinate() {coordinateIndex=2}
  n29 = OneMinus(None=MaterialFunctionCall(
    Radius=ScalarParameter() {defaultValue=0.2, group=MaskReveal, parameterName=InnerMaskRevealRadius}
    Density=ScalarParameter() {defaultValue=10, group=MaskReveal, parameterName=InnerMaskRevealDensity}
   ) {materialFunction=&RadialGradientExponential}.RadialGradientExponential)
  n30 = Max(A=@n20, B=OneMinus(None=Multiply(A=@n29, B=@n21.RadialGradientExponential)))
  n31 = Multiply(A=@n30, B=ScalarParameter() {group=MaskReveal, parameterName=MaskEdgeSoftness})
  n32 = Multiply(
   A=@n27
   B=MaterialFunctionCall(
    Alpha=TextureSample(UVs=@n28) {texture=/Game/Global/Textures/Default/Masks_and_Gradients/T_TrianglePattern}.R
    Min=@n30
    Max=Add(A=@n30, B=@n31)
   ) {materialFunction=&SmoothStep}.Result
  )
  StaticSwitchParameter(True=@n24.0-1 mask, False=StaticSwitchParameter(True=Multiply(B=@n32) {constA=1}, False=@n27) {parameterName=UseMask}) {parameterName=UseRoundMask}
 ]
}
