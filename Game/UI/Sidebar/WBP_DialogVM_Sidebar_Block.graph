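# Sidebar dialog for blocking a player, built as a CommonActivatableWidget that closes on activation unless the ClamOne runtime option is enabled.
# A visibility switcher starts on the confirmation panel, and Block executes the social interaction viewmodel, then immediately shows the blocked panel.
# That panel offers Close or Report, which opens the player feedback modal.
# Focus follows the active panel; the loading panel and name setters go unused at runtime in these graphs.
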WidgetBlueprint /Game/UI/Sidebar/WBP_DialogVM_Sidebar_Block
graphs [
 BP_GetDesiredFocusTarget {nodes [
   Get Desired Focus Target()
   n1 = Get S_Content() -> (S_Content:Common Visibility Switcher Object Reference)
   > Return Node!(
    execute
    ReturnValue:Widget Object Reference=Select!(
     Option 0:Widget Object Reference=Get Button_Cancel() -> (Button_Cancel:WBP UIKit Button Regular Object Reference)
     Option 1:Widget Object Reference
     Option 2:Widget Object Reference=Get Button_Close() -> (Button_Close:WBP UIKit Button Regular Object Reference)
     Index:int=GetActiveWidgetIndex!(self:Common Visibility Switcher Object Reference=@n1) -> (ReturnValue:int=0)
    ) -> (ReturnValue:Widget Object Reference)
   )
  ]}
 EventGraph {nodes [
   n1 = "On Clicked (Button_Cancel)"!() -> (OutputDelegate, then, Button:Common Button Base Object Reference) {is=title K2Node_ComponentBoundEvent}
   DeactivateWidget(execute=[@n1.then, @n2.then])
   n2 = "On Clicked (Button_Close)"!() -> (OutputDelegate, then, Button:Common Button Base Object Reference) {is=title K2Node_ComponentBoundEvent}
   "On Clicked (Button_Block)"!() -> (OutputDelegate, then, Button:Common Button Base Object Reference) {is=title K2Node_ComponentBoundEvent}
   > ExecuteInteraction(self:Fort Social User Interaction Viewmodel Object Reference=Get FortSocialUserInteractionViewModel() -> (FortSocialUserInteractionViewModel:Fort Social User Interaction Viewmodel Object Reference))
   > SetActiveWidget(self=Get S_Content(), Widget=Get GP_PlayerBlocked() -> (GP_PlayerBlocked:Grid Panel Object Reference))
   > RequestRefreshFocus(self:Common Activatable Widget Object Reference)
   "On Clicked (Button_Report)"!() -> (OutputDelegate, then, Button:Common Button Base Object Reference) {is=title K2Node_ComponentBoundEvent}
   > DeactivateWidget()
   n3 = Get FortSocialUserInteractionViewModel()
   > RequestToOpenSpecificPlayerFeedbackModal!(
    execute
    Context:User Widget Object Reference=Self-Reference()
    UserToReport:Social User Object Reference=GetSocialUser!(self:Fort Social User Viewmodel Object Reference=@n3) -> (ReturnValue:Fort Social User Object Reference)
    InAnalyticsContext:Map of Strings to Strings=GetAnalyticsContext!(self:Fort Social User Interaction Viewmodel Object Reference=@n3) -> (ReturnValue:Map of Strings to Strings)
   ) -> (then)
   n4 = Event PreConstruct()
   > n5 = Sequence()
   > Branch(Condition=@n4.IsDesignTime)
   > SetPlayerNameOnConfirmationText!(execute, self, PlayerText:String=VeryLongPlayerName) -> (then)
   > SetPlayerNameOnResultText!(execute, self, PlayerName:String=VeryLongPlayerName) -> (then)
   SetColorAndOpacity!(
    execute=@n5.then_1
    self=[
     Get BlockDescriptionText() -> (BlockDescriptionText:Common Text Object Reference)
     Get BlockedDescriptionText() -> (BlockedDescriptionText:Common Text Object Reference)
    ]
    InColorAndOpacity_SpecifiedColor:Linear Color Structure=(R=0.521,G=0.521,B=0.521,A=1)
    InColorAndOpacity_ColorUseRule:ESlateColorStylingMode Enum
   ) -> (then)
   Event On Activated()
   > n6 = Branch(Condition=GetIsClamOneEnabled!(self:Fort Runtime Options Object Reference=GetRuntimeOptions()) -> (ReturnValue:bool))
   DeactivateWidget(execute=@n6.else)
   Event Construct()
   > SetActiveWidget(self=Get S_Content(), Widget=Get GP_BlockPlayer() -> (GP_BlockPlayer:Grid Panel Object Reference))
  ]}
 SetPlayerNameOnConfirmationText {nodes [
   n1 = SetPlayerNameOnConfirmationText!() -> (then, PlayerText:String) {is=title K2Node_FunctionEntry}
   > "SetText (Text)"(
    self=Get BlockPlayerNameText() -> (BlockPlayerNameText:Common Text Object Reference)
    InText=Format Text!(Format="LOCTEXT(\"Would you like to block {Player_Name}?\")", Player_Name:Text="To Text (String)"(InString=@n1)) -> (Result)
   )
  ]}
 SetPlayerNameOnResultText {nodes [
   n1 = SetPlayerNameOnResultText!() -> (then, PlayerName:String) {is=title K2Node_FunctionEntry}
   > "SetText (Text)"(
    self=Get BlockedPlayerNameText() -> (BlockedPlayerNameText:Common Text Object Reference)
    InText=Format Text(Format="LOCTEXT(\"You’ve blocked {Player_Name}.\")", Player_Name="To Text (String)"(InString=@n1))
   )
  ]}
]
variables []
widgets {
 info {
  inheritedWidgetCount=0
  namedSlotCount=1
  parentClass=&CommonActivatableWidget
  rootWidgetClass=/FortUILibrary/Systems/Dialog/WBP_UIKit_Dialog_Internal.WBP_UIKit_Dialog_Internal_C
  widgetCount=21
 }
 widgets [
  Dialog_Internal {
   widgetClassPath=/FortUILibrary/Systems/Dialog/WBP_UIKit_Dialog_Internal
   bIsVariable=true
   widget=Dialog_Internal
   properties {
    titleImageSizeX=64
    titleImageSizeY=64
    titleTextVerticalAligment=VAlign_Center
    useTitle=false
    buttons {
     each Button Horizontal Fill=false
     each Button Min Width=258
     each Button Size=Medium
     focuse Button=0
     space Between Buttons=24
     cTA Buttons Types []
     group Placement {
      horizontal Alignment=HAlign_Fill
      vertical Alignment=VAlign_Bottom
      grid {column=1, column Span=0, layer=0, row=2, row Span=0, nudge=<0 0>}
      padding=<ltrb 80 64 80 48>
     }
    }
    dialogSIze {
     height Override {enable=false, value=0}
     max Aspect Ratio {enable=false, value=0}
     max Desired Height {enable=false, value=800}
     max Desired Width {enable=false, value=0}
     min Aspect Ratio {enable=false, value=0}
     min Desired Height {enable=true, value=485}
     min Desired Width {enable=false, value=0}
     width Override {enable=true, value=866}
    }
    titleImagePadding=<ltrb 10 0 15 0>
    titleProperties {
     auto Wrap Text=true
     justification=Left
     line Height Percentage=1
     mobile Font Size Multiplier=1.3
     transform Policy=ToUpper
     wrap Text At=700
     color {colorUseRule=UseColor_Specified, specifiedColor=<rgba 1 1 1 1>}
     font {
      bForceMonospaced=false
      bMaterialIsStencil=false
      fontMaterial=null
      fontObject=&HeadingNow
      letterSpacing=0
      monospacedWidth=1
      size=25.5
      skewAmount=0
      typefaceFontName=86 Bold Italic
      outlineSettings {
       bApplyOutlineToDropShadows=false
       bMiteredCorners=false
       bSeparateFillAlpha=false
       outlineMaterial=null
       outlineSize=0
       outlineColor=<rgba 0 0 0 1>
      }
     }
     placement {
      horizontal Alignment=HAlign_Fill
      vertical Alignment=VAlign_Top
      grid {column=1, column Span=0, layer=0, row=0, row Span=0, nudge=<0 0>}
      padding=<ltrb 80 80 80 32>
     }
    }
   }
  }
  GridPanel_3 {
   widgetClassPath=&GridPanel
   namedSlotHost=Dialog_Internal
   widget=GridPanel_3
   properties {bOverride_Cursor=true, columnFill [1], rowFill [1]}
  }
  S_Content {
   widgetClassPath=&CommonVisibilitySwitcher
   bIsVariable=true
   parent=GridPanel_3
   widget=S_Content
   properties {bAutoActivateSlot=false}
   slot {class=GridSlot, properties {padding=<ltrb 80 80 80 48>}}
  }
  GP_BlockPlayer {
   widgetClassPath=&GridPanel
   bIsVariable=true
   parent=S_Content
   widget=GP_BlockPlayer
   properties {columnFill [1], rowFill [0,1,0]}
   slot {class=CommonVisibilitySwitcherSlot}
  }
  TitleText {
   widgetClassPath=&CommonTextBlock
   parent=GP_BlockPlayer
   widget=TitleText
   properties {
    autoWrapText=true
    bIsScrollingEnabled=false
    style=&TextStyle-Base_C
    text=Block Player
    wrappingPolicy=AllowPerCharacterWrapping
    font {
     bForceMonospaced=false
     bMaterialIsStencil=false
     fontMaterial=null
     fontObject=&NotoSans
     letterSpacing=0
     monospacedWidth=1
     size=24
     skewAmount=0
     typefaceFontName=NotoSans-Regular
     outlineSettings {
      bApplyOutlineToDropShadows=false
      bMiteredCorners=false
      bSeparateFillAlpha=false
      outlineMaterial=null
      outlineSize=0
      outlineColor=<rgba 0 0 0 1>
     }
    }
    shadowOffset=<0 0>
   }
   slot {class=GridSlot}
  }
  VerticalBox_0 {
   widgetClassPath=&VerticalBox
   parent=GP_BlockPlayer
   widget=VerticalBox_0
   slot {class=GridSlot, properties {row=1, padding=<ltrb 0 48 0 64>}}
  }
  BlockPlayerNameText {
   widgetClassPath=&CommonTextBlock
   bIsVariable=true
   parent=VerticalBox_0
   widget=BlockPlayerNameText
   properties {
    autoWrapText=true
    bIsScrollingEnabled=false
    style=&TextStyle-Base_C
    text="Would you like to block {Player_Name}?"
    font {
     bForceMonospaced=false
     bMaterialIsStencil=false
     fontMaterial=null
     fontObject=&NotoSans
     letterSpacing=0
     monospacedWidth=1
     size=24
     skewAmount=0
     typefaceFontName=NotoSans-Regular
     outlineSettings {
      bApplyOutlineToDropShadows=false
      bMiteredCorners=false
      bSeparateFillAlpha=false
      outlineMaterial=null
      outlineSize=0
      outlineColor=<rgba 0 0 0 1>
     }
    }
    shadowOffset=<0 0>
   }
   slot {class=VerticalBoxSlot, properties {verticalAlignment=VAlign_Top, padding=<ltrb 0 0 0 20>}}
  }
  BlockDescriptionText {
   widgetClassPath=&CommonTextBlock
   bIsVariable=true
   parent=VerticalBox_0
   widget=BlockDescriptionText
   properties {
    autoWrapText=true
    bIsScrollingEnabled=false
    style=&TextStyle-Base_C
    text="Blocked players will be removed from your friends list, cannot invite you to their party, and cannot voice chat or text chat with you. Blocked players will not be notified that you blocked them."
    font {
     bForceMonospaced=false
     bMaterialIsStencil=false
     fontMaterial=null
     fontObject=&NotoSans
     letterSpacing=0
     monospacedWidth=1
     size=24
     skewAmount=0
     typefaceFontName=NotoSans-Regular
     outlineSettings {
      bApplyOutlineToDropShadows=false
      bMiteredCorners=false
      bSeparateFillAlpha=false
      outlineMaterial=null
      outlineSize=0
      outlineColor=<rgba 0 0 0 1>
     }
    }
    shadowOffset=<0 0>
   }
   slot {class=VerticalBoxSlot, properties {verticalAlignment=VAlign_Top, size {sizeRule=Fill, value=1}}}
  }
  GridPanel_1 {
   widgetClassPath=&GridPanel
   parent=GP_BlockPlayer
   widget=GridPanel_1
   slot {class=GridSlot, properties {horizontalAlignment=HAlign_Left, row=2}}
  }
  Button_Block {
   widgetClassPath=/FortUILibrary/UIKit/Buttons/WBP_UIKit_Button_Loud
   bIsVariable=true
   parent=GridPanel_1
   widget=Button_Block
   properties {
    bDisplayInputActionWhenNotInteractable=false
    bShouldUseFallbackDefaultInputAction=false
    minWidthOverride=220
    navigation=&self:WidgetTree.Button_Block.WidgetNavigation_0
    text=Block
   }
   slot {class=GridSlot, properties {horizontalAlignment=HAlign_Left}}
  }
  Button_Cancel {
   widgetClassPath=&WBP_UIKit_Button_Regular
   bIsVariable=true
   parent=GridPanel_1
   widget=Button_Cancel
   properties {
    bDisplayInputActionWhenNotInteractable=false
    bShouldUseFallbackDefaultInputAction=false
    buttonType=Quiet
    minWidthOverride=205
    navigation=&self:WidgetTree.Button_Cancel.WidgetNavigation_0
    text=Cancel
   }
   slot {class=GridSlot, properties {column=1, horizontalAlignment=HAlign_Left, padding=<ltrb 24 0 0 0>}}
  }
  GP_Loading {
   widgetClassPath=&GridPanel
   bIsVariable=true
   parent=S_Content
   widget=GP_Loading
   properties {columnFill [1], rowFill [1]}
   slot {class=CommonVisibilitySwitcherSlot}
  }
  LoadingSpinner {
   widgetClassPath=/FortUILibrary/UIKit/Throbber/WBP_UIKit_Throbber
   parent=GP_Loading
   widget=LoadingSpinner
   properties {displayBackground=false, throbberSize=Large}
   slot {class=GridSlot, properties {horizontalAlignment=HAlign_Center, verticalAlignment=VAlign_Center}}
  }
  GP_PlayerBlocked {
   widgetClassPath=&GridPanel
   bIsVariable=true
   parent=S_Content
   widget=GP_PlayerBlocked
   properties {columnFill [1], rowFill [0,1,0]}
   slot {class=CommonVisibilitySwitcherSlot}
  }
  TitleText_1 {
   widgetClassPath=&CommonTextBlock
   parent=GP_PlayerBlocked
   widget=TitleText_1
   properties {
    autoWrapText=true
    bIsScrollingEnabled=false
    style=&TextStyle-Base_C
    text=Player Blocked
    wrappingPolicy=AllowPerCharacterWrapping
    font {
     bForceMonospaced=false
     bMaterialIsStencil=false
     fontMaterial=null
     fontObject=&NotoSans
     letterSpacing=0
     monospacedWidth=1
     size=24
     skewAmount=0
     typefaceFontName=NotoSans-Regular
     outlineSettings {
      bApplyOutlineToDropShadows=false
      bMiteredCorners=false
      bSeparateFillAlpha=false
      outlineMaterial=null
      outlineSize=0
      outlineColor=<rgba 0 0 0 1>
     }
    }
    shadowOffset=<0 0>
   }
   slot {class=GridSlot}
  }
  VerticalBox {
   widgetClassPath=&VerticalBox
   parent=GP_PlayerBlocked
   widget=VerticalBox
   slot {class=GridSlot, properties {row=1, padding=<ltrb 0 48 0 64>}}
  }
  BlockedPlayerNameText {
   widgetClassPath=&CommonTextBlock
   bIsVariable=true
   parent=VerticalBox
   widget=BlockedPlayerNameText
   properties {
    autoWrapText=true
    bIsScrollingEnabled=false
    text="{Player_Name} has been blocked"
    font {
     bForceMonospaced=false
     bMaterialIsStencil=false
     fontMaterial=null
     fontObject=&NotoSans
     letterSpacing=0
     monospacedWidth=1
     size=24
     skewAmount=0
     typefaceFontName=NotoSans-Regular
     outlineSettings {
      bApplyOutlineToDropShadows=false
      bMiteredCorners=false
      bSeparateFillAlpha=false
      outlineMaterial=null
      outlineSize=0
      outlineColor=<rgba 0 0 0 1>
     }
    }
    shadowOffset=<0 0>
   }
   slot {class=VerticalBoxSlot, properties {padding=<ltrb 0 0 0 20>}}
  }
  BlockedDescriptionText {
   widgetClassPath=&CommonTextBlock
   bIsVariable=true
   parent=VerticalBox
   widget=BlockedDescriptionText
   properties {
    autoWrapText=true
    bIsScrollingEnabled=false
    style=&TextStyle-Base_C
    text="Blocked players will be removed from your friends list, cannot invite you to their party, and cannot voice chat or text chat with you. Blocked players will not be notified that you blocked them."
    font {
     bForceMonospaced=false
     bMaterialIsStencil=false
     fontMaterial=null
     fontObject=&NotoSans
     letterSpacing=0
     monospacedWidth=1
     size=24
     skewAmount=0
     typefaceFontName=NotoSans-Regular
     outlineSettings {
      bApplyOutlineToDropShadows=false
      bMiteredCorners=false
      bSeparateFillAlpha=false
      outlineMaterial=null
      outlineSize=0
      outlineColor=<rgba 0 0 0 1>
     }
    }
    shadowOffset=<0 0>
   }
   slot {class=VerticalBoxSlot, properties {verticalAlignment=VAlign_Top, size {sizeRule=Fill, value=1}}}
  }
  GridPanel_4 {
   widgetClassPath=&GridPanel
   parent=GP_PlayerBlocked
   widget=GridPanel_4
   slot {class=GridSlot, properties {horizontalAlignment=HAlign_Left, row=2}}
  }
  Button_Close {
   widgetClassPath=&WBP_UIKit_Button_Regular
   bIsVariable=true
   parent=GridPanel_4
   widget=Button_Close
   properties {
    bDisplayInputActionWhenNotInteractable=false
    bShouldUseFallbackDefaultInputAction=false
    minWidthOverride=220
    navigation=&self:WidgetTree.Button_Close.WidgetNavigation_0
    text=Close
   }
   slot {class=GridSlot, properties {horizontalAlignment=HAlign_Left}}
  }
  Button_Report {
   widgetClassPath=&WBP_UIKit_Button_Regular
   bIsVariable=true
   parent=GridPanel_4
   widget=Button_Report
   properties {
    bDisplayInputActionWhenNotInteractable=false
    bShouldUseFallbackDefaultInputAction=false
    buttonType=Quiet
    minWidthOverride=155
    navigation=&self:WidgetTree.Button_Report.WidgetNavigation_0
    text=Report
   }
   slot {class=GridSlot, properties {column=1, horizontalAlignment=HAlign_Left, padding=<ltrb 24 0 0 0>}}
  }
 ]
}
