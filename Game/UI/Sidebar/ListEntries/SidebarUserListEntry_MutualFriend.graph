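# Sidebar list entry widget for a mutual friend, derived from the native FortMutualFriendSocialUserListEntry.
# The row shows an avatar, Reboot Rally icon and social name, with a right-aligned user action menu anchor.
# OnNewAvatarsEnabledChanged toggles visibility between PlayerAvatar and the older Avatar_MemberIcon, so only one shows.
# Several inherited widgets, including rich presence and last interaction text, have no instance in this layout.
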WidgetBlueprint /Game/UI/Sidebar/ListEntries/SidebarUserListEntry_MutualFriend
graphs [EventGraph {nodes [
   n1 = Event OnNewAvatarsEnabledChanged!() -> (OutputDelegate, then, bUsingNewAvatars:bool) {is=title K2Node_Event}
   > SetVisibility(
    self=Get PlayerAvatar() -> (PlayerAvatar:Fort Player Avatar Object Reference)
    InVisibility=Select(Option 0=Collapsed, Option 1=HitTestInvisible, Index=@n1.bUsingNewAvatars)
   )
   > SetVisibility(
    self=Get Avatar_MemberIcon() -> (Avatar_MemberIcon:Fort Social Avatar Icon Object Reference)
    InVisibility=Select(Option 0=HitTestInvisible, Option 1=Collapsed, Index=@n1.bUsingNewAvatars)
   )
  ]}]
variables [
 EncourageEpicFriend_HoveredColor {type=LinearColor, default=<rgba 1 1 1 1>}
 EncourageEpicFriend_IdleColor {type=LinearColor, default=<rgba 0.956 1 0.0666 1>}
]
widgets {
 info {
  inheritedWidgetCount=13
  parentClass=/Script/FortniteUI.FortMutualFriendSocialUserListEntry
  rootWidgetClass=&SizeBox
  widgetCount=23
 }
 widgets [
  SizeBox_0 {widgetClassPath=&SizeBox, widget=SizeBox_0, properties {bOverride_MinDesiredHeight=true, heightOverride=72, minDesiredHeight=72}}
  Overlay_1 {widgetClassPath=&Overlay, parent=SizeBox_0, widget=Overlay_1, slot {class=SizeBoxSlot}}
  MenuAnchor_Actions {
   widgetClassPath=/Script/UMG.MenuAnchor
   bInherited=true
   bIsVariable=true
   parent=Overlay_1
   widget=MenuAnchor_Actions
   properties {
    menuClass=/Game/UI/SocialPanel/UserActionMenu.UserActionMenu_C
    placement=MenuPlacement_MenuRight
    renderTransform {angle=0, scale=<1 1>, shear=<0 0>, translation=<8 0>}
   }
   slot {class=OverlaySlot, properties {horizontalAlignment=HAlign_Right}}
  }
  Border_Background {
   widgetClassPath=&Border
   bInherited=true
   bIsVariable=true
   parent=Overlay_1
   widget=Border_Background
   properties {
    visibility=SelfHitTestInvisible
    background {
     bIsDynamicallyLoaded=false
     drawAs=Image
     imageType=NoImage
     mirroring=NoMirror
     resourceName=None
     resourceObject=/Game/UI/Sidebar/Materials/M_UI_sidebar_Button_Selection_Inst
     tiling=NoTile
     imageSize=<32 32>
     margin=<ltrb 0 0 0 0>
     outlineSettings {
      bUseBrushTransparency=false
      roundingType=HalfHeightRadius
      width=0
      color {colorUseRule=UseColor_Specified, specifiedColor=<rgba 0 0 0 0>}
      cornerRadii=<0 0 0 0>
     }
     tintColor {colorUseRule=UseColor_Specified, specifiedColor=<rgba 1 1 1 1>}
     uVRegion {bIsValid=false, max=<0 0>, min=<0 0>}
    }
    padding=<ltrb 0 0 0 0>
   }
   slot {class=OverlaySlot, properties {horizontalAlignment=HAlign_Fill, verticalAlignment=VAlign_Fill}}
  }
  HorizontalBox_2 {
   widgetClassPath=&HorizontalBox
   parent=Border_Background
   widget=HorizontalBox_2
   properties {flowDirectionPreference=Culture}
   slot {class=BorderSlot, properties {padding=<ltrb 0 0 0 0>}}
  }
  SizeBox_2 {
   widgetClassPath=&SizeBox
   parent=HorizontalBox_2
   widget=SizeBox_2
   properties {
    bOverride_MinDesiredHeight=true
    bOverride_MinDesiredWidth=true
    heightOverride=68
    minDesiredHeight=68
    minDesiredWidth=68
    widthOverride=68
   }
   slot {class=HorizontalBoxSlot, properties {horizontalAlignment=HAlign_Center, verticalAlignment=VAlign_Center, padding=<ltrb 10 10 10 10>}}
  }
  ScaleBox_0 {widgetClassPath=&ScaleBox, parent=SizeBox_2, widget=ScaleBox_0, properties {stretch=ScaleToFitY}, slot {class=SizeBoxSlot}}
  Overlay_2 {widgetClassPath=&Overlay, parent=ScaleBox_0, widget=Overlay_2, slot {class=ScaleBoxSlot}}
  Avatar_MemberIcon {
   widgetClassPath=/Script/FortniteUI.FortSocialAvatarIcon
   bInherited=true
   bIsVariable=true
   parent=Overlay_2
   widget=Avatar_MemberIcon
   properties {
    avatarMaterial=/Game/UI/SocialPanel/Art/Materials/UI_UserPortrait_Buffet
    materialTextureParamName=SmallPreview
    visibility=Collapsed
    brush {
     bIsDynamicallyLoaded=false
     drawAs=Image
     imageType=NoImage
     mirroring=NoMirror
     resourceName=None
     resourceObject=&UI_UserPortrait_Buffet
     tiling=NoTile
     imageSize=<32 32>
     margin=<ltrb 0 0 0 0>
     outlineSettings {
      bUseBrushTransparency=false
      roundingType=HalfHeightRadius
      width=0
      color {colorUseRule=UseColor_Specified, specifiedColor=<rgba 0 0 0 0>}
      cornerRadii=<0 0 0 0>
     }
     tintColor {colorUseRule=UseColor_Specified, specifiedColor=<rgba 1 1 1 1>}
     uVRegion {bIsValid=false, max=<0 0>, min=<0 0>}
    }
    loadingBackgroundBrush {
     bIsDynamicallyLoaded=false
     drawAs=Image
     imageType=NoImage
     mirroring=NoMirror
     resourceName=None
     resourceObject=/FortUILibrary/UIKit/Throbber/MI_UIKit_Monochrome_Throbber
     tiling=NoTile
     imageSize=<32 32>
     margin=<ltrb 0 0 0 0>
     outlineSettings {
      bUseBrushTransparency=false
      roundingType=HalfHeightRadius
      width=0
      color {colorUseRule=UseColor_Specified, specifiedColor=<rgba 0 0 0 0>}
      cornerRadii=<0 0 0 0>
     }
     tintColor {colorUseRule=UseColor_Specified, specifiedColor=<rgba 1 1 1 1>}
     uVRegion {bIsValid=false, max=<0 0>, min=<0 0>}
    }
   }
   slot {class=OverlaySlot}
  }
  PlayerAvatar {
   widgetClassPath=/Game/UI/Foundation/Profile/WBP_PlayerAvatar
   bInherited=true
   bIsVariable=true
   parent=Overlay_2
   widget=PlayerAvatar
   slot {class=OverlaySlot}
  }
  HBox_SocialName {
   widgetClassPath=&HorizontalBox
   parent=HorizontalBox_2
   widget=HBox_SocialName
   slot {class=HorizontalBoxSlot, properties {verticalAlignment=VAlign_Center, padding=<ltrb 10 10 10 10>, size {sizeRule=Fill, value=1}}}
  }
  WBP_RebootRallyIcon {
   widgetClassPath=/Game/UI/Sidebar/RebootRally/WBP_RebootRallyIcon
   bInherited=true
   bIsVariable=true
   parent=HBox_SocialName
   widget=WBP_RebootRallyIcon
   properties {visibility=Collapsed}
   slot {class=HorizontalBoxSlot, properties {horizontalAlignment=HAlign_Left, verticalAlignment=VAlign_Center, padding=<ltrb -3 -3 0 -3>}}
  }
  UserTextInfoVB {
   widgetClassPath=&VerticalBox
   parent=HBox_SocialName
   widget=UserTextInfoVB
   properties {flowDirectionPreference=Culture}
   slot {class=HorizontalBoxSlot, properties {verticalAlignment=VAlign_Center, size {sizeRule=Fill, value=1}}}
  }
  Text_SocialName {
   widgetClassPath=/Script/FortniteUI.FortSocialNameTextBlock
   bInherited=true
   bIsVariable=true
   parent=UserTextInfoVB
   widget=Text_SocialName
   properties {
    bOverrideDefaultStyle=true
    bTintInlineIcon=true
    mobileTextBlockScale=1.2
    platformIconDisplayRule=AlwaysIfDifferent
    scrollStyle=&TextScrollStyle-Base_C
    secondNameFontSize=14
    text="Nickname <PC color=\"82BDBDFF\" size=\"12\"></><text color=\"82BDBDFF\" size=\"12\">Display Name</>"
    defaultTextStyleOverride {
     overflowPolicy=Clip
     transformPolicy=None
     colorAndOpacity {colorUseRule=UseColor_Specified, specifiedColor=<rgba 1 1 1 1>}
     font {
      bForceMonospaced=false
      bMaterialIsStencil=false
      fontMaterial=null
      fontObject=&BurbankSmall
      letterSpacing=0
      monospacedWidth=1
      size=16
      skewAmount=0
      typefaceFontName=Black
      outlineSettings {
       bApplyOutlineToDropShadows=false
       bMiteredCorners=false
       bSeparateFillAlpha=false
       outlineMaterial=null
       outlineSize=0
       outlineColor=<rgba 0 0 0 1>
      }
     }
     highlightColor {colorUseRule=UseColor_Specified, specifiedColor=<rgba 0 0 0 1>}
     highlightShape {
      bIsDynamicallyLoaded=false
      drawAs=Image
      imageType=NoImage
      mirroring=NoMirror
      resourceName=None
      resourceObject=null
      tiling=NoTile
      imageSize=<32 32>
      margin=<ltrb 0 0 0 0>
      outlineSettings {
       bUseBrushTransparency=false
       roundingType=HalfHeightRadius
       width=0
       color {colorUseRule=UseColor_Specified, specifiedColor=<rgba 0 0 0 0>}
       cornerRadii=<0 0 0 0>
      }
      tintColor {colorUseRule=UseColor_Specified, specifiedColor=<rgba 1 1 1 1>}
      uVRegion {bIsValid=false, max=<0 0>, min=<0 0>}
     }
     selectedBackgroundColor {colorUseRule=UseColor_Foreground, specifiedColor=<rgba 1 0 1 1>}
     shadowColorAndOpacity=<rgba 0 0 0 1>
     shadowOffset=<0 0>
     strikeBrush {
      bIsDynamicallyLoaded=false
      drawAs=Image
      imageType=NoImage
      mirroring=NoMirror
      resourceName=None
      resourceObject=null
      tiling=NoTile
      imageSize=<32 32>
      margin=<ltrb 0 0 0 0>
      outlineSettings {
       bUseBrushTransparency=false
       roundingType=HalfHeightRadius
       width=0
       color {colorUseRule=UseColor_Specified, specifiedColor=<rgba 0 0 0 0>}
       cornerRadii=<0 0 0 0>
      }
      tintColor {colorUseRule=UseColor_Specified, specifiedColor=<rgba 1 1 1 1>}
      uVRegion {bIsValid=false, max=<0 0>, min=<0 0>}
     }
     underlineBrush {
      bIsDynamicallyLoaded=false
      drawAs=Image
      imageType=NoImage
      mirroring=NoMirror
      resourceName=None
      resourceObject=null
      tiling=NoTile
      imageSize=<32 32>
      margin=<ltrb 0 0 0 0>
      outlineSettings {
       bUseBrushTransparency=false
       roundingType=HalfHeightRadius
       width=0
       color {colorUseRule=UseColor_Specified, specifiedColor=<rgba 0 0 0 0>}
       cornerRadii=<0 0 0 0>
      }
      tintColor {colorUseRule=UseColor_Specified, specifiedColor=<rgba 1 1 1 1>}
      uVRegion {bIsValid=false, max=<0 0>, min=<0 0>}
     }
    }
   }
   slot {class=VerticalBoxSlot, properties {verticalAlignment=VAlign_Center}}
  }
  Text_DisplayName {
   widgetClassPath=&CommonRichTextBlock
   bIsVariable=true
   parent=UserTextInfoVB
   widget=Text_DisplayName
   properties {
    bOverrideDefaultStyle=true
    bTintInlineIcon=true
    mobileTextBlockScale=1.2
    scrollStyle=&TextScrollStyle-Base_C
    text="</><text color=\"82BDBDFF\" size=\"12\">User Name<PC/>"
    visibility=Collapsed
    defaultTextStyleOverride {
     overflowPolicy=Clip
     transformPolicy=None
     colorAndOpacity {colorUseRule=UseColor_Specified, specifiedColor=<rgba 0.4452 1 1 1>}
     font {
      bForceMonospaced=false
      bMaterialIsStencil=false
      fontMaterial=null
      fontObject=&BurbankSmall
      letterSpacing=0
      monospacedWidth=1
      size=16
      skewAmount=0
      typefaceFontName=Black
      outlineSettings {
       bApplyOutlineToDropShadows=false
       bMiteredCorners=false
       bSeparateFillAlpha=false
       outlineMaterial=null
       outlineSize=0
       outlineColor=<rgba 0 0 0 1>
      }
     }
     highlightColor {colorUseRule=UseColor_Specified, specifiedColor=<rgba 0 0 0 1>}
     highlightShape {
      bIsDynamicallyLoaded=false
      drawAs=Image
      imageType=NoImage
      mirroring=NoMirror
      resourceName=None
      resourceObject=null
      tiling=NoTile
      imageSize=<32 32>
      margin=<ltrb 0 0 0 0>
      outlineSettings {
       bUseBrushTransparency=false
       roundingType=HalfHeightRadius
       width=0
       color {colorUseRule=UseColor_Specified, specifiedColor=<rgba 0 0 0 0>}
       cornerRadii=<0 0 0 0>
      }
      tintColor {colorUseRule=UseColor_Specified, specifiedColor=<rgba 1 1 1 1>}
      uVRegion {bIsValid=false, max=<0 0>, min=<0 0>}
     }
     selectedBackgroundColor {colorUseRule=UseColor_Foreground, specifiedColor=<rgba 1 0 1 1>}
     shadowColorAndOpacity=<rgba 0 0 0 1>
     shadowOffset=<0 0>
     strikeBrush {
      bIsDynamicallyLoaded=false
      drawAs=Image
      imageType=NoImage
      mirroring=NoMirror
      resourceName=None
      resourceObject=null
      tiling=NoTile
      imageSize=<32 32>
      margin=<ltrb 0 0 0 0>
      outlineSettings {
       bUseBrushTransparency=false
       roundingType=HalfHeightRadius
       width=0
       color {colorUseRule=UseColor_Specified, specifiedColor=<rgba 0 0 0 0>}
       cornerRadii=<0 0 0 0>
      }
      tintColor {colorUseRule=UseColor_Specified, specifiedColor=<rgba 1 1 1 1>}
      uVRegion {bIsValid=false, max=<0 0>, min=<0 0>}
     }
     underlineBrush {
      bIsDynamicallyLoaded=false
      drawAs=Image
      imageType=NoImage
      mirroring=NoMirror
      resourceName=None
      resourceObject=null
      tiling=NoTile
      imageSize=<32 32>
      margin=<ltrb 0 0 0 0>
      outlineSettings {
       bUseBrushTransparency=false
       roundingType=HalfHeightRadius
       width=0
       color {colorUseRule=UseColor_Specified, specifiedColor=<rgba 0 0 0 0>}
       cornerRadii=<0 0 0 0>
      }
      tintColor {colorUseRule=UseColor_Specified, specifiedColor=<rgba 1 1 1 1>}
      uVRegion {bIsValid=false, max=<0 0>, min=<0 0>}
     }
    }
   }
   slot {class=VerticalBoxSlot}
  }
  EmptyButtonToEatOneMouseUpEvent {
   widgetClassPath=&IconTextButton
   bIsVariable=true
   parent=Overlay_1
   widget=EmptyButtonToEatOneMouseUpEvent
   properties {
    buttonText=""
    onButtonBaseClicked=(&self:WidgetTree.EmptyButtonToEatOneMouseUpEvent.HandleOnButtonClicked)
    onButtonBaseDoubleClicked=(&self:WidgetTree.EmptyButtonToEatOneMouseUpEvent.HandleOnButtonDoubleClicked)
    onButtonBaseHovered=(&self:WidgetTree.EmptyButtonToEatOneMouseUpEvent.HandleOnButtonHovered)
    onButtonBaseUnhovered=(&self:WidgetTree.EmptyButtonToEatOneMouseUpEvent.HandleOnButtonUnhovered)
    onSelectedChangedBase=(&self:WidgetTree.EmptyButtonToEatOneMouseUpEvent.HandleOnSelectedChanged)
    style=/FortUILibrary/Foundation/Buttons/ButtonStyles/Athena/ButtonStyle_TransparentList_Social.ButtonStyle_TransparentList_Social_C
    visibility=Collapsed
    hoveredSlateSoundOverride {resourceObject=null}
    pressedSlateSoundOverride {resourceObject=null}
   }
   slot {class=OverlaySlot, properties {horizontalAlignment=HAlign_Fill, verticalAlignment=VAlign_Fill}}
  }
  Image_EncourageEpicFriend {widgetClassPath=&Image, bInherited=true, widget=null}
  Text_LastInteraction {widgetClassPath=&CommonRichTextBlock, bInherited=true, widget=null}
  Text_LastInteraction_Hover {widgetClassPath=&CommonRichTextBlock, bInherited=true, widget=null}
  Text_SocialName_Hover {widgetClassPath=/Script/SocialUMG.SocialNameTextBlock, bInherited=true, widget=null}
  Text_RichPresence {widgetClassPath=&CommonRichTextBlock, bInherited=true, widget=null}
  Text_RichPresence_Hover {widgetClassPath=&CommonRichTextBlock, bInherited=true, widget=null}
  InputActionWidget {widgetClassPath=&CommonActionWidget, bInherited=true, widget=null}
 ]
}
