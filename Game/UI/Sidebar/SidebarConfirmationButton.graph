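# Sidebar confirmation button widget built on CommonButtonLegacy, showing a centered text label and a right-aligned input action icon on a colored border.
# PreConstruct sets the label from DisplayText, and icon updates set the icon material to dark with no outline.
# Hover plays OnHover, turning the border from blue to white, darkening and shifting the text, and revealing the icon; unhover reverses it at double speed.
# The UserActionMenu anchor and bIsAccept are unused by the graph.
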WidgetBlueprint /Game/UI/Sidebar/SidebarConfirmationButton
animations [OnHover {
  binding_count=4
  display_label=""
  length_seconds=0.2
  object=&self:OnHover
  sequence {bindings [
    InputActionWidget {
     id=#1
     possessed_class=CommonActionWidget
     tracks [{
       type=MovieSceneFloatTrack
       display_name=Render Opacity
       sections [{type=MovieSceneFloatSection, end_frame=4, start_frame=0, channels [None {type=MovieSceneScriptingFloatChannel, keys [[0,0],[2,1]]}]}]
      }]
    }
    OverlaySlot_0 {
     id=#2
     parent_id=#1
     possessed_class=OverlaySlot
     tracks [{
       type=MovieSceneMarginTrack
       display_name=Padding
       sections [{
         type=MovieSceneMarginSection
         end_frame=2
         start_frame=0
         channels [
          Left {type=MovieSceneScriptingFloatChannel}
          Top {type=MovieSceneScriptingFloatChannel}
          Right {type=MovieSceneScriptingFloatChannel, keys [[0,-32],[2,20]]}
          Bottom {type=MovieSceneScriptingFloatChannel}
         ]
        }]
      }]
    }
    EntryBorder {
     id=#3
     possessed_class=Border
     tracks [{
       type=MovieSceneColorTrack
       display_name=Brush Color
       sections [{
         type=MovieSceneColorSection
         end_frame=2
         start_frame=0
         channels [
          Color.R {type=MovieSceneScriptingFloatChannel, keys [[0,0.0021],[2,1]]}
          Color.G {type=MovieSceneScriptingFloatChannel, keys [[0,0.0612],[2,1]]}
          Color.B {type=MovieSceneScriptingFloatChannel, keys [[0,0.5029],[2,1]]}
          Color.A {type=MovieSceneScriptingFloatChannel, const [1,2]}
         ]
        }]
      }]
    }
    Text_InteractionName {
     id=#4
     possessed_class=CommonTextBlock
     tracks [
      {
       type=MovieScene2DTransformTrack
       sections [{
         type=MovieScene2DTransformSection
         end_frame=4
         start_frame=0
         channels [
          Translation.X {type=MovieSceneScriptingFloatChannel, keys [[0,0],[2,-8]]}
          Translation.Y {type=MovieSceneScriptingFloatChannel}
          Angle {type=MovieSceneScriptingFloatChannel}
          Scale.X {type=MovieSceneScriptingFloatChannel}
          Scale.Y {type=MovieSceneScriptingFloatChannel}
          Shear.X {type=MovieSceneScriptingFloatChannel}
          Shear.Y {type=MovieSceneScriptingFloatChannel}
         ]
        }]
      }
      {
       type=MovieSceneColorTrack
       display_name=Color and Opacity
       sections [{
         type=MovieSceneColorSection
         end_frame=2
         start_frame=0
         channels [
          Color.R {type=MovieSceneScriptingFloatChannel, keys [[0,0.0331],[2,0.0513]]}
          Color.G {type=MovieSceneScriptingFloatChannel, keys [[0,0.402],[2,0.0513]]}
          Color.B {type=MovieSceneScriptingFloatChannel, keys [[0,1],[2,0.0513]]}
          Color.A {type=MovieSceneScriptingFloatChannel, const [1,2]}
         ]
        }]
      }
     ]
    }
   ]}
 }]
graphs [EventGraph {nodes [
   "On Input Icon Updated (InputActionWidget)"() {is=title K2Node_ComponentBoundEvent}
   > n1 = GetIconDynamicMaterial!(execute, self:Common Action Widget Object Reference=Get InputActionWidget() -> (InputActionWidget:Common Action Widget Object Reference)) -> (then, ReturnValue:Material Instance Dynamic Object Reference)
   > SetScalarParameterValue(self=@n1, ParameterName=Use Dark, Value=1)
   > SetScalarParameterValue(self=@n1, ParameterName=Outline Opacity, Value=0)
   Event PreConstruct()
   > "SetText (Text)"(self=Get Text_InteractionName() -> (Text_InteractionName:Common Text Object Reference), InText=Get DisplayText() -> (DisplayText:Text))
   Event On Hovered()
   > PlayAnimation(InAnimation=Get OnHover() -> (OnHover:Widget Animation Object Reference))
   Event On Unhovered()
   > PlayAnimation(InAnimation=Get OnHover(), PlayMode=Reverse, PlaybackSpeed=2)
  ]}]
variables [
 bIsAccept {type=boolean}
 DisplayText {type=string, default=""}
]
widgets {
 info {inheritedWidgetCount=1, parentClass=&CommonButtonLegacy, rootWidgetClass=&SizeBox, widgetCount=8}
 widgets [
  SizeBox_0 {widgetClassPath=&SizeBox, widget=SizeBox_0, properties {bOverride_HeightOverride=true, heightOverride=48}}
  EntryBorder {
   widgetClassPath=&Border
   parent=SizeBox_0
   widget=EntryBorder
   properties {visibility=SelfHitTestInvisible, brushColor=<rgba 0.0021 0.0612 0.5029 1>, padding=<ltrb 0 0 0 0>}
   slot {class=SizeBoxSlot, properties {padding=<ltrb -16 0 -16 0>}}
  }
  Overlay_0 {widgetClassPath=&Overlay, parent=EntryBorder, widget=Overlay_0, slot {class=BorderSlot, properties {padding=<ltrb 0 0 0 0>}}}
  SizeBox_3 {
   widgetClassPath=&SizeBox
   parent=Overlay_0
   widget=SizeBox_3
   properties {bOverride_MinDesiredHeight=true, heightOverride=24, minDesiredHeight=24}
   slot {class=OverlaySlot, properties {horizontalAlignment=HAlign_Center, verticalAlignment=VAlign_Center}}
  }
  Text_InteractionName {
   widgetClassPath=&CommonTextBlock
   bIsVariable=true
   parent=SizeBox_3
   widget=Text_InteractionName
   properties {
    mobileFontSizeMultiplier=1.2
    text=BUTTON
    colorAndOpacity {colorUseRule=UseColor_Specified, specifiedColor=<rgba 0.0331 0.402 1 1>}
    font {
     bForceMonospaced=false
     bMaterialIsStencil=false
     fontMaterial=null
     fontObject=&BurbankSmall
     letterSpacing=0
     monospacedWidth=1
     size=18
     skewAmount=0
     typefaceFontName=Bold
     outlineSettings {
      bApplyOutlineToDropShadows=false
      bMiteredCorners=false
      bSeparateFillAlpha=false
      outlineMaterial=null
      outlineSize=0
      outlineColor=<rgba 0 0 0 1>
     }
    }
   }
   slot {class=SizeBoxSlot, properties {horizontalAlignment=HAlign_Center, verticalAlignment=VAlign_Center}}
  }
  Overlay_1 {
   widgetClassPath=&Overlay
   parent=Overlay_0
   widget=Overlay_1
   slot {class=OverlaySlot, properties {horizontalAlignment=HAlign_Right, verticalAlignment=VAlign_Fill}}
  }
  InputActionWidget {
   widgetClassPath=&CommonActionWidget
   bInherited=true
   parent=Overlay_1
   widget=InputActionWidget
   properties {visibility=SelfHitTestInvisible, inputActions [{dataTable=&InputActionDataTable, rowName=Select}]}
   slot {class=OverlaySlot, properties {horizontalAlignment=HAlign_Right, verticalAlignment=VAlign_Center, padding=<ltrb 0 0 20 0>}}
  }
  MenuAnchor_Actions {
   widgetClassPath=/Script/UMG.MenuAnchor
   bIsVariable=true
   parent=Overlay_1
   widget=MenuAnchor_Actions
   properties {menuClass=/Game/UI/SocialPanel/UserActionMenu.UserActionMenu_C, placement=MenuPlacement_RightLeftCenter}
   slot {class=OverlaySlot, properties {horizontalAlignment=HAlign_Right, verticalAlignment=VAlign_Center}}
  }
 ]
}
