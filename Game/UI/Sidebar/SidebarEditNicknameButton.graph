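# Widget Blueprint on FortEditNicknameButton: the sidebar field for adding or editing a nickname.
# Nested borders hold the inherited editable text, character counter and Select input action glyph, which starts invisible.
# The event graph only plays the OnHovered animation forward on hover and in reverse on unhover.
# That animation brightens the borders, nudges the text and fades in the glyph while sliding its padding into place.
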WidgetBlueprint /Game/UI/Sidebar/SidebarEditNicknameButton
animations [OnHovered {
  binding_count=6
  length_seconds=0.1
  object=&self:OnHovered
  sequence {bindings [
    HorizontalBoxSlot_1 {
     id=#1
     parent_id=#2
     possessed_class=HorizontalBoxSlot
     tracks [{
       type=MovieSceneMarginTrack
       display_name=Padding
       sections [{
         type=MovieSceneMarginSection
         end_frame=2
         start_frame=0
         channels [
          Left {type=MovieSceneScriptingFloatChannel}
          Top {type=MovieSceneScriptingFloatChannel}
          Right {type=MovieSceneScriptingFloatChannel, keys [[0,-32],[2,0]]}
          Bottom {type=MovieSceneScriptingFloatChannel}
         ]
        }]
      }]
    }
    OverlaySlot_1 {
     id=#3
     parent_id=#4
     possessed_class=OverlaySlot
     tracks [{
       type=MovieSceneMarginTrack
       display_name=Padding
       sections [{
         type=MovieSceneMarginSection
         end_frame=2
         start_frame=0
         channels [
          Left {type=MovieSceneScriptingFloatChannel, keys [[0,0],[2,-5]]}
          Top {type=MovieSceneScriptingFloatChannel}
          Right {type=MovieSceneScriptingFloatChannel}
          Bottom {type=MovieSceneScriptingFloatChannel}
         ]
        }]
      }]
    }
    Border_1 {
     id=#5
     possessed_class=Border
     tracks [{
       type=MovieSceneColorTrack
       display_name=Brush Color
       sections [{
         type=MovieSceneColorSection
         end_frame=2
         start_frame=0
         channels [
          Color.R {type=MovieSceneScriptingFloatChannel, keys [[0,0.0075],[2,0.0065]]}
          Color.G {type=MovieSceneScriptingFloatChannel, keys [[0,0.0262],[2,0.0467]]}
          Color.B {type=MovieSceneScriptingFloatChannel, keys [[0,0.0976],[2,0.2502]]}
          Color.A {type=MovieSceneScriptingFloatChannel, const [1,2]}
         ]
        }]
      }]
    }
    EditableText_EditNickname {id=#4, possessed_class=FortMobileEditableText, tracks []}
    InputActionWidget {
     id=#2
     possessed_class=CommonActionWidget
     tracks [{
       type=MovieSceneFloatTrack
       display_name=Render Opacity
       sections [{type=MovieSceneFloatSection, end_frame=2, start_frame=0, channels [None {type=MovieSceneScriptingFloatChannel, keys [[0,0],[2,1]]}]}]
      }]
    }
    ClickCapture {
     id=#6
     possessed_class=Border
     tracks [{
       type=MovieSceneColorTrack
       display_name=Brush Color
       sections [{
         type=MovieSceneColorSection
         end_frame=2
         start_frame=0
         channels [
          Color.R {type=MovieSceneScriptingFloatChannel, keys [[0,0.0021],[2,1]]}
          Color.G {type=MovieSceneScriptingFloatChannel, keys [[0,0.0612],[2,1]]}
          Color.B {type=MovieSceneScriptingFloatChannel, keys [[0,0.5029],[2,1]]}
          Color.A {type=MovieSceneScriptingFloatChannel, const [1,2]}
         ]
        }]
      }]
    }
   ]}
 }]
graphs [EventGraph {nodes [
   Event On Hovered()
   n1 = Get OnHovered() -> (OnHovered:Widget Animation Object Reference)
   > PlayAnimation(InAnimation=@n1)
   Event On Unhovered()
   > PlayAnimation(InAnimation=@n1, PlayMode=Reverse)
  ]}]
variables []
widgets {
 info {inheritedWidgetCount=3, parentClass=/Script/FortniteUI.FortEditNicknameButton, rootWidgetClass=&SizeBox, widgetCount=10}
 widgets [
  SizeBox_24 {widgetClassPath=&SizeBox, widget=SizeBox_24, properties {bOverride_WidthOverride=true, widthOverride=520}}
  BackgroundBorder {
   widgetClassPath=&Border
   parent=SizeBox_24
   widget=BackgroundBorder
   properties {
    verticalAlignment=VAlign_Top
    background {
     bIsDynamicallyLoaded=false
     drawAs=Image
     imageType=NoImage
     mirroring=NoMirror
     resourceName=None
     resourceObject=/Game/UI/SocialPanel/Art/Materials/M_UI_SocialUserItemBorder_UserListItem
     tiling=NoTile
     imageSize=<32 32>
     margin=<ltrb 0 0 0 0>
     outlineSettings {
      bUseBrushTransparency=false
      roundingType=HalfHeightRadius
      width=0
      color {colorUseRule=UseColor_Specified, specifiedColor=<rgba 0 0 0 0>}
      cornerRadii=<0 0 0 0>
     }
     tintColor {colorUseRule=UseColor_Specified, specifiedColor=<rgba 1 1 1 1>}
     uVRegion {bIsValid=false, max=<0 0>, min=<0 0>}
    }
    padding=<ltrb 0 0 0 0>
   }
   slot {class=SizeBoxSlot}
  }
  ClickCapture {
   widgetClassPath=&Border
   bIsVariable=true
   parent=BackgroundBorder
   widget=ClickCapture
   properties {brushColor=<rgba 0.0021 0.0612 0.5029 1>, padding=<ltrb 4 4 4 4>}
   slot {class=BorderSlot, properties {verticalAlignment=VAlign_Top, padding=<ltrb 0 0 0 0>}}
  }
  Border_1 {
   widgetClassPath=&Border
   parent=ClickCapture
   widget=Border_1
   properties {
    bOverride_Cursor=true
    cursor=TextEditBeam
    verticalAlignment=VAlign_Center
    brushColor=<rgba 0.0075 0.0262 0.0976 1>
    padding=<ltrb 20 0 20 0>
   }
   slot {class=BorderSlot, properties {padding=<ltrb 4 4 4 4>}}
  }
  HorizontalBox_0 {
   widgetClassPath=&HorizontalBox
   parent=Border_1
   widget=HorizontalBox_0
   slot {class=BorderSlot, properties {verticalAlignment=VAlign_Center, padding=<ltrb 20 0 20 0>}}
  }
  SizeBox_1 {
   widgetClassPath=&SizeBox
   parent=HorizontalBox_0
   widget=SizeBox_1
   properties {bOverride_MinDesiredHeight=true, heightOverride=24, minDesiredHeight=62, minDesiredWidth=368, widthOverride=368}
   slot {class=HorizontalBoxSlot, properties {size {sizeRule=Fill, value=1}}}
  }
  Overlay_0 {
   widgetClassPath=&Overlay
   parent=SizeBox_1
   widget=Overlay_0
   slot {class=SizeBoxSlot, properties {verticalAlignment=VAlign_Center, padding=<ltrb 0 -6 0 -8>}}
  }
  EditableText_EditNickname {
   widgetClassPath=/Script/FortniteUI.FortMobileEditableText
   bInherited=true
   bIsVariable=true
   parent=Overlay_0
   widget=EditableText_EditNickname
   properties {
    clearKeyboardFocusOnCommit=false
    clipping=ClipToBoundsAlways
    hintText=ADD NICKNAME
    mobileFontSizeMultiplier=1.34
    text=Diedra
    virtualKeyboardDismissAction=TextCommitOnAccept
    virtualKeyboardTrigger=OnAllFocusEvents
    widgetStyle {
     backgroundImageComposing {
      bIsDynamicallyLoaded=false
      drawAs=Border
      imageType=FullColor
      mirroring=NoMirror
      resourceName=../../../Engine/Content/Slate/Old/HyperlinkDotted.png
      resourceObject=null
      tiling=Both
      imageSize=<0 0>
      margin=<ltrb 0 0 0 0.1875>
      outlineSettings {
       bUseBrushTransparency=false
       roundingType=HalfHeightRadius
       width=0
       color {colorUseRule=UseColor_Specified, specifiedColor=<rgba 0 0 0 0>}
       cornerRadii=<0 0 0 0>
      }
      tintColor {colorUseRule=UseColor_Specified, specifiedColor=<rgba 1 1 1 1>}
      uVRegion {bIsValid=false, max=<0 0>, min=<0 0>}
     }
     backgroundImageSelected {
      bIsDynamicallyLoaded=false
      drawAs=Box
      imageType=FullColor
      mirroring=NoMirror
      resourceName=../../../Engine/Content/Slate/Common/EditableTextSelectionBackground.png
      resourceObject=null
      tiling=NoTile
      imageSize=<0 0>
      margin=<ltrb 0.25 0.25 0.25 0.25>
      outlineSettings {
       bUseBrushTransparency=false
       roundingType=HalfHeightRadius
       width=0
       color {colorUseRule=UseColor_Specified, specifiedColor=<rgba 0 0 0 0>}
       cornerRadii=<0 0 0 0>
      }
      tintColor {colorUseRule=UseColor_Specified, specifiedColor=<rgba 1 1 1 1>}
      uVRegion {bIsValid=false, max=<0 0>, min=<0 0>}
     }
     caretImage {
      bIsDynamicallyLoaded=false
      drawAs=Image
      imageType=FullColor
      mirroring=NoMirror
      resourceName=../../../Engine/Content/Slate/Old/White.png
      resourceObject=null
      tiling=NoTile
      imageSize=<16 16>
      margin=<ltrb 0 0 0 0>
      outlineSettings {
       bUseBrushTransparency=false
       roundingType=HalfHeightRadius
       width=0
       color {colorUseRule=UseColor_Specified, specifiedColor=<rgba 0 0 0 0>}
       cornerRadii=<0 0 0 0>
      }
      tintColor {colorUseRule=UseColor_Specified, specifiedColor=<rgba 1 1 1 1>}
      uVRegion {bIsValid=false, max=<0 0>, min=<0 0>}
     }
     colorAndOpacity {colorUseRule=UseColor_Specified, specifiedColor=<rgba 1 1 1 1>}
     font {
      bForceMonospaced=false
      bMaterialIsStencil=false
      fontMaterial=null
      fontObject=&BurbankSmall
      letterSpacing=0
      monospacedWidth=1
      size=15
      skewAmount=0
      typefaceFontName=Bold
      outlineSettings {
       bApplyOutlineToDropShadows=false
       bMiteredCorners=false
       bSeparateFillAlpha=false
       outlineMaterial=null
       outlineSize=0
       outlineColor=<rgba 0 0 0 1>
      }
     }
     hintColorAndOpacity {colorUseRule=UseColor_Foreground, specifiedColor=<rgba 1 0 1 1>}
     selectedBackgroundColor {colorUseRule=UseColor_Foreground, specifiedColor=<rgba 1 0 1 1>}
    }
   }
   slot {class=OverlaySlot, properties {horizontalAlignment=HAlign_Fill, verticalAlignment=VAlign_Fill}}
  }
  RichText_CharacterCounter {
   widgetClassPath=&CommonRichTextBlock
   bInherited=true
   bIsVariable=true
   parent=HorizontalBox_0
   widget=RichText_CharacterCounter
   properties {
    bOverrideDefaultStyle=true
    mobileTextBlockScale=1.2
    text=0/16
    defaultTextStyleOverride {
     overflowPolicy=Clip
     transformPolicy=None
     colorAndOpacity {colorUseRule=UseColor_Specified, specifiedColor=<rgba 0.6562 0.6562 0.6562 0.8>}
     font {
      bForceMonospaced=false
      bMaterialIsStencil=false
      fontMaterial=null
      fontObject=&BurbankSmall
      letterSpacing=0
      monospacedWidth=1
      size=12
      skewAmount=0
      typefaceFontName=Bold
      outlineSettings {
       bApplyOutlineToDropShadows=false
       bMiteredCorners=false
       bSeparateFillAlpha=false
       outlineMaterial=null
       outlineSize=0
       outlineColor=<rgba 0 0 0 1>
      }
     }
     highlightColor {colorUseRule=UseColor_Specified, specifiedColor=<rgba 0 0 0 1>}
     highlightShape {
      bIsDynamicallyLoaded=false
      drawAs=Image
      imageType=NoImage
      mirroring=NoMirror
      resourceName=None
      resourceObject=null
      tiling=NoTile
      imageSize=<32 32>
      margin=<ltrb 0 0 0 0>
      outlineSettings {
       bUseBrushTransparency=false
       roundingType=HalfHeightRadius
       width=0
       color {colorUseRule=UseColor_Specified, specifiedColor=<rgba 0 0 0 0>}
       cornerRadii=<0 0 0 0>
      }
      tintColor {colorUseRule=UseColor_Specified, specifiedColor=<rgba 1 1 1 1>}
      uVRegion {bIsValid=false, max=<0 0>, min=<0 0>}
     }
     selectedBackgroundColor {colorUseRule=UseColor_Foreground, specifiedColor=<rgba 1 0 1 1>}
     shadowColorAndOpacity=<rgba 0 0 0 1>
     shadowOffset=<0 0>
     strikeBrush {
      bIsDynamicallyLoaded=false
      drawAs=Image
      imageType=NoImage
      mirroring=NoMirror
      resourceName=None
      resourceObject=null
      tiling=NoTile
      imageSize=<32 32>
      margin=<ltrb 0 0 0 0>
      outlineSettings {
       bUseBrushTransparency=false
       roundingType=HalfHeightRadius
       width=0
       color {colorUseRule=UseColor_Specified, specifiedColor=<rgba 0 0 0 0>}
       cornerRadii=<0 0 0 0>
      }
      tintColor {colorUseRule=UseColor_Specified, specifiedColor=<rgba 1 1 1 1>}
      uVRegion {bIsValid=false, max=<0 0>, min=<0 0>}
     }
     underlineBrush {
      bIsDynamicallyLoaded=false
      drawAs=Image
      imageType=NoImage
      mirroring=NoMirror
      resourceName=None
      resourceObject=null
      tiling=NoTile
      imageSize=<32 32>
      margin=<ltrb 0 0 0 0>
      outlineSettings {
       bUseBrushTransparency=false
       roundingType=HalfHeightRadius
       width=0
       color {colorUseRule=UseColor_Specified, specifiedColor=<rgba 0 0 0 0>}
       cornerRadii=<0 0 0 0>
      }
      tintColor {colorUseRule=UseColor_Specified, specifiedColor=<rgba 1 1 1 1>}
      uVRegion {bIsValid=false, max=<0 0>, min=<0 0>}
     }
    }
   }
   slot {class=HorizontalBoxSlot, properties {horizontalAlignment=HAlign_Right, verticalAlignment=VAlign_Center, padding=<ltrb 20 0 0 0>}}
  }
  InputActionWidget {
   widgetClassPath=&CommonActionWidget
   bInherited=true
   parent=HorizontalBox_0
   widget=InputActionWidget
   properties {renderOpacity=0, visibility=SelfHitTestInvisible, inputActions [{dataTable=&InputActionDataTable, rowName=Select}]}
   slot {class=HorizontalBoxSlot, properties {horizontalAlignment=HAlign_Right, verticalAlignment=VAlign_Center, padding=<ltrb 20 0 -32 0>}}
  }
 ]
}
