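# Social interaction button for the sidebar, built on FortSidebarSocialInteractionButton, showing an interaction name, input icon and optional directional arrow, taller on mobile.
# When the interaction is set, it orients the arrow, shows an Epic-friend badge for platform-only friend add/accept actions, and shows a Reboot Rally icon for RallyEmBack.
# Hover plays a color/opacity animation (reversed on unhover), recolors the badge, scrolls text and flags dark input-icon styling, applied through material parameters on icon updates.
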WidgetBlueprint /Game/UI/Sidebar/SidebarInteraction_Profile
animations [OnHover {
  binding_count=4
  display_label=""
  length_seconds=0.2
  object=&self:OnHover
  sequence {bindings [
    InputActionWidget {
     id=#1
     possessed_class=CommonActionWidget
     tracks [{
       type=MovieSceneFloatTrack
       display_name=Render Opacity
       sections [{type=MovieSceneFloatSection, end_frame=4, start_frame=0, channels [None {type=MovieSceneScriptingFloatChannel, keys [[0,0],[1,0],[2,1]]}]}]
      }]
    }
    Text_InteractionName {
     id=#2
     possessed_class=CommonTextBlock
     tracks [{
       type=MovieSceneColorTrack
       display_name=Color and Opacity
       sections [{
         type=MovieSceneColorSection
         end_frame=2
         start_frame=0
         channels [
          Color.R {type=MovieSceneScriptingFloatChannel, keys [[0,0.0331],[2,0.0513]]}
          Color.G {type=MovieSceneScriptingFloatChannel, keys [[0,0.402],[2,0.0513]]}
          Color.B {type=MovieSceneScriptingFloatChannel, keys [[0,1],[2,0.0513]]}
          Color.A {type=MovieSceneScriptingFloatChannel, const [1,2]}
         ]
        }]
      }]
    }
    Image_Arrow {
     id=#3
     possessed_class=Image
     tracks [{
       type=MovieSceneColorTrack
       display_name=Color and Opacity
       sections [{
         type=MovieSceneColorSection
         end_frame=2
         start_frame=0
         channels [
          Color.R {type=MovieSceneScriptingFloatChannel, keys [[0,0.0331],[2,0]]}
          Color.G {type=MovieSceneScriptingFloatChannel, keys [[0,0.402],[2,0]]}
          Color.B {type=MovieSceneScriptingFloatChannel, keys [[0,1],[2,0]]}
          Color.A {type=MovieSceneScriptingFloatChannel, const [1,2]}
         ]
        }]
      }]
    }
    EntryBorder {
     id=#4
     possessed_class=Border
     tracks [{
       type=MovieSceneColorTrack
       display_name=Brush Color
       sections [{
         type=MovieSceneColorSection
         end_frame=2
         start_frame=0
         channels [
          Color.R {type=MovieSceneScriptingFloatChannel, keys [[0,0.0021],[2,1]]}
          Color.G {type=MovieSceneScriptingFloatChannel, keys [[0,0.0612],[2,1]]}
          Color.B {type=MovieSceneScriptingFloatChannel, keys [[0,0.5029],[2,1]]}
          Color.A {type=MovieSceneScriptingFloatChannel, const [1,2]}
         ]
        }]
      }]
    }
   ]}
 }]
graphs [
 EventGraph {nodes [
   "On Input Icon Updated (InputActionWidget)"() {is=title K2Node_ComponentBoundEvent}
   > n1 = GetIconDynamicMaterial!(execute, self:Common Action Widget Object Reference=Get InputActionWidget() -> (InputActionWidget:Common Action Widget Object Reference)) -> (then, ReturnValue:Material Instance Dynamic Object Reference)
   > SetScalarParameterValue(
    self=@n1
    ParameterName=Use Dark
    Value=Select(Option 0:float=0, Option 1:float=1, Index=Get UseDarkInputIcon() -> (UseDarkInputIcon:bool)) -> (ReturnValue:float)
   )
   > SetScalarParameterValue(self=@n1, ParameterName=Outline Opacity, Value=Select(Option 0=0.5, Option 1=0, Index=Get UseDarkInputIcon()))
   Event OnInteractionSet() {is=title K2Node_Event}
   > SetVisibility(self=Get EncourageEpicFriend() -> (EncourageEpicFriend:Image Object Reference), InVisibility=Collapsed)
   > SetArrow()
   n2 = GetInteractionName!(self:Social Interaction Button Object Reference) -> (ReturnValue:Name=None)
   n3 = "Equal (Name)"!(A:Name=@n2, B:Name=AddFriend) -> (ReturnValue:bool) {is=title K2Node_CallFunction}
   > n4 = Branch(Condition=@n3)
   > n5 = SetBrushColor(
    self=Get Border_InteractionIndicator() -> (Border_InteractionIndicator:Border Object Reference)
    InBrushColor=Get EncourageEpicFriend_IdleColor() -> (EncourageEpicFriend_IdleColor:Linear Color Structure)
   )
   n6 = Branch(execute=[@n5.then, @n4.else], Condition=OR Boolean(A=@n3, B="Equal (Name)"(A=@n2, B=AcceptFriendInvite)))
   > n7 = SetVisibility(
    self=Get EncourageEpicFriend()
    InVisibility=Select(
     Option 0:ESlateVisibility Enum=Collapsed
     Option 1:ESlateVisibility Enum=SelfHitTestInvisible
     Index=IsPlatformOnlyFriend!(self:Social Interaction Button Object Reference) -> (ReturnValue:bool)
    ) -> (ReturnValue:ESlateVisibility Enum)
   )
   SetVisibility(
    execute=[@n7.then, @n6.else]
    self=Get WBP_RebootRallyIcon() -> (WBP_RebootRallyIcon:WBP Reboot Rally Icon Object Reference)
    InVisibility=Select(Option 0=Collapsed, Option 1=SelfHitTestInvisible, Index="Equal (Name)"(A=@n2, B=RallyEmBack) {is=title K2Node_PromotableOperator})
   )
   Event OnInitialized()
   > SetMinDesiredHeight!(
    execute
    self:Size Box Object Reference=Get SizeBox_0() -> (SizeBox_0:Size Box Object Reference)
    InMinDesiredHeight:float=Select(Option 0:double=48, Option 1:double=80, Index=IsMobileGame()) -> (ReturnValue:double)
   ) -> (then)
   Event On Hovered()
   > PlayAnimation(InAnimation=Get OnHover() -> (OnHover:Widget Animation Object Reference))
   > SetColorAndOpacity(
    self:Image Object Reference=Get EncourageEpicFriend()
    InColorAndOpacity:Linear Color Structure=Get EncourageEpicFriend_HoveredColor() -> (EncourageEpicFriend_HoveredColor:Linear Color Structure)
   )
   > SetScrollingEnabled!(
    execute
    self:Common Text Object Reference=Get Text_InteractionName() -> (Text_InteractionName:Common Text Object Reference)
    bInIsScrollingEnabled:bool=true
   ) -> (then)
   > Set UseDarkInputIcon(UseDarkInputIcon:bool=true) -> (Output_Get:bool)
   Event On Unhovered()
   > PlayAnimation(InAnimation=Get OnHover(), PlayMode=Reverse, PlaybackSpeed=2)
   > SetColorAndOpacity(self=Get EncourageEpicFriend(), InColorAndOpacity=Get EncourageEpicFriend_IdleColor())
   > ShowAlternateInputAction!(execute, self:Fort Sidebar Social Interaction Button Object Reference, bShow:bool=true) -> (then)
   > SetScrollingEnabled(self=Get Text_InteractionName(), bInIsScrollingEnabled=false)
   > Set UseDarkInputIcon(UseDarkInputIcon=false)
  ]}
 SetArrow {nodes [
   SetArrow() {is=title K2Node_FunctionEntry}
   n1 = Get ArrowDirection() -> (ArrowDirection:EFortSidebarSocialInteractionArrow Enum=None)
   n2 = Get Image_Arrow() -> (Image_Arrow:Image Object Reference)
   > SetRenderTransformAngle!(
    execute
    self:Widget Object Reference=@n2
    Angle:float=Select!(None:double=0, Up:double=270, Down:double=90, Left:double=180, Right:double=0, Index:EFortSidebarSocialInteractionArrow Enum=@n1) -> (ReturnValue)
   ) -> (then)
   > SetVisibility(
    self=@n2
    InVisibility=Select!(
     Option 0:ESlateVisibility Enum=HitTestInvisible
     Option 1:ESlateVisibility Enum=Collapsed
     Index:bool="Equal (Enum)"(A:EFortSidebarSocialInteractionArrow Enum=@n1, B:EFortSidebarSocialInteractionArrow Enum=None)
    ) -> (ReturnValue:ESlateVisibility Enum)
   )
  ]}
]
variables [
 AddFriendColor {type=LinearColor, default=<rgba 0 0 0 0>}
 EncourageEpicFriend_HoveredColor {type=LinearColor, default=<rgba 1 1 1 1>}
 EncourageEpicFriend_IdleColor {type=LinearColor, default=<rgba 0.9558 1 0.0672 1>}
 UseDarkInputIcon {type=boolean}
]
widgets {
 info {
  inheritedWidgetCount=4
  parentClass=/Script/FortniteUI.FortSidebarSocialInteractionButton
  rootWidgetClass=&SizeBox
  widgetCount=19
 }
 widgets [
  SizeBox_0 {
   widgetClassPath=&SizeBox
   bIsVariable=true
   widget=SizeBox_0
   properties {bOverride_MinDesiredHeight=true, heightOverride=48, minDesiredHeight=48}
  }
  EntryBorder {
   widgetClassPath=&Border
   parent=SizeBox_0
   widget=EntryBorder
   properties {
    visibility=SelfHitTestInvisible
    background {
     bIsDynamicallyLoaded=false
     drawAs=Box
     imageType=NoImage
     mirroring=NoMirror
     resourceName=None
     resourceObject=null
     tiling=NoTile
     imageSize=<32 32>
     margin=<ltrb 0.5 0.5 0.5 0.5>
     outlineSettings {
      bUseBrushTransparency=false
      roundingType=HalfHeightRadius
      width=0
      color {colorUseRule=UseColor_Specified, specifiedColor=<rgba 0 0 0 0>}
      cornerRadii=<0 0 0 0>
     }
     tintColor {colorUseRule=UseColor_Specified, specifiedColor=<rgba 1 1 1 1>}
     uVRegion {bIsValid=false, max=<0 0>, min=<0 0>}
    }
    brushColor=<rgba 0.0021 0.0612 0.5029 1>
    padding=<ltrb 0 0 0 0>
   }
   slot {class=SizeBoxSlot}
  }
  Overlay_0 {widgetClassPath=&Overlay, parent=EntryBorder, widget=Overlay_0, slot {class=BorderSlot, properties {padding=<ltrb 0 0 0 0>}}}
  Border_InteractionIndicator {
   widgetClassPath=&Border
   bInherited=true
   bIsVariable=true
   parent=Overlay_0
   widget=Border_InteractionIndicator
   properties {renderOpacity=0.5, visibility=Collapsed, brushColor=<rgba 0.4287 0.6654 0.7991 1>, padding=<ltrb 0 0 0 0>}
   slot {class=OverlaySlot}
  }
  Spacer_0 {
   widgetClassPath=&Spacer
   parent=Border_InteractionIndicator
   widget=Spacer_0
   properties {size=<8 8>}
   slot {class=BorderSlot, properties {padding=<ltrb 0 0 0 0>}}
  }
  GridPanel_0 {
   widgetClassPath=&GridPanel
   parent=Overlay_0
   widget=GridPanel_0
   properties {columnFill [0,1,0]}
   slot {class=OverlaySlot, properties {horizontalAlignment=HAlign_Center, verticalAlignment=VAlign_Center, padding=<ltrb 30 0 30 0>}}
  }
  WBP_RebootRallyIcon {
   widgetClassPath=/Game/UI/Sidebar/RebootRally/WBP_RebootRallyIcon
   bIsVariable=true
   parent=GridPanel_0
   widget=WBP_RebootRallyIcon
   properties {visibility=Collapsed}
   slot {class=GridSlot, properties {horizontalAlignment=HAlign_Right, verticalAlignment=VAlign_Center, nudge=<-10 0>, padding=<ltrb 0 0 -32 0>}}
  }
  Text_InteractionName {
   widgetClassPath=&CommonTextBlock
   bInherited=true
   bIsVariable=true
   parent=GridPanel_0
   widget=Text_InteractionName
   properties {
    clipping=ClipToBounds
    mobileFontSizeMultiplier=1.25
    scrollStyle=&TextScrollStyle-Base_C
    text="{dhdhhsdINTERACTION NAME and stuff and things }"
    textOverflowPolicy=Ellipsis
    colorAndOpacity {colorUseRule=UseColor_Specified, specifiedColor=<rgba 0.0331 0.402 1 1>}
    font {
     bForceMonospaced=false
     bMaterialIsStencil=false
     fontMaterial=null
     fontObject=&BurbankSmall
     letterSpacing=0
     monospacedWidth=1
     size=16
     skewAmount=0
     typefaceFontName=Black
     outlineSettings {
      bApplyOutlineToDropShadows=false
      bMiteredCorners=false
      bSeparateFillAlpha=false
      outlineMaterial=null
      outlineSize=0
      outlineColor=<rgba 0 0 0 1>
     }
    }
   }
   slot {class=GridSlot, properties {column=1, verticalAlignment=VAlign_Center, padding=<ltrb 32 0 32 0>}}
  }
  Image_Arrow {
   widgetClassPath=&Image
   bIsVariable=true
   parent=GridPanel_0
   widget=Image_Arrow
   properties {
    visibility=Collapsed
    brush {
     bIsDynamicallyLoaded=false
     drawAs=Image
     imageType=NoImage
     mirroring=NoMirror
     resourceName=None
     resourceObject=/FortUILibrary/Foundation/Buttons/Materials/Arrows/M_CycleButton_Arrow_Rgt
     tiling=NoTile
     imageSize=<24 24>
     margin=<ltrb 0 0 0 0>
     outlineSettings {
      bUseBrushTransparency=false
      roundingType=HalfHeightRadius
      width=0
      color {colorUseRule=UseColor_Specified, specifiedColor=<rgba 0 0 0 0>}
      cornerRadii=<0 0 0 0>
     }
     tintColor {colorUseRule=UseColor_Specified, specifiedColor=<rgba 1 1 1 1>}
     uVRegion {bIsValid=false, max=<0 0>, min=<0 0>}
    }
    colorAndOpacity=<rgba 0.0331 0.402 1 1>
   }
   slot {
    class=GridSlot
    properties {column=2, horizontalAlignment=HAlign_Left, verticalAlignment=VAlign_Center, nudge=<10 0>, padding=<ltrb -32 0 0 0>}
   }
  }
  HoverBubble {
   widgetClassPath=&SizeBox
   parent=Overlay_0
   widget=HoverBubble
   properties {bOverride_MinDesiredWidth=true, minDesiredWidth=80, visibility=HitTestInvisible}
   slot {class=OverlaySlot, properties {horizontalAlignment=HAlign_Right, verticalAlignment=VAlign_Center}}
  }
  Overlay_1 {widgetClassPath=&Overlay, parent=HoverBubble, widget=Overlay_1, slot {class=SizeBoxSlot}}
  AnimatedInputBorder {
   widgetClassPath=&Border
   parent=Overlay_1
   widget=AnimatedInputBorder
   properties {
    visibility=SelfHitTestInvisible
    background {
     bIsDynamicallyLoaded=false
     drawAs=Image
     imageType=NoImage
     mirroring=NoMirror
     resourceName=None
     resourceObject=/Game/UI/SocialPanel/Art/Materials/M_UI_UserHover_Inst
     tiling=NoTile
     imageSize=<64 64>
     margin=<ltrb 1 0 0 0>
     outlineSettings {
      bUseBrushTransparency=false
      roundingType=HalfHeightRadius
      width=0
      color {colorUseRule=UseColor_Specified, specifiedColor=<rgba 0 0 0 0>}
      cornerRadii=<0 0 0 0>
     }
     tintColor {colorUseRule=UseColor_Specified, specifiedColor=<rgba 1 1 1 1>}
     uVRegion {bIsValid=false, max=<0 0>, min=<0 0>}
    }
    padding=<ltrb 0 0 0 0>
   }
   slot {class=OverlaySlot, properties {horizontalAlignment=HAlign_Right, verticalAlignment=VAlign_Fill}}
  }
  HorizontalBox_6 {
   widgetClassPath=&HorizontalBox
   parent=AnimatedInputBorder
   widget=HorizontalBox_6
   slot {class=BorderSlot, properties {padding=<ltrb 0 0 0 0>}}
  }
  Overlay_187 {widgetClassPath=&Overlay, parent=HorizontalBox_6, widget=Overlay_187, slot {class=HorizontalBoxSlot}}
  AlternateInputActionWidget {
   widgetClassPath=&CommonActionWidget
   bInherited=true
   parent=Overlay_187
   widget=AlternateInputActionWidget
   properties {visibility=Collapsed}
   slot {class=OverlaySlot, properties {verticalAlignment=VAlign_Center, padding=<ltrb 8 0 16 0>}}
  }
  InputActionWidget {
   widgetClassPath=&CommonActionWidget
   bInherited=true
   parent=Overlay_187
   widget=InputActionWidget
   properties {visibility=SelfHitTestInvisible, inputActions [{dataTable=&InputActionDataTable, rowName=Select}]}
   slot {class=OverlaySlot, properties {verticalAlignment=VAlign_Center, padding=<ltrb 8 0 16 0>}}
  }
  Spacer_7 {
   widgetClassPath=&Spacer
   parent=HorizontalBox_6
   widget=Spacer_7
   properties {size=<8 8>}
   slot {class=HorizontalBoxSlot, properties {size {sizeRule=Fill, value=1}}}
  }
  MenuAnchor_Actions {
   widgetClassPath=/Script/UMG.MenuAnchor
   bIsVariable=true
   parent=Overlay_1
   widget=MenuAnchor_Actions
   properties {menuClass=/Game/UI/SocialPanel/UserActionMenu.UserActionMenu_C, placement=MenuPlacement_RightLeftCenter}
   slot {class=OverlaySlot, properties {horizontalAlignment=HAlign_Right, verticalAlignment=VAlign_Center}}
  }
  EncourageEpicFriend {
   widgetClassPath=&Image
   bIsVariable=true
   parent=Overlay_1
   widget=EncourageEpicFriend
   properties {
    bFlipForRightToLeftFlowDirection=true
    visibility=Collapsed
    brush {
     bIsDynamicallyLoaded=false
     drawAs=Image
     imageType=NoImage
     mirroring=NoMirror
     resourceName=None
     resourceObject=/Game/UI/SocialPanel/Art/Materials/M_UI_EpicFriendOpportunity_Badge_UserEntry
     tiling=NoTile
     imageSize=<32 32>
     margin=<ltrb 0 0 0 0>
     outlineSettings {
      bUseBrushTransparency=false
      roundingType=HalfHeightRadius
      width=0
      color {colorUseRule=UseColor_Specified, specifiedColor=<rgba 0 0 0 0>}
      cornerRadii=<0 0 0 0>
     }
     tintColor {colorUseRule=UseColor_Specified, specifiedColor=<rgba 1 1 1 1>}
     uVRegion {bIsValid=false, max=<0 0>, min=<0 0>}
    }
    colorAndOpacity=<rgba 1 0.9552 0 1>
   }
   slot {class=OverlaySlot, properties {horizontalAlignment=HAlign_Right}}
  }
 ]
}
