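# Sidebar social menu for actions on a selected user, built on FortSidebarSocialInteractionMenu.
# A user cell with name, platform aliases and presence sits above a switcher between interaction lists, a yes/no confirmation and a nickname editor.
# Positive, expanded and negative interactions are dynamic entry boxes of profile interaction entries.
# The graph only plays animations on open and when confirmation becomes visible; other widget logic comes from the parent class.
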WidgetBlueprint /Game/UI/Sidebar/SidebarUserActionMenu
animations [
 Opened {
  binding_count=3
  display_label=""
  length_seconds=0.3
  object=&self:Opened
  sequence {bindings [
    UserCellBorder {
     id=#1
     possessed_class=SizeBox
     tracks [
      {
       type=MovieSceneFloatTrack
       display_name=Render Opacity
       sections [{
         type=MovieSceneFloatSection
         end_frame=6
         start_frame=0
         channels [None {type=MovieSceneScriptingFloatChannel, thinned_from=4, keys [[0,0],[3,1],[6,1]]}]
        }]
      }
      {
       type=MovieScene2DTransformTrack
       sections [{
         type=MovieScene2DTransformSection
         end_frame=6
         start_frame=0
         channels [
          Translation.X {type=MovieSceneScriptingFloatChannel, thinned_from=4, keys [[0,-256],[3,0],[6,0]]}
          Translation.Y {type=MovieSceneScriptingFloatChannel}
          Angle {type=MovieSceneScriptingFloatChannel}
          Scale.X {type=MovieSceneScriptingFloatChannel}
          Scale.Y {type=MovieSceneScriptingFloatChannel}
          Shear.X {type=MovieSceneScriptingFloatChannel, thinned_from=4, keys [[0,16],[3,0],[6,0]]}
          Shear.Y {type=MovieSceneScriptingFloatChannel}
         ]
        }]
      }
     ]
    }
    InteractionsBorder {
     id=#2
     possessed_class=Border
     tracks [
      {
       type=MovieScene2DTransformTrack
       sections [{
         type=MovieScene2DTransformSection
         channels [
          Translation.X {type=MovieSceneScriptingFloatChannel}
          Translation.Y {type=MovieSceneScriptingFloatChannel, keys [[0,0],[1,-72],[3,0],[4,0]]}
          Angle {type=MovieSceneScriptingFloatChannel}
          Scale.X {type=MovieSceneScriptingFloatChannel}
          Scale.Y {type=MovieSceneScriptingFloatChannel}
          Shear.X {type=MovieSceneScriptingFloatChannel}
          Shear.Y {type=MovieSceneScriptingFloatChannel}
         ]
        }]
      }
      {
       type=MovieSceneFloatTrack
       display_name=Render Opacity
       sections [{
         type=MovieSceneFloatSection
         end_frame=6
         start_frame=0
         channels [None {type=MovieSceneScriptingFloatChannel, keys [[0,0],[1,0],[3,1],[6,1]]}]
        }]
      }
     ]
    }
    EntryBorder {
     id=#3
     possessed_class=Border
     tracks [{
       type=MovieSceneWidgetMaterialTrack
       sections [{type=MovieSceneParameterSection, channels [AnimateHighlight {type=MovieSceneScriptingFloatChannel, keys [[0,0],[4,1],[6,1]]}]}]
      }]
    }
   ]}
 }
 ConfirmationOpened {
  binding_count=1
  display_label=""
  length_seconds=0.3
  object=&self:ConfirmationOpened
  sequence {bindings [Border_Confirmation {
     id=#4
     possessed_class=Border
     tracks [
      {
       type=MovieScene2DTransformTrack
       sections [{
         type=MovieScene2DTransformSection
         channels [
          Translation.X {type=MovieSceneScriptingFloatChannel}
          Translation.Y {type=MovieSceneScriptingFloatChannel, keys [[0,0],[1,-72],[3,0],[4,0]]}
          Angle {type=MovieSceneScriptingFloatChannel}
          Scale.X {type=MovieSceneScriptingFloatChannel}
          Scale.Y {type=MovieSceneScriptingFloatChannel}
          Shear.X {type=MovieSceneScriptingFloatChannel}
          Shear.Y {type=MovieSceneScriptingFloatChannel}
         ]
        }]
      }
      {
       type=MovieSceneFloatTrack
       display_name=Render Opacity
       sections [{
         type=MovieSceneFloatSection
         end_frame=6
         start_frame=0
         channels [None {type=MovieSceneScriptingFloatChannel, keys [[0,0],[1,0],[3,1],[6,1]]}]
        }]
      }
     ]
    }]}
 }
]
graphs [EventGraph {nodes [
   n1 = Event OnToggleConfirmation!() -> (OutputDelegate, then, bIsVisible:bool) {is=title K2Node_Event}
   > Branch(Condition=@n1.bIsVisible)
   > PlayAnimation(InAnimation=Get ConfirmationOpened() -> (ConfirmationOpened:Widget Animation Object Reference))
   Event OnOpened() {is=title K2Node_Event}
   > PlayAnimation(InAnimation=Get Opened() -> (Opened:Widget Animation Object Reference))
  ]}]
variables []
widgets {
 info {
  inheritedWidgetCount=16
  parentClass=/Script/FortniteUI.FortSidebarSocialInteractionMenu
  rootWidgetClass=&Overlay
  widgetCount=39
 }
 widgets [
  Overlay_0 {widgetClassPath=&Overlay, widget=Overlay_0}
  SizeBox_0 {
   widgetClassPath=&SizeBox
   parent=Overlay_0
   widget=SizeBox_0
   properties {bOverride_WidthOverride=true, widthOverride=520}
   slot {class=OverlaySlot, properties {horizontalAlignment=HAlign_Fill}}
  }
  VB_Content {widgetClassPath=&VerticalBox, parent=SizeBox_0, widget=VB_Content, slot {class=SizeBoxSlot}}
  UserCellBorder {
   widgetClassPath=&SizeBox
   parent=VB_Content
   widget=UserCellBorder
   properties {bOverride_MinDesiredHeight=true, minDesiredHeight=72, visibility=Collapsed}
   slot {class=VerticalBoxSlot}
  }
  Border_4 {
   widgetClassPath=&Border
   parent=UserCellBorder
   widget=Border_4
   properties {brushColor=<rgba 0 0.0168 0.0887 1>, padding=<ltrb 8 0 8 0>}
   slot {class=SizeBoxSlot}
  }
  EntryBorder {
   widgetClassPath=&Border
   bIsVariable=true
   parent=Border_4
   widget=EntryBorder
   properties {
    visibility=SelfHitTestInvisible
    background {
     bIsDynamicallyLoaded=false
     drawAs=Image
     imageType=NoImage
     mirroring=NoMirror
     resourceName=None
     resourceObject=/Game/UI/SocialPanel/Art/Materials/M_UI_SocialUserItemBorder_UserListItem
     tiling=NoTile
     imageSize=<32 32>
     margin=<ltrb 0 0 0 0>
     outlineSettings {
      bUseBrushTransparency=false
      roundingType=HalfHeightRadius
      width=0
      color {colorUseRule=UseColor_Specified, specifiedColor=<rgba 0 0 0 0>}
      cornerRadii=<0 0 0 0>
     }
     tintColor {colorUseRule=UseColor_Specified, specifiedColor=<rgba 1 1 1 1>}
     uVRegion {bIsValid=false, max=<0 0>, min=<0 0>}
    }
    padding=<ltrb 0 0 0 0>
   }
   slot {class=BorderSlot, properties {padding=<ltrb 8 0 8 0>}}
  }
  HorizontalBox_2 {
   widgetClassPath=&HorizontalBox
   parent=EntryBorder
   widget=HorizontalBox_2
   properties {flowDirectionPreference=Culture}
   slot {class=BorderSlot, properties {padding=<ltrb 0 0 0 0>}}
  }
  Border_OnlineStatusIndicator {
   widgetClassPath=&Border
   bIsVariable=true
   parent=HorizontalBox_2
   widget=Border_OnlineStatusIndicator
   properties {brushColor=<rgba 0.4287 0.6654 0.7991 0.6>, padding=<ltrb 0 0 0 0>}
   slot {class=HorizontalBoxSlot}
  }
  Spacer_0 {
   widgetClassPath=&Spacer
   parent=Border_OnlineStatusIndicator
   widget=Spacer_0
   properties {size=<8 8>}
   slot {class=BorderSlot, properties {padding=<ltrb 0 0 0 0>}}
  }
  UserTextInfoVB {
   widgetClassPath=&VerticalBox
   parent=HorizontalBox_2
   widget=UserTextInfoVB
   slot {class=HorizontalBoxSlot, properties {verticalAlignment=VAlign_Center, padding=<ltrb 24 10 0 10>}}
  }
  SizeBox_2 {
   widgetClassPath=&SizeBox
   parent=UserTextInfoVB
   widget=SizeBox_2
   properties {
    bOverride_MinDesiredHeight=true
    bOverride_WidthOverride=true
    heightOverride=24
    maxDesiredWidth=286
    minDesiredHeight=24
    widthOverride=326
   }
   slot {class=VerticalBoxSlot, properties {horizontalAlignment=HAlign_Left, size {sizeRule=Fill, value=1}}}
  }
  Text_SocialName {
   widgetClassPath=/Script/FortniteUI.FortSocialNameTextBlock
   bInherited=true
   bIsVariable=true
   parent=SizeBox_2
   widget=Text_SocialName
   properties {
    bOverrideDefaultStyle=true
    bShowSecondName=false
    bTintInlineIcon=true
    mobileTextBlockScale=1.34
    platformIconDisplayRule=AlwaysIfDifferent
    scrollStyle=/FortUILibrary/Styles/Typography/Foundation/TextScrollStyle-NoFade.TextScrollStyle-NoFade_C
    text="User Name<PC/> <text color=\"1DEFA9FF\" size=\"12\">(NickName)</>"
    defaultTextStyleOverride {
     overflowPolicy=Clip
     transformPolicy=None
     colorAndOpacity {colorUseRule=UseColor_Specified, specifiedColor=<rgba 0.521 1 0.8228 1>}
     font {
      bForceMonospaced=false
      bMaterialIsStencil=false
      fontMaterial=null
      fontObject=&BurbankSmall
      letterSpacing=0
      monospacedWidth=1
      size=18
      skewAmount=0
      typefaceFontName=Black
      outlineSettings {
       bApplyOutlineToDropShadows=false
       bMiteredCorners=false
       bSeparateFillAlpha=false
       outlineMaterial=null
       outlineSize=0
       outlineColor=<rgba 0 0 0 1>
      }
     }
     highlightColor {colorUseRule=UseColor_Specified, specifiedColor=<rgba 0 0 0 1>}
     highlightShape {
      bIsDynamicallyLoaded=false
      drawAs=Image
      imageType=NoImage
      mirroring=NoMirror
      resourceName=None
      resourceObject=null
      tiling=NoTile
      imageSize=<32 32>
      margin=<ltrb 0 0 0 0>
      outlineSettings {
       bUseBrushTransparency=false
       roundingType=HalfHeightRadius
       width=0
       color {colorUseRule=UseColor_Specified, specifiedColor=<rgba 0 0 0 0>}
       cornerRadii=<0 0 0 0>
      }
      tintColor {colorUseRule=UseColor_Specified, specifiedColor=<rgba 1 1 1 1>}
      uVRegion {bIsValid=false, max=<0 0>, min=<0 0>}
     }
     selectedBackgroundColor {colorUseRule=UseColor_Foreground, specifiedColor=<rgba 1 0 1 1>}
     shadowColorAndOpacity=<rgba 0 0 0 1>
     shadowOffset=<0 0>
     strikeBrush {
      bIsDynamicallyLoaded=false
      drawAs=Image
      imageType=NoImage
      mirroring=NoMirror
      resourceName=None
      resourceObject=null
      tiling=NoTile
      imageSize=<32 32>
      margin=<ltrb 0 0 0 0>
      outlineSettings {
       bUseBrushTransparency=false
       roundingType=HalfHeightRadius
       width=0
       color {colorUseRule=UseColor_Specified, specifiedColor=<rgba 0 0 0 0>}
       cornerRadii=<0 0 0 0>
      }
      tintColor {colorUseRule=UseColor_Specified, specifiedColor=<rgba 1 1 1 1>}
      uVRegion {bIsValid=false, max=<0 0>, min=<0 0>}
     }
     underlineBrush {
      bIsDynamicallyLoaded=false
      drawAs=Image
      imageType=NoImage
      mirroring=NoMirror
      resourceName=None
      resourceObject=null
      tiling=NoTile
      imageSize=<32 32>
      margin=<ltrb 0 0 0 0>
      outlineSettings {
       bUseBrushTransparency=false
       roundingType=HalfHeightRadius
       width=0
       color {colorUseRule=UseColor_Specified, specifiedColor=<rgba 0 0 0 0>}
       cornerRadii=<0 0 0 0>
      }
      tintColor {colorUseRule=UseColor_Specified, specifiedColor=<rgba 1 1 1 1>}
      uVRegion {bIsValid=false, max=<0 0>, min=<0 0>}
     }
    }
   }
   slot {class=SizeBoxSlot, properties {padding=<ltrb 0 0 0 4>}}
  }
  RichText_PlatformNames {
   widgetClassPath=&CommonRichTextBlock
   bInherited=true
   bIsVariable=true
   parent=UserTextInfoVB
   widget=RichText_PlatformNames
   properties {
    bOverrideDefaultStyle=true
    mobileTextBlockScale=1.34
    scrollStyle=&TextScrollStyle-NoFade_C
    text="<EPIC/> ALIAS\r\n<PSN/> ALIAS"
    defaultTextStyleOverride {
     overflowPolicy=Clip
     transformPolicy=None
     colorAndOpacity {colorUseRule=UseColor_Specified, specifiedColor=<rgba 0.4452 1 1 0.5>}
     font {
      bForceMonospaced=false
      bMaterialIsStencil=false
      fontMaterial=null
      fontObject=&BurbankSmall
      letterSpacing=0
      monospacedWidth=1
      size=16
      skewAmount=0
      typefaceFontName=Black
      outlineSettings {
       bApplyOutlineToDropShadows=false
       bMiteredCorners=false
       bSeparateFillAlpha=false
       outlineMaterial=null
       outlineSize=0
       outlineColor=<rgba 0 0 0 1>
      }
     }
     highlightColor {colorUseRule=UseColor_Specified, specifiedColor=<rgba 0 0 0 1>}
     highlightShape {
      bIsDynamicallyLoaded=false
      drawAs=Image
      imageType=NoImage
      mirroring=NoMirror
      resourceName=None
      resourceObject=null
      tiling=NoTile
      imageSize=<32 32>
      margin=<ltrb 0 0 0 0>
      outlineSettings {
       bUseBrushTransparency=false
       roundingType=HalfHeightRadius
       width=0
       color {colorUseRule=UseColor_Specified, specifiedColor=<rgba 0 0 0 0>}
       cornerRadii=<0 0 0 0>
      }
      tintColor {colorUseRule=UseColor_Specified, specifiedColor=<rgba 1 1 1 1>}
      uVRegion {bIsValid=false, max=<0 0>, min=<0 0>}
     }
     selectedBackgroundColor {colorUseRule=UseColor_Foreground, specifiedColor=<rgba 1 0 1 1>}
     shadowColorAndOpacity=<rgba 0 0 0 1>
     shadowOffset=<0 0>
     strikeBrush {
      bIsDynamicallyLoaded=false
      drawAs=Image
      imageType=NoImage
      mirroring=NoMirror
      resourceName=None
      resourceObject=null
      tiling=NoTile
      imageSize=<32 32>
      margin=<ltrb 0 0 0 0>
      outlineSettings {
       bUseBrushTransparency=false
       roundingType=HalfHeightRadius
       width=0
       color {colorUseRule=UseColor_Specified, specifiedColor=<rgba 0 0 0 0>}
       cornerRadii=<0 0 0 0>
      }
      tintColor {colorUseRule=UseColor_Specified, specifiedColor=<rgba 1 1 1 1>}
      uVRegion {bIsValid=false, max=<0 0>, min=<0 0>}
     }
     underlineBrush {
      bIsDynamicallyLoaded=false
      drawAs=Image
      imageType=NoImage
      mirroring=NoMirror
      resourceName=None
      resourceObject=null
      tiling=NoTile
      imageSize=<32 32>
      margin=<ltrb 0 0 0 0>
      outlineSettings {
       bUseBrushTransparency=false
       roundingType=HalfHeightRadius
       width=0
       color {colorUseRule=UseColor_Specified, specifiedColor=<rgba 0 0 0 0>}
       cornerRadii=<0 0 0 0>
      }
      tintColor {colorUseRule=UseColor_Specified, specifiedColor=<rgba 1 1 1 1>}
      uVRegion {bIsValid=false, max=<0 0>, min=<0 0>}
     }
    }
   }
   slot {class=VerticalBoxSlot}
  }
  SizeBox_3 {
   widgetClassPath=&SizeBox
   parent=UserTextInfoVB
   widget=SizeBox_3
   properties {
    bOverride_MaxDesiredHeight=true
    bOverride_MaxDesiredWidth=true
    bOverride_MinDesiredHeight=true
    heightOverride=16
    maxDesiredHeight=32
    maxDesiredWidth=325
    minDesiredHeight=16
   }
   slot {class=VerticalBoxSlot, properties {horizontalAlignment=HAlign_Left, padding=<ltrb 0 4 0 2>}}
  }
  Text_Presence {
   widgetClassPath=&CommonRichTextBlock
   bInherited=true
   bIsVariable=true
   parent=SizeBox_3
   widget=Text_Presence
   properties {
    bOverrideDefaultStyle=true
    mobileTextBlockScale=1.34
    scrollStyle=&TextScrollStyle-NoFade_C
    text="{USER PRESENCE}"
    defaultTextStyleOverride {
     overflowPolicy=Clip
     transformPolicy=None
     colorAndOpacity {colorUseRule=UseColor_Specified, specifiedColor=<rgba 1 1 1 0.5>}
     font {
      bForceMonospaced=false
      bMaterialIsStencil=false
      fontMaterial=null
      fontObject=&BurbankSmall
      letterSpacing=0
      monospacedWidth=1
      size=15
      skewAmount=0
      typefaceFontName=Bold
      outlineSettings {
       bApplyOutlineToDropShadows=false
       bMiteredCorners=false
       bSeparateFillAlpha=false
       outlineMaterial=null
       outlineSize=0
       outlineColor=<rgba 0 0 0 1>
      }
     }
     highlightColor {colorUseRule=UseColor_Specified, specifiedColor=<rgba 0 0 0 1>}
     highlightShape {
      bIsDynamicallyLoaded=false
      drawAs=Image
      imageType=NoImage
      mirroring=NoMirror
      resourceName=None
      resourceObject=null
      tiling=NoTile
      imageSize=<32 32>
      margin=<ltrb 0 0 0 0>
      outlineSettings {
       bUseBrushTransparency=false
       roundingType=HalfHeightRadius
       width=0
       color {colorUseRule=UseColor_Specified, specifiedColor=<rgba 0 0 0 0>}
       cornerRadii=<0 0 0 0>
      }
      tintColor {colorUseRule=UseColor_Specified, specifiedColor=<rgba 1 1 1 1>}
      uVRegion {bIsValid=false, max=<0 0>, min=<0 0>}
     }
     selectedBackgroundColor {colorUseRule=UseColor_Foreground, specifiedColor=<rgba 1 0 1 1>}
     shadowColorAndOpacity=<rgba 0 0 0 1>
     shadowOffset=<0 0>
     strikeBrush {
      bIsDynamicallyLoaded=false
      drawAs=Image
      imageType=NoImage
      mirroring=NoMirror
      resourceName=None
      resourceObject=null
      tiling=NoTile
      imageSize=<32 32>
      margin=<ltrb 0 0 0 0>
      outlineSettings {
       bUseBrushTransparency=false
       roundingType=HalfHeightRadius
       width=0
       color {colorUseRule=UseColor_Specified, specifiedColor=<rgba 0 0 0 0>}
       cornerRadii=<0 0 0 0>
      }
      tintColor {colorUseRule=UseColor_Specified, specifiedColor=<rgba 1 1 1 1>}
      uVRegion {bIsValid=false, max=<0 0>, min=<0 0>}
     }
     underlineBrush {
      bIsDynamicallyLoaded=false
      drawAs=Image
      imageType=NoImage
      mirroring=NoMirror
      resourceName=None
      resourceObject=null
      tiling=NoTile
      imageSize=<32 32>
      margin=<ltrb 0 0 0 0>
      outlineSettings {
       bUseBrushTransparency=false
       roundingType=HalfHeightRadius
       width=0
       color {colorUseRule=UseColor_Specified, specifiedColor=<rgba 0 0 0 0>}
       cornerRadii=<0 0 0 0>
      }
      tintColor {colorUseRule=UseColor_Specified, specifiedColor=<rgba 1 1 1 1>}
      uVRegion {bIsValid=false, max=<0 0>, min=<0 0>}
     }
    }
   }
   slot {class=SizeBoxSlot}
  }
  PlatformIconSB {
   widgetClassPath=&SizeBox
   parent=HorizontalBox_2
   widget=PlatformIconSB
   properties {
    bOverride_MinDesiredHeight=true
    bOverride_MinDesiredWidth=true
    heightOverride=72
    minDesiredHeight=72
    minDesiredWidth=72
    widthOverride=72
   }
   slot {class=HorizontalBoxSlot, properties {horizontalAlignment=HAlign_Right, size {sizeRule=Fill, value=1}}}
  }
  Overlay_1 {
   widgetClassPath=&Overlay
   parent=PlatformIconSB
   widget=Overlay_1
   slot {class=SizeBoxSlot, properties {horizontalAlignment=HAlign_Center, verticalAlignment=VAlign_Center}}
  }
  Image_Platform {
   widgetClassPath=/Script/FortniteUI.FortMobileImage
   bIsVariable=true
   parent=Overlay_1
   widget=Image_Platform
   properties {
    mobileSizeMultiplier=1.34
    visibility=Collapsed
    brush {
     bIsDynamicallyLoaded=false
     drawAs=Image
     imageType=NoImage
     mirroring=NoMirror
     resourceName=None
     resourceObject=/Game/UI/Foundation/Textures/Icons/Social/T-Icon-EpicGamesLogo-64
     tiling=NoTile
     imageSize=<64 64>
     margin=<ltrb 0 0 0 0>
     outlineSettings {
      bUseBrushTransparency=false
      roundingType=HalfHeightRadius
      width=0
      color {colorUseRule=UseColor_Specified, specifiedColor=<rgba 0 0 0 0>}
      cornerRadii=<0 0 0 0>
     }
     tintColor {colorUseRule=UseColor_Specified, specifiedColor=<rgba 1 1 1 1>}
     uVRegion {bIsValid=false, max=<0 0>, min=<0 0>}
    }
    colorAndOpacity=<rgba 1 1 1 0.298>
   }
   slot {class=OverlaySlot}
  }
  Switcher_Confirmation {
   widgetClassPath=&CommonWidgetSwitcherLegacy
   bInherited=true
   bIsVariable=true
   parent=VB_Content
   widget=Switcher_Confirmation
   properties {transitionDuration=0.2}
   slot {class=VerticalBoxSlot}
  }
  InteractionsBorder {
   widgetClassPath=&Border
   parent=Switcher_Confirmation
   widget=InteractionsBorder
   properties {
    background {
     bIsDynamicallyLoaded=false
     drawAs=Image
     imageType=NoImage
     mirroring=NoMirror
     resourceName=None
     resourceObject=null
     tiling=NoTile
     imageSize=<32 32>
     margin=<ltrb 0 0 0 0>
     outlineSettings {
      bUseBrushTransparency=false
      roundingType=HalfHeightRadius
      width=0
      color {colorUseRule=UseColor_Specified, specifiedColor=<rgba 0 0 0 0>}
      cornerRadii=<0 0 0 0>
     }
     tintColor {colorUseRule=UseColor_Specified, specifiedColor=<rgba 1 1 1 0>}
     uVRegion {bIsValid=false, max=<0 0>, min=<0 0>}
    }
    padding=<ltrb 25 20 25 20>
   }
   slot {class=WidgetSwitcherSlot}
  }
  VerticalBox_1 {
   widgetClassPath=&VerticalBox
   parent=InteractionsBorder
   widget=VerticalBox_1
   slot {class=BorderSlot, properties {padding=<ltrb 25 20 25 20>}}
  }
  EntryBox_PositiveInteractions {
   widgetClassPath=&DynamicEntryBox
   bInherited=true
   bIsVariable=true
   parent=VerticalBox_1
   widget=EntryBox_PositiveInteractions
   properties {entryBoxType=Vertical, entryWidgetClass=&/SidebarInteraction_Profile.SidebarInteraction_Profile_C, entrySpacing=<0 10>}
   slot {class=VerticalBoxSlot}
  }
  Spacer_ExpandedMenuSeparator {
   widgetClassPath=&Spacer
   bInherited=true
   parent=VerticalBox_1
   widget=Spacer_ExpandedMenuSeparator
   properties {size=<1 20>}
   slot {class=VerticalBoxSlot}
  }
  EntryBox_ExpandedInteractions {
   widgetClassPath=&DynamicEntryBox
   bInherited=true
   bIsVariable=true
   parent=VerticalBox_1
   widget=EntryBox_ExpandedInteractions
   properties {entryBoxType=Vertical, entryWidgetClass=&SidebarInteraction_Profile_C, entrySpacing=<0 10>}
   slot {class=VerticalBoxSlot}
  }
  Spacer_InteractionSeparator {
   widgetClassPath=&Spacer
   bInherited=true
   parent=VerticalBox_1
   widget=Spacer_InteractionSeparator
   properties {size=<1 20>}
   slot {class=VerticalBoxSlot}
  }
  EntryBox_NegativeInteractions {
   widgetClassPath=&DynamicEntryBox
   bInherited=true
   bIsVariable=true
   parent=VerticalBox_1
   widget=EntryBox_NegativeInteractions
   properties {entryBoxType=Vertical, entryWidgetClass=&SidebarInteraction_Profile_C, entrySpacing=<0 10>}
   slot {class=VerticalBoxSlot}
  }
  Border_Confirmation {
   widgetClassPath=&Border
   parent=Switcher_Confirmation
   widget=Border_Confirmation
   properties {
    navigation=&self:WidgetTree.Border_Confirmation.WidgetNavigation_0
    verticalAlignment=VAlign_Center
    visibility=SelfHitTestInvisible
    brushColor=<rgba 0 0.0168 0.0887 0>
    padding=<ltrb 0 0 0 0>
    renderTransformPivot=<0 0.5>
   }
   slot {class=WidgetSwitcherSlot}
  }
  VerticalBox_0 {
   widgetClassPath=&VerticalBox
   parent=Border_Confirmation
   widget=VerticalBox_0
   slot {class=BorderSlot, properties {verticalAlignment=VAlign_Center, padding=<ltrb 0 0 0 0>}}
  }
  Text_ConfirmationLabel {
   widgetClassPath=&CommonTextBlock
   bInherited=true
   parent=VerticalBox_0
   widget=Text_ConfirmationLabel
   properties {
    justification=Center
    scrollStyle=&TextScrollStyle-Base_C
    text=CONFIRM ACTION?
    font {
     bForceMonospaced=false
     bMaterialIsStencil=false
     fontMaterial=null
     fontObject=&BurbankSmall
     letterSpacing=0
     monospacedWidth=1
     size=24
     skewAmount=0
     typefaceFontName=Bold
     outlineSettings {
      bApplyOutlineToDropShadows=false
      bMiteredCorners=false
      bSeparateFillAlpha=false
      outlineMaterial=null
      outlineSize=0
      outlineColor=<rgba 0 0 0 1>
     }
    }
   }
   slot {class=VerticalBoxSlot, properties {horizontalAlignment=HAlign_Center, padding=<ltrb 0 40 0 0>}}
  }
  HorizontalBox_0 {
   widgetClassPath=&HorizontalBox
   parent=VerticalBox_0
   widget=HorizontalBox_0
   properties {flowDirectionPreference=Culture}
   slot {class=VerticalBoxSlot, properties {padding=<ltrb 20 20 20 20>}}
  }
  Button_Confirm {
   widgetClassPath=&/SidebarConfirmationButton
   bInherited=true
   bIsVariable=true
   parent=HorizontalBox_0
   widget=Button_Confirm
   properties {
    bIsAccept=true
    displayText=YES
    navigation=&self:WidgetTree.Button_Confirm.WidgetNavigation_0
    onButtonBaseClicked=(&self:WidgetTree.Button_Confirm.HandleOnButtonClicked)
    onButtonBaseDoubleClicked=(&self:WidgetTree.Button_Confirm.HandleOnButtonDoubleClicked)
    onButtonBaseHovered=(&self:WidgetTree.Button_Confirm.HandleOnButtonHovered)
    onButtonBaseUnhovered=(&self:WidgetTree.Button_Confirm.HandleOnButtonUnhovered)
    onSelectedChangedBase=(&self:WidgetTree.Button_Confirm.HandleOnSelectedChanged)
   }
   slot {class=HorizontalBoxSlot, properties {size {sizeRule=Fill, value=1}}}
  }
  Spacer_54 {
   widgetClassPath=&Spacer
   parent=HorizontalBox_0
   widget=Spacer_54
   slot {class=HorizontalBoxSlot, properties {padding=<ltrb 10 0 10 0>}}
  }
  Button_Decline {
   widgetClassPath=&SidebarConfirmationButton
   bInherited=true
   bIsVariable=true
   parent=HorizontalBox_0
   widget=Button_Decline
   properties {
    displayText=NO
    navigation=&self:WidgetTree.Button_Decline.WidgetNavigation_0
    onButtonBaseClicked=(&self:WidgetTree.Button_Decline.HandleOnButtonClicked)
    onButtonBaseDoubleClicked=(&self:WidgetTree.Button_Decline.HandleOnButtonDoubleClicked)
    onButtonBaseHovered=(&self:WidgetTree.Button_Decline.HandleOnButtonHovered)
    onButtonBaseUnhovered=(&self:WidgetTree.Button_Decline.HandleOnButtonUnhovered)
    onSelectedChangedBase=(&self:WidgetTree.Button_Decline.HandleOnSelectedChanged)
   }
   slot {class=HorizontalBoxSlot, properties {size {sizeRule=Fill, value=1}}}
  }
  Border_Nickname {
   widgetClassPath=&Border
   parent=Switcher_Confirmation
   widget=Border_Nickname
   properties {
    navigation=&self:WidgetTree.Border_Nickname.WidgetNavigation_0
    verticalAlignment=VAlign_Center
    visibility=SelfHitTestInvisible
    brushColor=<rgba 0 0.0168 0.0887 0>
    padding=<ltrb 20 40 20 20>
    renderTransformPivot=<0 0.5>
   }
   slot {class=WidgetSwitcherSlot}
  }
  VerticalBox {
   widgetClassPath=&VerticalBox
   parent=Border_Nickname
   widget=VerticalBox
   slot {class=BorderSlot, properties {verticalAlignment=VAlign_Center, padding=<ltrb 20 40 20 20>}}
  }
  Button_EditNickname {
   widgetClassPath=&/SidebarEditNicknameButton
   bInherited=true
   bIsVariable=true
   parent=VerticalBox
   widget=Button_EditNickname
   properties {
    onButtonBaseClicked=(&self:WidgetTree.Button_EditNickname.HandleOnButtonClicked)
    onButtonBaseDoubleClicked=(&self:WidgetTree.Button_EditNickname.HandleOnButtonDoubleClicked)
    onButtonBaseHovered=(&self:WidgetTree.Button_EditNickname.HandleOnButtonHovered)
    onButtonBaseUnhovered=(&self:WidgetTree.Button_EditNickname.HandleOnButtonUnhovered)
    onSelectedChangedBase=(&self:WidgetTree.Button_EditNickname.HandleOnSelectedChanged)
   }
   slot {class=VerticalBoxSlot}
  }
  Text_ErrorMessage {
   widgetClassPath=&CommonTextBlock
   bInherited=true
   parent=VerticalBox
   widget=Text_ErrorMessage
   properties {
    scrollStyle=&TextScrollStyle-NoFade_C
    text="Sorry, the friend alias you entered contains disallowed words"
    visibility=Collapsed
    wrapTextAt=500
    colorAndOpacity {colorUseRule=UseColor_Specified, specifiedColor=<rgba 1 0.0469 0.0469 1>}
    font {
     bForceMonospaced=false
     bMaterialIsStencil=false
     fontMaterial=null
     fontObject=&BurbankSmall
     letterSpacing=0
     monospacedWidth=1
     size=16
     skewAmount=0
     typefaceFontName=Bold
     outlineSettings {
      bApplyOutlineToDropShadows=false
      bMiteredCorners=false
      bSeparateFillAlpha=false
      outlineMaterial=null
      outlineSize=0
      outlineColor=<rgba 0 0 0 1>
     }
    }
    shadowOffset=<0 0>
   }
   slot {class=VerticalBoxSlot, properties {padding=<ltrb 16 0 16 8>}}
  }
  Button_DoneEdit {
   widgetClassPath=&SidebarConfirmationButton
   bInherited=true
   bIsVariable=true
   parent=VerticalBox
   widget=Button_DoneEdit
   properties {
    bIsAccept=true
    displayText=DONE
    onButtonBaseClicked=(&self:WidgetTree.Button_DoneEdit.HandleOnButtonClicked)
    onButtonBaseDoubleClicked=(&self:WidgetTree.Button_DoneEdit.HandleOnButtonDoubleClicked)
    onButtonBaseHovered=(&self:WidgetTree.Button_DoneEdit.HandleOnButtonHovered)
    onButtonBaseUnhovered=(&self:WidgetTree.Button_DoneEdit.HandleOnButtonUnhovered)
    onSelectedChangedBase=(&self:WidgetTree.Button_DoneEdit.HandleOnSelectedChanged)
   }
   slot {class=VerticalBoxSlot, properties {padding=<ltrb 0 20 0 0>}}
  }
  Button_CancelEdit {widgetClassPath=&CommonButtonLegacy, bInherited=true, widget=null}
 ]
}
