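# Widget Blueprint for the actions sub-panel of a sidebar social profile, derived from the native FortSocialProfileActions class.
# Its root is a SidebarUserActionMenu, filling the inherited FortSocialInteractionMenu slot to present user interaction options.
# The inherited loading sub-panel has no widget assigned.
# It declares no variables and its event graph holds only unconnected default events, leaving behavior to the parent class and child menu.
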WidgetBlueprint /Game/UI/Sidebar/SideBarProfileSubPanelActions
graphs [EventGraph {nodes [
   Event PreConstruct()
   Event Construct()
   Event Tick()
  ]}]
variables []
widgets {
 info {
  inheritedWidgetCount=2
  parentClass=/Script/FortniteUI.FortSocialProfileActions
  rootWidgetClass=&/SidebarUserActionMenu.SidebarUserActionMenu_C
  widgetCount=2
 }
 widgets [
  FortSocialInteractionMenu {widgetClassPath=&/SidebarUserActionMenu, bInherited=true, bIsVariable=true, widget=FortSocialInteractionMenu}
  SubPanel_Loading {widgetClassPath=/Script/FortniteUI.FortSocialProfileSubPanel_Loading, bInherited=true, widget=null}
 ]
}
