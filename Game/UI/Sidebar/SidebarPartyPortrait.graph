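# Sidebar party portrait widget, a FortJoinablePartyPortrait child that shows a party member's avatar or a "+X" remaining-players badge.
# A visibility switcher holds both states, built mostly from inherited widgets, adding a crown image and a badge background.
# PreConstruct reveals the crown when IsPartyLeader is true.
# OnNewAvatarsEnabledChanged swaps visibility between the PlayerAvatar widget and the older social avatar icon, so only one shows.
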WidgetBlueprint /Game/UI/Sidebar/SidebarPartyPortrait
graphs [EventGraph {nodes [
   Event PreConstruct()
   > Branch(Condition=Get IsPartyLeader() -> (IsPartyLeader:bool))
   > SetVisibility(self=Get Image_Crown_1() -> (Image_Crown_1:Image Object Reference), InVisibility=SelfHitTestInvisible)
   n1 = Event OnNewAvatarsEnabledChanged!() -> (OutputDelegate, then, bUsingNewAvatars:bool) {is=title K2Node_Event}
   > SetVisibility(
    self=Get PlayerAvatar() -> (PlayerAvatar:Fort Player Avatar Object Reference)
    InVisibility=Select(Option 0=Collapsed, Option 1=HitTestInvisible, Index=@n1.bUsingNewAvatars)
   )
   > SetVisibility(
    self=Get Avatar_MemberIcon() -> (Avatar_MemberIcon:Fort Social Avatar Icon Object Reference)
    InVisibility=Select(Option 0=HitTestInvisible, Option 1=Collapsed, Index=@n1.bUsingNewAvatars)
   )
  ]}]
variables [IsPartyLeader {type=boolean}]
widgets {
 info {
  inheritedWidgetCount=6
  parentClass=/Script/FortniteUI.FortJoinablePartyPortrait
  rootWidgetClass=&CommonVisibilitySwitcher
  widgetCount=8
 }
 widgets [
  CommonSwitcher_PortraitState {widgetClassPath=&CommonVisibilitySwitcher, bInherited=true, widget=CommonSwitcher_PortraitState}
  Overlay_Portrait {
   widgetClassPath=&Overlay
   bInherited=true
   parent=CommonSwitcher_PortraitState
   widget=Overlay_Portrait
   slot {class=CommonVisibilitySwitcherSlot, properties {horizontalAlignment=HAlign_Center, verticalAlignment=VAlign_Center}}
  }
  Avatar_MemberIcon {
   widgetClassPath=/Script/FortniteUI.FortSocialAvatarIcon
   bInherited=true
   bIsVariable=true
   parent=Overlay_Portrait
   widget=Avatar_MemberIcon
   properties {
    avatarMaterial=/Game/UI/SocialPanel/Art/Materials/UI_UserPortrait_Buffet
    materialTextureParamName=SmallPreview
    visibility=Collapsed
    brush {
     bIsDynamicallyLoaded=false
     drawAs=Image
     imageType=NoImage
     mirroring=NoMirror
     resourceName=None
     resourceObject=&UI_UserPortrait_Buffet
     tiling=NoTile
     imageSize=<72 72>
     margin=<ltrb 0 0 0 0>
     outlineSettings {
      bUseBrushTransparency=false
      roundingType=HalfHeightRadius
      width=0
      color {colorUseRule=UseColor_Specified, specifiedColor=<rgba 0 0 0 0>}
      cornerRadii=<0 0 0 0>
     }
     tintColor {colorUseRule=UseColor_Specified, specifiedColor=<rgba 1 1 1 1>}
     uVRegion {bIsValid=false, max=<0 0>, min=<0 0>}
    }
    loadingBackgroundBrush {
     bIsDynamicallyLoaded=false
     drawAs=Image
     imageType=NoImage
     mirroring=NoMirror
     resourceName=None
     resourceObject=/FortUILibrary/UIKit/Throbber/MI_UIKit_Monochrome_Throbber
     tiling=NoTile
     imageSize=<32 32>
     margin=<ltrb 0 0 0 0>
     outlineSettings {
      bUseBrushTransparency=false
      roundingType=HalfHeightRadius
      width=0
      color {colorUseRule=UseColor_Specified, specifiedColor=<rgba 0 0 0 0>}
      cornerRadii=<0 0 0 0>
     }
     tintColor {colorUseRule=UseColor_Specified, specifiedColor=<rgba 1 1 1 1>}
     uVRegion {bIsValid=false, max=<0 0>, min=<0 0>}
    }
   }
   slot {class=OverlaySlot}
  }
  Image_Crown_1 {
   widgetClassPath=&Image
   bIsVariable=true
   parent=Overlay_Portrait
   widget=Image_Crown_1
   properties {
    visibility=Collapsed
    brush {
     bIsDynamicallyLoaded=false
     drawAs=Image
     imageType=NoImage
     mirroring=NoMirror
     resourceName=None
     resourceObject=/Game/UI/Foundation/Banner/M-CrownShine
     tiling=NoTile
     imageSize=<16 16>
     margin=<ltrb 0 0 0 0>
     outlineSettings {
      bUseBrushTransparency=false
      roundingType=HalfHeightRadius
      width=0
      color {colorUseRule=UseColor_Specified, specifiedColor=<rgba 0 0 0 0>}
      cornerRadii=<0 0 0 0>
     }
     tintColor {colorUseRule=UseColor_Specified, specifiedColor=<rgba 1 1 1 1>}
     uVRegion {bIsValid=false, max=<0 0>, min=<0 0>}
    }
    renderTransform {angle=8, scale=<1 1>, shear=<0 0>, translation=<0 0>}
   }
   slot {class=OverlaySlot, properties {horizontalAlignment=HAlign_Right, verticalAlignment=VAlign_Bottom, padding=<ltrb 0 0 1 1>}}
  }
  PlayerAvatar {
   widgetClassPath=/Game/UI/Foundation/Profile/WBP_PlayerAvatar
   bInherited=true
   bIsVariable=true
   parent=Overlay_Portrait
   widget=PlayerAvatar
   properties {showOnlinePresence=false, avatarSize=<72 72>}
   slot {class=OverlaySlot}
  }
  Overlay_RemainingPlayersValue {
   widgetClassPath=&Overlay
   bInherited=true
   parent=CommonSwitcher_PortraitState
   widget=Overlay_RemainingPlayersValue
   slot {class=CommonVisibilitySwitcherSlot, properties {horizontalAlignment=HAlign_Left, verticalAlignment=VAlign_Top}}
  }
  Image_ValueBackground {
   widgetClassPath=&FortSocialAvatarIcon
   bIsVariable=true
   parent=Overlay_RemainingPlayersValue
   widget=Image_ValueBackground
   properties {
    avatarMaterial=/Game/UI/Sidebar/Materials/UI_UserPortraitBackMaster_Inst
    materialTextureParamName=SmallPreview
    brush {
     bIsDynamicallyLoaded=false
     drawAs=Image
     imageType=NoImage
     mirroring=NoMirror
     resourceName=None
     resourceObject=&UI_UserPortraitBackMaster_Inst
     tiling=NoTile
     imageSize=<72 72>
     margin=<ltrb 0 0 0 0>
     outlineSettings {
      bUseBrushTransparency=false
      roundingType=HalfHeightRadius
      width=0
      color {colorUseRule=UseColor_Specified, specifiedColor=<rgba 0 0 0 0>}
      cornerRadii=<0 0 0 0>
     }
     tintColor {colorUseRule=UseColor_Specified, specifiedColor=<rgba 1 1 1 1>}
     uVRegion {bIsValid=false, max=<0 0>, min=<0 0>}
    }
    renderTransform {angle=0, scale=<0.925 0.925>, shear=<0 0>, translation=<0 0>}
   }
   slot {class=OverlaySlot}
  }
  Text_RemainingPlayers {
   widgetClassPath=&CommonRichTextBlock
   bInherited=true
   bIsVariable=true
   parent=Overlay_RemainingPlayersValue
   widget=Text_RemainingPlayers
   properties {
    bOverrideDefaultStyle=true
    bTintInlineIcon=true
    lineHeightPercentage=1.2
    mobileTextBlockScale=1.34
    scrollStyle=/FortUILibrary/Styles/Typography/Foundation/TextScrollStyle-NoFade.TextScrollStyle-NoFade_C
    text=+X
    defaultTextStyleOverride {
     overflowPolicy=Clip
     transformPolicy=None
     colorAndOpacity {colorUseRule=UseColor_Specified, specifiedColor=<rgba 1 1 1 1>}
     font {
      bForceMonospaced=false
      bMaterialIsStencil=false
      fontMaterial=null
      fontObject=&BurbankSmall
      letterSpacing=0
      monospacedWidth=1
      size=28
      skewAmount=0
      typefaceFontName=Black
      outlineSettings {
       bApplyOutlineToDropShadows=false
       bMiteredCorners=false
       bSeparateFillAlpha=false
       outlineMaterial=null
       outlineSize=0
       outlineColor=<rgba 0 0 0 1>
      }
     }
     highlightColor {colorUseRule=UseColor_Specified, specifiedColor=<rgba 0 0 0 1>}
     highlightShape {
      bIsDynamicallyLoaded=false
      drawAs=Image
      imageType=NoImage
      mirroring=NoMirror
      resourceName=None
      resourceObject=null
      tiling=NoTile
      imageSize=<32 32>
      margin=<ltrb 0 0 0 0>
      outlineSettings {
       bUseBrushTransparency=false
       roundingType=HalfHeightRadius
       width=0
       color {colorUseRule=UseColor_Specified, specifiedColor=<rgba 0 0 0 0>}
       cornerRadii=<0 0 0 0>
      }
      tintColor {colorUseRule=UseColor_Specified, specifiedColor=<rgba 1 1 1 1>}
      uVRegion {bIsValid=false, max=<0 0>, min=<0 0>}
     }
     selectedBackgroundColor {colorUseRule=UseColor_Foreground, specifiedColor=<rgba 1 0 1 1>}
     shadowColorAndOpacity=<rgba 0 0 0 1>
     shadowOffset=<0 0>
     strikeBrush {
      bIsDynamicallyLoaded=false
      drawAs=Image
      imageType=NoImage
      mirroring=NoMirror
      resourceName=None
      resourceObject=null
      tiling=NoTile
      imageSize=<32 32>
      margin=<ltrb 0 0 0 0>
      outlineSettings {
       bUseBrushTransparency=false
       roundingType=HalfHeightRadius
       width=0
       color {colorUseRule=UseColor_Specified, specifiedColor=<rgba 0 0 0 0>}
       cornerRadii=<0 0 0 0>
      }
      tintColor {colorUseRule=UseColor_Specified, specifiedColor=<rgba 1 1 1 1>}
      uVRegion {bIsValid=false, max=<0 0>, min=<0 0>}
     }
     underlineBrush {
      bIsDynamicallyLoaded=false
      drawAs=Image
      imageType=NoImage
      mirroring=NoMirror
      resourceName=None
      resourceObject=null
      tiling=NoTile
      imageSize=<32 32>
      margin=<ltrb 0 0 0 0>
      outlineSettings {
       bUseBrushTransparency=false
       roundingType=HalfHeightRadius
       width=0
       color {colorUseRule=UseColor_Specified, specifiedColor=<rgba 0 0 0 0>}
       cornerRadii=<0 0 0 0>
      }
      tintColor {colorUseRule=UseColor_Specified, specifiedColor=<rgba 1 1 1 1>}
      uVRegion {bIsValid=false, max=<0 0>, min=<0 0>}
     }
    }
   }
   slot {class=OverlaySlot, properties {horizontalAlignment=HAlign_Center, verticalAlignment=VAlign_Center, padding=<ltrb 0 5 0 0>}}
  }
 ]
}
