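# Sidebar profile sub-panel widget that tells the player someone sent them a friend request and explains what friends can do.
# A grid panel stacks two centered, auto-wrapping CommonTextBlocks: an uppercase title above a smaller gray subtitle covering online status, parties, chat and reporting.
# The content is static, with no variables or buttons, and the PreConstruct, Construct and Tick events are empty.
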WidgetBlueprint /Game/UI/Sidebar/SidebarProfileSubPanelFriendRequest
graphs [EventGraph {nodes [
   Event PreConstruct()
   Event Construct()
   Event Tick()
  ]}]
variables []
widgets {
 info {inheritedWidgetCount=0, parentClass=&UserWidget, rootWidgetClass=&GridPanel, widgetCount=3}
 widgets [
  GridPanel_2 {widgetClassPath=&GridPanel, widget=GridPanel_2, properties {columnFill [1], rowFill [1]}}
  TitleText {
   widgetClassPath=&CommonTextBlock
   parent=GridPanel_2
   widget=TitleText
   properties {
    autoWrapText=true
    bIsScrollingEnabled=false
    justification=Center
    mobileFontSizeMultiplier=1.3
    text="Sent You A Friend Request!"
    textTransformPolicy=ToUpper
    wrapTextAt=500
    font {
     bForceMonospaced=false
     bMaterialIsStencil=false
     fontMaterial=null
     fontObject=&BurbankSmall
     letterSpacing=0
     monospacedWidth=1
     size=18
     skewAmount=0
     typefaceFontName=Black
     outlineSettings {
      bApplyOutlineToDropShadows=false
      bMiteredCorners=false
      bSeparateFillAlpha=false
      outlineMaterial=null
      outlineSize=0
      outlineColor=<rgba 0 0 0 1>
     }
    }
    shadowOffset=<0 0>
   }
   slot {class=GridSlot, properties {verticalAlignment=VAlign_Top, padding=<ltrb 20 0 20 0>}}
  }
  SubtitleText {
   widgetClassPath=&CommonTextBlock
   parent=GridPanel_2
   widget=SubtitleText
   properties {
    autoWrapText=true
    bIsScrollingEnabled=false
    justification=Center
    lineHeightPercentage=1.3
    mobileFontSizeMultiplier=1.4
    text="Friends can see your online status, form parties together, and chat with you depending on your Communication Settings. You can report players from their profile."
    colorAndOpacity {colorUseRule=UseColor_Specified, specifiedColor=<rgba 0.521 0.521 0.521 1>}
    font {
     bForceMonospaced=false
     bMaterialIsStencil=false
     fontMaterial=null
     fontObject=&BurbankSmall
     letterSpacing=0
     monospacedWidth=1
     size=13.5
     skewAmount=0
     typefaceFontName=Bold
     outlineSettings {
      bApplyOutlineToDropShadows=false
      bMiteredCorners=false
      bSeparateFillAlpha=false
      outlineMaterial=null
      outlineSize=0
      outlineColor=<rgba 0 0 0 1>
     }
    }
    shadowOffset=<0 0>
   }
   slot {class=GridSlot, properties {row=1, verticalAlignment=VAlign_Top, padding=<ltrb 20 8 20 0>}}
  }
 ]
}
