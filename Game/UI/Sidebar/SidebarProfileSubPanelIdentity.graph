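# Sidebar profile identity header, built on FortSocialProfileIdentity, showing avatar, social name, voice status, platform names, Battle Pass and Crew badges, and season level.
# Parent-class events only toggle visibility and set the level text.
# The new-avatar event swaps PlayerAvatar against the older social avatar icon and its online-status dot.
# The level row collapses at zero or below; loading, construct and tick events are empty.
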WidgetBlueprint /Game/UI/Sidebar/SidebarProfileSubPanelIdentity
graphs [EventGraph {nodes [
   n1 = Event OnNewAvatarsEnabledChanged!() -> (OutputDelegate, then, bUsingNewAvatars:bool) {is=title K2Node_Event}
   > SetVisibility(
    self=Get PlayerAvatar() -> (PlayerAvatar:Fort Player Avatar Object Reference)
    InVisibility=Select(Option 0=Collapsed, Option 1=HitTestInvisible, Index=@n1.bUsingNewAvatars)
   )
   n2 = Select(Option 0=HitTestInvisible, Option 1=Collapsed, Index=@n1.bUsingNewAvatars)
   > SetVisibility(self=Get Avatar_SocialAvatarIcon() -> (Avatar_SocialAvatarIcon:Fort Social Avatar Icon Object Reference), InVisibility=@n2)
   > SetVisibility(self=Get Overlay_OnlineStatus() -> (Overlay_OnlineStatus:Overlay Object Reference), InVisibility=@n2)
   n3 = Event SetFortniteCrewVisiblity!() -> (OutputDelegate, then, bIsVisible:bool) {is=title K2Node_Event}
   > SetVisibility(
    self=Get Overlay_FortniteCrew() -> (Overlay_FortniteCrew:Overlay Object Reference)
    InVisibility=Select(Option 0=Collapsed, Option 1=HitTestInvisible, Index=@n3.bIsVisible)
   )
   n4 = Event SetBattlePassVisiblity!() -> (OutputDelegate, then, bIsVisible:bool) {is=title K2Node_Event}
   > SetVisibility(
    self=Get Overlay_BattlePass() -> (Overlay_BattlePass:Overlay Object Reference)
    InVisibility=Select(Option 0=Collapsed, Option 1=HitTestInvisible, Index=@n4.bIsVisible)
   )
   n5 = Event SetSeasonLevel!() -> (OutputDelegate, then, Level:int=0) {is=title K2Node_Event}
   > "SetText (Text)"(self=Get Text_SeasonLevel() -> (Text_SeasonLevel:Common Text Object Reference), InText="To Text (Integer)"(Value=@n5.Level))
   > SetVisibility(
    self=Get HorizontalBox_SeasonLevel() -> (HorizontalBox_SeasonLevel:Horizontal Box Object Reference)
    InVisibility=Select(Option 0=HitTestInvisible, Option 1=Collapsed, Index="integer <= integer"(A=@n5.Level))
   )
   Event PreConstruct()
   Event Construct()
   Event Tick()
   Event OnFinishedProfileLoading() {is=title K2Node_Event}
   Event OnStartProfileLoading() {is=title K2Node_Event}
  ]}]
variables []
widgets {
 info {inheritedWidgetCount=11, parentClass=/Script/FortniteUI.FortSocialProfileIdentity, rootWidgetClass=&SizeBox, widgetCount=31}
 widgets [
  SizeBox_0 {widgetClassPath=&SizeBox, widget=SizeBox_0, properties {bOverride_WidthOverride=true, heightOverride=250, widthOverride=500}}
  VerticalBox_2 {widgetClassPath=&VerticalBox, parent=SizeBox_0, widget=VerticalBox_2, slot {class=SizeBoxSlot}}
  Border_1 {
   widgetClassPath=&Border
   parent=VerticalBox_2
   widget=Border_1
   properties {
    background {
     bIsDynamicallyLoaded=false
     drawAs=Image
     imageType=NoImage
     mirroring=NoMirror
     resourceName=None
     resourceObject=/Game/UI/Sidebar/Materials/MI_UI_Sidebar_Header_Base
     tiling=NoTile
     imageSize=<32 32>
     margin=<ltrb 0 0 0 0>
     outlineSettings {
      bUseBrushTransparency=false
      roundingType=HalfHeightRadius
      width=0
      color {colorUseRule=UseColor_Specified, specifiedColor=<rgba 0 0 0 0>}
      cornerRadii=<0 0 0 0>
     }
     tintColor {colorUseRule=UseColor_Specified, specifiedColor=<rgba 1 1 1 1>}
     uVRegion {bIsValid=false, max=<0 0>, min=<0 0>}
    }
    padding=<ltrb 0 0 0 0>
   }
   slot {class=VerticalBoxSlot, properties {size {sizeRule=Fill, value=1}}}
  }
  Overlay_1 {widgetClassPath=&Overlay, parent=Border_1, widget=Overlay_1, slot {class=BorderSlot, properties {padding=<ltrb 0 0 0 0>}}}
  VerticalBox_347 {
   widgetClassPath=&VerticalBox
   parent=Overlay_1
   widget=VerticalBox_347
   slot {class=OverlaySlot, properties {horizontalAlignment=HAlign_Fill, verticalAlignment=VAlign_Fill, padding=<ltrb 0 160 0 0>}}
  }
  ScaleBox_0 {
   widgetClassPath=&ScaleBox
   parent=VerticalBox_347
   widget=ScaleBox_0
   slot {class=VerticalBoxSlot, properties {horizontalAlignment=HAlign_Center, padding=<ltrb 0 -85 0 -30>}}
  }
  Overlay_0 {widgetClassPath=&Overlay, parent=ScaleBox_0, widget=Overlay_0, slot {class=ScaleBoxSlot}}
  Avatar_SocialAvatarIcon {
   widgetClassPath=/Script/FortniteUI.FortSocialAvatarIcon
   bInherited=true
   bIsVariable=true
   parent=Overlay_0
   widget=Avatar_SocialAvatarIcon
   properties {
    avatarMaterial=/Game/UI/SocialPanel/Art/Materials/UI_UserPortrait_Buffet
    visibility=Collapsed
    brush {
     bIsDynamicallyLoaded=false
     drawAs=Image
     imageType=NoImage
     mirroring=NoMirror
     resourceName=None
     resourceObject=&UI_UserPortrait_Buffet
     tiling=NoTile
     imageSize=<135 135>
     margin=<ltrb 0 0 0 0>
     outlineSettings {
      bUseBrushTransparency=false
      roundingType=HalfHeightRadius
      width=0
      color {colorUseRule=UseColor_Specified, specifiedColor=<rgba 0 0 0 0>}
      cornerRadii=<0 0 0 0>
     }
     tintColor {colorUseRule=UseColor_Specified, specifiedColor=<rgba 1 1 1 1>}
     uVRegion {bIsValid=false, max=<0 0>, min=<0 0>}
    }
    loadingBackgroundBrush {
     bIsDynamicallyLoaded=false
     drawAs=Image
     imageType=NoImage
     mirroring=NoMirror
     resourceName=None
     resourceObject=/FortUILibrary/UIKit/Throbber/MI_UIKit_Monochrome_Throbber
     tiling=NoTile
     imageSize=<32 32>
     margin=<ltrb 0 0 0 0>
     outlineSettings {
      bUseBrushTransparency=false
      roundingType=HalfHeightRadius
      width=0
      color {colorUseRule=UseColor_Specified, specifiedColor=<rgba 0 0 0 0>}
      cornerRadii=<0 0 0 0>
     }
     tintColor {colorUseRule=UseColor_Specified, specifiedColor=<rgba 1 1 1 1>}
     uVRegion {bIsValid=false, max=<0 0>, min=<0 0>}
    }
   }
   slot {class=OverlaySlot}
  }
  Overlay_OnlineStatus {
   widgetClassPath=&Overlay
   bIsVariable=true
   parent=Overlay_0
   widget=Overlay_OnlineStatus
   properties {visibility=Collapsed}
   slot {class=OverlaySlot, properties {horizontalAlignment=HAlign_Right, verticalAlignment=VAlign_Bottom, padding=<ltrb 0 0 7 7>}}
  }
  Border_OnlineStatusOutline {
   widgetClassPath=&Border
   bInherited=true
   parent=Overlay_OnlineStatus
   widget=Border_OnlineStatusOutline
   properties {
    background {
     bIsDynamicallyLoaded=false
     drawAs=Image
     imageType=NoImage
     mirroring=NoMirror
     resourceName=None
     resourceObject=/LauncherSocial/circle
     tiling=NoTile
     imageSize=<28 28>
     margin=<ltrb 0 0 0 0>
     outlineSettings {
      bUseBrushTransparency=false
      roundingType=HalfHeightRadius
      width=0
      color {colorUseRule=UseColor_Specified, specifiedColor=<rgba 0 0 0 0>}
      cornerRadii=<0 0 0 0>
     }
     tintColor {colorUseRule=UseColor_Specified, specifiedColor=<rgba 1 1 1 1>}
     uVRegion {bIsValid=false, max=<0 0>, min=<0 0>}
    }
    brushColor=<rgba 0.0037 0.0168 0.0908 1>
    padding=<ltrb 0 0 36 36>
   }
   slot {class=OverlaySlot, properties {horizontalAlignment=HAlign_Center, verticalAlignment=VAlign_Center}}
  }
  Border_OnlineStatusIndicator {
   widgetClassPath=&Border
   bInherited=true
   parent=Overlay_OnlineStatus
   widget=Border_OnlineStatusIndicator
   properties {
    background {
     bIsDynamicallyLoaded=false
     drawAs=Image
     imageType=NoImage
     mirroring=NoMirror
     resourceName=None
     resourceObject=&circle
     tiling=NoTile
     imageSize=<28 28>
     margin=<ltrb 0 0 0 0>
     outlineSettings {
      bUseBrushTransparency=false
      roundingType=HalfHeightRadius
      width=0
      color {colorUseRule=UseColor_Specified, specifiedColor=<rgba 0 0 0 0>}
      cornerRadii=<0 0 0 0>
     }
     tintColor {colorUseRule=UseColor_Specified, specifiedColor=<rgba 1 1 1 1>}
     uVRegion {bIsValid=false, max=<0 0>, min=<0 0>}
    }
    brushColor=<rgba 0.1946 0.9823 0.1812 1>
    padding=<ltrb 0 0 28 28>
   }
   slot {class=OverlaySlot, properties {horizontalAlignment=HAlign_Center, verticalAlignment=VAlign_Center}}
  }
  PlayerAvatar {
   widgetClassPath=/Game/UI/Foundation/Profile/WBP_PlayerAvatar
   bInherited=true
   bIsVariable=true
   parent=Overlay_0
   widget=PlayerAvatar
   properties {presenceIconScale=0.3, showPresenceBg=true, avatarSize=<135 135>, presenceIconOffset=<-1.1 -1.1>}
   slot {class=OverlaySlot}
  }
  Border_3 {
   widgetClassPath=&Border
   parent=VerticalBox_347
   widget=Border_3
   properties {
    horizontalAlignment=HAlign_Center
    background {
     bIsDynamicallyLoaded=false
     drawAs=Image
     imageType=NoImage
     mirroring=NoMirror
     resourceName=None
     resourceObject=/Game/UI/Sidebar/Materials/M_UI_SkewBox_ProfileSubPanelIdentity
     tiling=NoTile
     imageSize=<32 32>
     margin=<ltrb 0 0 0 0>
     outlineSettings {
      bUseBrushTransparency=false
      roundingType=HalfHeightRadius
      width=0
      color {colorUseRule=UseColor_Specified, specifiedColor=<rgba 0 0 0 0>}
      cornerRadii=<0 0 0 0>
     }
     tintColor {colorUseRule=UseColor_Specified, specifiedColor=<rgba 0.004 0.0056 0.016 1>}
     uVRegion {bIsValid=false, max=<0 0>, min=<0 0>}
    }
    padding=<ltrb 0 40 0 0>
   }
   slot {class=VerticalBoxSlot}
  }
  HorizontalBox_0 {
   widgetClassPath=&HorizontalBox
   parent=Border_3
   widget=HorizontalBox_0
   slot {class=BorderSlot, properties {horizontalAlignment=HAlign_Center, padding=<ltrb 0 40 0 0>}}
  }
  WBP_RebootRallyIcon {
   widgetClassPath=/Game/UI/Sidebar/RebootRally/WBP_RebootRallyIcon
   bInherited=true
   bIsVariable=true
   parent=HorizontalBox_0
   widget=WBP_RebootRallyIcon
   properties {visibility=Collapsed}
   slot {class=HorizontalBoxSlot, properties {horizontalAlignment=HAlign_Left, verticalAlignment=VAlign_Center, padding=<ltrb 0 0 3 0>}}
  }
  Text_SocialName {
   widgetClassPath=/Script/FortniteUI.FortSocialNameTextBlock
   bInherited=true
   bIsVariable=true
   parent=HorizontalBox_0
   widget=Text_SocialName
   properties {
    bOverrideDefaultStyle=true
    bShowSecondName=false
    bTintInlineIcon=true
    justification=Center
    platformIconDisplayRule=AlwaysIfDifferent
    secondNameFontSize=28
    text="Nickname <PC color=\"FFFFFFFF\" size=\"28\"></><text color=\"FFFFFFFF\" size=\"28\">Display Name</>"
    defaultTextStyleOverride {
     overflowPolicy=Clip
     transformPolicy=None
     colorAndOpacity {colorUseRule=UseColor_Specified, specifiedColor=<rgba 1 1 1 1>}
     font {
      bForceMonospaced=false
      bMaterialIsStencil=false
      fontMaterial=null
      fontObject=&BurbankSmall
      letterSpacing=0
      monospacedWidth=1
      size=28
      skewAmount=0
      typefaceFontName=Black
      outlineSettings {
       bApplyOutlineToDropShadows=false
       bMiteredCorners=false
       bSeparateFillAlpha=false
       outlineMaterial=null
       outlineSize=0
       outlineColor=<rgba 0 0 0 1>
      }
     }
     highlightColor {colorUseRule=UseColor_Specified, specifiedColor=<rgba 0 0 0 1>}
     highlightShape {
      bIsDynamicallyLoaded=false
      drawAs=Image
      imageType=NoImage
      mirroring=NoMirror
      resourceName=None
      resourceObject=null
      tiling=NoTile
      imageSize=<32 32>
      margin=<ltrb 0 0 0 0>
      outlineSettings {
       bUseBrushTransparency=false
       roundingType=HalfHeightRadius
       width=0
       color {colorUseRule=UseColor_Specified, specifiedColor=<rgba 0 0 0 0>}
       cornerRadii=<0 0 0 0>
      }
      tintColor {colorUseRule=UseColor_Specified, specifiedColor=<rgba 1 1 1 1>}
      uVRegion {bIsValid=false, max=<0 0>, min=<0 0>}
     }
     selectedBackgroundColor {colorUseRule=UseColor_Foreground, specifiedColor=<rgba 1 0 1 1>}
     shadowColorAndOpacity=<rgba 0 0 0 1>
     shadowOffset=<0 0>
     strikeBrush {
      bIsDynamicallyLoaded=false
      drawAs=Image
      imageType=NoImage
      mirroring=NoMirror
      resourceName=None
      resourceObject=null
      tiling=NoTile
      imageSize=<32 32>
      margin=<ltrb 0 0 0 0>
      outlineSettings {
       bUseBrushTransparency=false
       roundingType=HalfHeightRadius
       width=0
       color {colorUseRule=UseColor_Specified, specifiedColor=<rgba 0 0 0 0>}
       cornerRadii=<0 0 0 0>
      }
      tintColor {colorUseRule=UseColor_Specified, specifiedColor=<rgba 1 1 1 1>}
      uVRegion {bIsValid=false, max=<0 0>, min=<0 0>}
     }
     underlineBrush {
      bIsDynamicallyLoaded=false
      drawAs=Image
      imageType=NoImage
      mirroring=NoMirror
      resourceName=None
      resourceObject=null
      tiling=NoTile
      imageSize=<32 32>
      margin=<ltrb 0 0 0 0>
      outlineSettings {
       bUseBrushTransparency=false
       roundingType=HalfHeightRadius
       width=0
       color {colorUseRule=UseColor_Specified, specifiedColor=<rgba 0 0 0 0>}
       cornerRadii=<0 0 0 0>
      }
      tintColor {colorUseRule=UseColor_Specified, specifiedColor=<rgba 1 1 1 1>}
      uVRegion {bIsValid=false, max=<0 0>, min=<0 0>}
     }
    }
    secondNameColor=<rgba 1 1 1 1>
   }
   slot {class=HorizontalBoxSlot, properties {padding=<ltrb 0 0 -32 0>}}
  }
  VoiceIcon_Status {
   widgetClassPath=/Script/FortniteUI.FortVoiceChatStatusIconProxyAnim
   bInherited=true
   bIsVariable=true
   parent=HorizontalBox_0
   widget=VoiceIcon_Status
   properties {renderTransform {angle=0, scale=<1 1>, shear=<0 0>, translation=<38 0>}}
   slot {class=HorizontalBoxSlot, properties {horizontalAlignment=HAlign_Center, verticalAlignment=VAlign_Center}}
  }
  Border_4 {
   widgetClassPath=&Border
   parent=VerticalBox_347
   widget=Border_4
   properties {
    background {
     bIsDynamicallyLoaded=false
     drawAs=Image
     imageType=NoImage
     mirroring=NoMirror
     resourceName=None
     resourceObject=null
     tiling=NoTile
     imageSize=<32 32>
     margin=<ltrb 0 0 0 0>
     outlineSettings {
      bUseBrushTransparency=false
      roundingType=HalfHeightRadius
      width=0
      color {colorUseRule=UseColor_Specified, specifiedColor=<rgba 0 0 0 0>}
      cornerRadii=<0 0 0 0>
     }
     tintColor {colorUseRule=UseColor_Specified, specifiedColor=<rgba 0.004 0.0056 0.016 1>}
     uVRegion {bIsValid=false, max=<0 0>, min=<0 0>}
    }
    padding=<ltrb 0 0 0 10>
   }
   slot {class=VerticalBoxSlot, properties {verticalAlignment=VAlign_Top}}
  }
  VerticalBox_1 {
   widgetClassPath=&VerticalBox
   parent=Border_4
   widget=VerticalBox_1
   slot {class=BorderSlot, properties {padding=<ltrb 0 0 0 10>}}
  }
  RichText_PlatformNames {
   widgetClassPath=&CommonRichTextBlock
   bInherited=true
   bIsVariable=true
   parent=VerticalBox_1
   widget=RichText_PlatformNames
   properties {
    autoWrapText=true
    bOverrideDefaultStyle=true
    justification=Center
    renderOpacity=0.5
    text=platformname
    defaultTextStyleOverride {
     overflowPolicy=Clip
     transformPolicy=None
     colorAndOpacity {colorUseRule=UseColor_Specified, specifiedColor=<rgba 1 1 1 1>}
     font {
      bForceMonospaced=false
      bMaterialIsStencil=false
      fontMaterial=null
      fontObject=&BurbankSmall
      letterSpacing=0
      monospacedWidth=1
      size=16
      skewAmount=0
      typefaceFontName=Bold
      outlineSettings {
       bApplyOutlineToDropShadows=false
       bMiteredCorners=false
       bSeparateFillAlpha=false
       outlineMaterial=null
       outlineSize=0
       outlineColor=<rgba 0 0 0 1>
      }
     }
     highlightColor {colorUseRule=UseColor_Specified, specifiedColor=<rgba 0 0 0 1>}
     highlightShape {
      bIsDynamicallyLoaded=false
      drawAs=Image
      imageType=NoImage
      mirroring=NoMirror
      resourceName=None
      resourceObject=null
      tiling=NoTile
      imageSize=<32 32>
      margin=<ltrb 0 0 0 0>
      outlineSettings {
       bUseBrushTransparency=false
       roundingType=HalfHeightRadius
       width=0
       color {colorUseRule=UseColor_Specified, specifiedColor=<rgba 0 0 0 0>}
       cornerRadii=<0 0 0 0>
      }
      tintColor {colorUseRule=UseColor_Specified, specifiedColor=<rgba 1 1 1 1>}
      uVRegion {bIsValid=false, max=<0 0>, min=<0 0>}
     }
     selectedBackgroundColor {colorUseRule=UseColor_Foreground, specifiedColor=<rgba 1 0 1 1>}
     shadowColorAndOpacity=<rgba 0 0 0 1>
     shadowOffset=<0 0>
     strikeBrush {
      bIsDynamicallyLoaded=false
      drawAs=Image
      imageType=NoImage
      mirroring=NoMirror
      resourceName=None
      resourceObject=null
      tiling=NoTile
      imageSize=<32 32>
      margin=<ltrb 0 0 0 0>
      outlineSettings {
       bUseBrushTransparency=false
       roundingType=HalfHeightRadius
       width=0
       color {colorUseRule=UseColor_Specified, specifiedColor=<rgba 0 0 0 0>}
       cornerRadii=<0 0 0 0>
      }
      tintColor {colorUseRule=UseColor_Specified, specifiedColor=<rgba 1 1 1 1>}
      uVRegion {bIsValid=false, max=<0 0>, min=<0 0>}
     }
     underlineBrush {
      bIsDynamicallyLoaded=false
      drawAs=Image
      imageType=NoImage
      mirroring=NoMirror
      resourceName=None
      resourceObject=null
      tiling=NoTile
      imageSize=<32 32>
      margin=<ltrb 0 0 0 0>
      outlineSettings {
       bUseBrushTransparency=false
       roundingType=HalfHeightRadius
       width=0
       color {colorUseRule=UseColor_Specified, specifiedColor=<rgba 0 0 0 0>}
       cornerRadii=<0 0 0 0>
      }
      tintColor {colorUseRule=UseColor_Specified, specifiedColor=<rgba 1 1 1 1>}
      uVRegion {bIsValid=false, max=<0 0>, min=<0 0>}
     }
    }
   }
   slot {class=VerticalBoxSlot}
  }
  VerticalBox_0 {
   widgetClassPath=&VerticalBox
   parent=Overlay_1
   widget=VerticalBox_0
   slot {class=OverlaySlot, properties {horizontalAlignment=HAlign_Right, padding=<ltrb 0 20 20 0>}}
  }
  HorizontalBox_42 {widgetClassPath=&HorizontalBox, parent=VerticalBox_0, widget=HorizontalBox_42, slot {class=VerticalBoxSlot}}
  Overlay_BattlePass {
   widgetClassPath=&Overlay
   bInherited=true
   parent=HorizontalBox_42
   widget=Overlay_BattlePass
   slot {class=HorizontalBoxSlot, properties {horizontalAlignment=HAlign_Right, padding=<ltrb 0 0 10 0>}}
  }
  Image_Battlepass {
   widgetClassPath=&Image
   bIsVariable=true
   parent=Overlay_BattlePass
   widget=Image_Battlepass
   properties {brush {
     bIsDynamicallyLoaded=false
     drawAs=Image
     imageType=NoImage
     mirroring=NoMirror
     resourceName=None
     resourceObject=/Game/UI/Foundation/XPBoost/M-BattlePassBadge
     tiling=NoTile
     imageSize=<40 40>
     margin=<ltrb 0 0 0 0>
     outlineSettings {
      bUseBrushTransparency=false
      roundingType=HalfHeightRadius
      width=0
      color {colorUseRule=UseColor_Specified, specifiedColor=<rgba 0 0 0 0>}
      cornerRadii=<0 0 0 0>
     }
     tintColor {colorUseRule=UseColor_Specified, specifiedColor=<rgba 1 1 1 1>}
     uVRegion {bIsValid=false, max=<0 0>, min=<0 0>}
    }}
   slot {class=OverlaySlot}
  }
  Overlay_FortniteCrew {
   widgetClassPath=&Overlay
   bInherited=true
   parent=HorizontalBox_42
   widget=Overlay_FortniteCrew
   slot {class=HorizontalBoxSlot, properties {horizontalAlignment=HAlign_Right, padding=<ltrb 0 0 5 0>}}
  }
  Image_89 {
   widgetClassPath=&Image
   bIsVariable=true
   parent=Overlay_FortniteCrew
   widget=Image_89
   properties {
    brush {
     bIsDynamicallyLoaded=false
     drawAs=Image
     imageType=NoImage
     mirroring=NoMirror
     resourceName=None
     resourceObject=/Game/Athena/UI/Subscriptions/Art/T_UI_SubsBadge128
     tiling=NoTile
     imageSize=<40 40>
     margin=<ltrb 0 0 0 0>
     outlineSettings {
      bUseBrushTransparency=false
      roundingType=HalfHeightRadius
      width=0
      color {colorUseRule=UseColor_Specified, specifiedColor=<rgba 0 0 0 0>}
      cornerRadii=<0 0 0 0>
     }
     tintColor {colorUseRule=UseColor_Specified, specifiedColor=<rgba 1 1 1 1>}
     uVRegion {bIsValid=false, max=<0 0>, min=<0 0>}
    }
    colorAndOpacity=<rgba 1 1 1 0.5>
   }
   slot {class=OverlaySlot, properties {horizontalAlignment=HAlign_Right}}
  }
  Image_243 {
   widgetClassPath=&Image
   bIsVariable=true
   parent=Overlay_FortniteCrew
   widget=Image_243
   properties {brush {
     bIsDynamicallyLoaded=false
     drawAs=Image
     imageType=NoImage
     mirroring=NoMirror
     resourceName=None
     resourceObject=/Game/Athena/UI/Subscriptions/Art/T_UI_Subs_Wordmark
     tiling=NoTile
     imageSize=<48 24>
     margin=<ltrb 0 0 0 0>
     outlineSettings {
      bUseBrushTransparency=false
      roundingType=HalfHeightRadius
      width=0
      color {colorUseRule=UseColor_Specified, specifiedColor=<rgba 0 0 0 0>}
      cornerRadii=<0 0 0 0>
     }
     tintColor {colorUseRule=UseColor_Specified, specifiedColor=<rgba 1 1 1 1>}
     uVRegion {bIsValid=false, max=<0 0>, min=<0 0>}
    }}
   slot {class=OverlaySlot, properties {horizontalAlignment=HAlign_Right, verticalAlignment=VAlign_Center, padding=<ltrb 0 1 -7 0>}}
  }
  HorizontalBox_SeasonLevel {
   widgetClassPath=&HorizontalBox
   bIsVariable=true
   parent=VerticalBox_0
   widget=HorizontalBox_SeasonLevel
   properties {visibility=Visible}
   slot {class=VerticalBoxSlot, properties {horizontalAlignment=HAlign_Right, verticalAlignment=VAlign_Top, padding=<ltrb 0 7 0 0>}}
  }
  CommonTextBlock_69 {
   widgetClassPath=&CommonTextBlock
   parent=HorizontalBox_SeasonLevel
   widget=CommonTextBlock_69
   properties {
    mobileFontSizeMultiplier=1.23
    text=LVL
    font {
     bForceMonospaced=false
     bMaterialIsStencil=false
     fontMaterial=null
     fontObject=&Burbank
     letterSpacing=0
     monospacedWidth=1
     size=13
     skewAmount=0
     typefaceFontName=BurbankRegular-Black
     outlineSettings {
      bApplyOutlineToDropShadows=false
      bMiteredCorners=false
      bSeparateFillAlpha=false
      outlineMaterial=null
      outlineSize=0
      outlineColor=<rgba 0 0 0 1>
     }
    }
    renderTransform {angle=0, scale=<1 1>, shear=<-8 0>, translation=<0 0>}
    shadowOffset=<0 0>
   }
   slot {class=HorizontalBoxSlot, properties {horizontalAlignment=HAlign_Right, verticalAlignment=VAlign_Bottom}}
  }
  Text_SeasonLevel {
   widgetClassPath=&CommonTextBlock
   bIsVariable=true
   parent=HorizontalBox_SeasonLevel
   widget=Text_SeasonLevel
   properties {
    mobileFontSizeMultiplier=1.15
    text=130
    font {
     bForceMonospaced=false
     bMaterialIsStencil=false
     fontMaterial=null
     fontObject=&Burbank
     letterSpacing=0
     monospacedWidth=1
     size=16
     skewAmount=0
     typefaceFontName=BurbankRegular-Black
     outlineSettings {
      bApplyOutlineToDropShadows=false
      bMiteredCorners=false
      bSeparateFillAlpha=false
      outlineMaterial=null
      outlineSize=0
      outlineColor=<rgba 0 0 0 1>
     }
    }
    renderTransform {angle=0, scale=<1 1>, shear=<-8 0>, translation=<0 0>}
    shadowOffset=<0 0>
   }
   slot {class=HorizontalBoxSlot, properties {horizontalAlignment=HAlign_Right, verticalAlignment=VAlign_Bottom, padding=<ltrb 5 0 0 0>}}
  }
  SubPanel_Loading {widgetClassPath=/Script/FortniteUI.FortSocialProfileSubPanel_Loading, bInherited=true, widget=null}
 ]
}
