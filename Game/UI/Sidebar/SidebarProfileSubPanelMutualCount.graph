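# Widget Blueprint for the mutual count sub-panel of the sidebar social profile, built on the native FortSocialProfileMutualCount class.
# The root is an inherited CommonTextBlock, styled as centered, wrapping grey bold text that auto-collapses when empty.
# An inherited loading sub-panel slot has no widget assigned here.
# The event graph holds only unconnected event stubs, including profile loading start and finish, so the Blueprint adds no logic of its own.
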WidgetBlueprint /Game/UI/Sidebar/SidebarProfileSubPanelMutualCount
graphs [EventGraph {nodes [
   Event PreConstruct()
   Event Construct()
   Event Tick()
   Event OnFinishedProfileLoading() {is=title K2Node_Event}
   Event OnStartProfileLoading() {is=title K2Node_Event}
  ]}]
variables []
widgets {
 info {
  inheritedWidgetCount=2
  parentClass=/Script/FortniteUI.FortSocialProfileMutualCount
  rootWidgetClass=&CommonTextBlock
  widgetCount=2
 }
 widgets [
  Text_MutualCount {
   widgetClassPath=&CommonTextBlock
   bInherited=true
   widget=Text_MutualCount
   properties {
    autoWrapText=true
    bAutoCollapseWithEmptyText=true
    bIsScrollingEnabled=false
    justification=Center
    mobileFontSizeMultiplier=1.15
    scrollStyle=&TextScrollStyle-Base_C
    text=Text_MutualCount
    colorAndOpacity {colorUseRule=UseColor_Specified, specifiedColor=<rgba 0.5208 0.5208 0.5208 1>}
    font {
     bForceMonospaced=false
     bMaterialIsStencil=false
     fontMaterial=null
     fontObject=&BurbankSmall
     letterSpacing=0
     monospacedWidth=1
     size=18
     skewAmount=0
     typefaceFontName=Bold
     outlineSettings {
      bApplyOutlineToDropShadows=false
      bMiteredCorners=false
      bSeparateFillAlpha=false
      outlineMaterial=null
      outlineSize=0
      outlineColor=<rgba 0 0 0 1>
     }
    }
    shadowOffset=<0 0>
   }
  }
  SubPanel_Loading {widgetClassPath=/Script/FortniteUI.FortSocialProfileSubPanel_Loading, bInherited=true, widget=null}
 ]
}
