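# Widget Blueprint for the Top Mutual Friends sub-panel of the sidebar social profile, based on FortSocialProfileMutualFriends.
# A size box holds a vertical stack of a dark header bar with skewed uppercase title text above a vertical dynamic entry box of SidebarUserListEntry_MutualFriend rows.
# The size box, entry box and loading sub-panel are inherited widgets, and the event graph has only unconnected stub events, leaving list population outside this Blueprint.
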WidgetBlueprint /Game/UI/Sidebar/SidebarProfileSubPanelMutualFriends
graphs [EventGraph {nodes [
   Event PreConstruct()
   Event Construct()
   Event Tick()
  ]}]
variables []
widgets {
 info {inheritedWidgetCount=3, parentClass=/Script/FortniteUI.FortSocialProfileMutualFriends, rootWidgetClass=&SizeBox, widgetCount=7}
 widgets [
  SizeBox_Root {widgetClassPath=&SizeBox, bInherited=true, widget=SizeBox_Root, properties {heightOverride=100, minDesiredHeight=100}}
  Border_0 {
   widgetClassPath=&Border
   parent=SizeBox_Root
   widget=Border_0
   properties {
    background {
     bIsDynamicallyLoaded=false
     drawAs=Image
     imageType=NoImage
     mirroring=NoMirror
     resourceName=None
     resourceObject=null
     tiling=NoTile
     imageSize=<32 32>
     margin=<ltrb 0 0 0 0>
     outlineSettings {
      bUseBrushTransparency=false
      roundingType=HalfHeightRadius
      width=0
      color {colorUseRule=UseColor_Specified, specifiedColor=<rgba 0 0 0 0>}
      cornerRadii=<0 0 0 0>
     }
     tintColor {colorUseRule=UseColor_Specified, specifiedColor=<rgba 1 1 1 0>}
     uVRegion {bIsValid=false, max=<0 0>, min=<0 0>}
    }
    padding=<ltrb 0 0 0 0>
   }
   slot {class=SizeBoxSlot}
  }
  VerticalBox_0 {widgetClassPath=&VerticalBox, parent=Border_0, widget=VerticalBox_0, slot {class=BorderSlot, properties {padding=<ltrb 0 0 0 0>}}}
  CommonBorder_0 {
   widgetClassPath=&CommonBorder
   parent=VerticalBox_0
   widget=CommonBorder_0
   properties {
    horizontalAlignment=HAlign_Left
    background {
     bIsDynamicallyLoaded=false
     drawAs=Box
     imageType=NoImage
     mirroring=NoMirror
     resourceName=None
     resourceObject=null
     tiling=NoTile
     imageSize=<32 32>
     margin=<ltrb 0 0 0 0>
     outlineSettings {
      bUseBrushTransparency=false
      roundingType=HalfHeightRadius
      width=0
      color {colorUseRule=UseColor_Specified, specifiedColor=<rgba 0 0 0 0>}
      cornerRadii=<0 0 0 0>
     }
     tintColor {colorUseRule=UseColor_Specified, specifiedColor=<rgba 1 1 1 1>}
     uVRegion {bIsValid=false, max=<0 0>, min=<0 0>}
    }
    brushColor=<rgba 0.004 0.0056 0.016 1>
    padding=<ltrb 20 10 0 10>
   }
   slot {class=VerticalBoxSlot}
  }
  Text_MutualFriends {
   widgetClassPath=&CommonTextBlock
   bIsVariable=true
   parent=CommonBorder_0
   widget=Text_MutualFriends
   properties {
    justification=Center
    mobileFontSizeMultiplier=1.1
    text=Top Mutual Friends
    textTransformPolicy=ToUpper
    colorAndOpacity {colorUseRule=UseColor_Specified, specifiedColor=<rgba 0.5089 0.5583 0.7084 1>}
    font {
     bForceMonospaced=false
     bMaterialIsStencil=false
     fontMaterial=null
     fontObject=&BurbankSmall
     letterSpacing=0
     monospacedWidth=1
     size=17
     skewAmount=0
     typefaceFontName=Black
     outlineSettings {
      bApplyOutlineToDropShadows=false
      bMiteredCorners=false
      bSeparateFillAlpha=false
      outlineMaterial=null
      outlineSize=0
      outlineColor=<rgba 0 0 0 1>
     }
    }
    renderTransform {angle=0, scale=<1 1>, shear=<-8 0>, translation=<0 0>}
    shadowOffset=<0 0>
   }
   slot {class=BorderSlot, properties {horizontalAlignment=HAlign_Left, padding=<ltrb 20 10 0 10>}}
  }
  EntryBox_MutualFriends {
   widgetClassPath=&DynamicEntryBox
   bInherited=true
   bIsVariable=true
   parent=VerticalBox_0
   widget=EntryBox_MutualFriends
   properties {
    entryBoxType=Vertical
    entryWidgetClass=/Game/UI/Sidebar/ListEntries/SidebarUserListEntry_MutualFriend.SidebarUserListEntry_MutualFriend_C
    numDesignerPreviewEntries=4
    entrySpacing=<0 3>
   }
   slot {class=VerticalBoxSlot}
  }
  SubPanel_Loading {widgetClassPath=/Script/FortniteUI.FortSocialProfileSubPanel_Loading, bInherited=true, widget=null}
 ]
}
