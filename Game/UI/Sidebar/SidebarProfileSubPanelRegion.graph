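# Sidebar widget for the region sub-panel of a social profile, built on the native FortSocialProfileRegion class.
# A centered, scale-boxed row shows mutual count, region and language text, separated by small circle images in matching translucent blue.
# The texts and separator boxes bind to inherited parent widgets, while the inherited loading sub-panel has no widget here.
# The event graph holds only unconnected stub events, so the Blueprint supplies layout and styling alone.
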WidgetBlueprint /Game/UI/Sidebar/SidebarProfileSubPanelRegion
graphs [EventGraph {nodes [
   Event PreConstruct()
   Event Construct()
   Event Tick()
  ]}]
variables []
widgets {
 info {inheritedWidgetCount=6, parentClass=/Script/FortniteUI.FortSocialProfileRegion, rootWidgetClass=&SizeBox, widgetCount=12}
 widgets [
  SizeBox_0 {widgetClassPath=&SizeBox, widget=SizeBox_0}
  Border_0 {
   widgetClassPath=&Border
   parent=SizeBox_0
   widget=Border_0
   properties {
    horizontalAlignment=HAlign_Center
    background {
     bIsDynamicallyLoaded=false
     drawAs=Image
     imageType=NoImage
     mirroring=NoMirror
     resourceName=None
     resourceObject=null
     tiling=NoTile
     imageSize=<32 32>
     margin=<ltrb 0 0 0 0>
     outlineSettings {
      bUseBrushTransparency=false
      roundingType=HalfHeightRadius
      width=0
      color {colorUseRule=UseColor_Specified, specifiedColor=<rgba 0 0 0 0>}
      cornerRadii=<0 0 0 0>
     }
     tintColor {colorUseRule=UseColor_Specified, specifiedColor=<rgba 0.006 0.016 0.0595 0>}
     uVRegion {bIsValid=false, max=<0 0>, min=<0 0>}
    }
    padding=<ltrb 10 10 10 10>
   }
   slot {class=SizeBoxSlot}
  }
  ScaleBox_0 {
   widgetClassPath=&ScaleBox
   parent=Border_0
   widget=ScaleBox_0
   slot {class=BorderSlot, properties {horizontalAlignment=HAlign_Center, padding=<ltrb 10 10 10 10>}}
  }
  HorizontalBox_0 {widgetClassPath=&HorizontalBox, parent=ScaleBox_0, widget=HorizontalBox_0, slot {class=ScaleBoxSlot}}
  Text_MutualCount {
   widgetClassPath=&CommonTextBlock
   bInherited=true
   parent=HorizontalBox_0
   widget=Text_MutualCount
   properties {
    justification=Right
    mobileFontSizeMultiplier=1.15
    text=Text_MutualCountText
    colorAndOpacity {colorUseRule=UseColor_Specified, specifiedColor=<rgba 0.0331 0.402 1 0.6>}
    font {
     bForceMonospaced=false
     bMaterialIsStencil=false
     fontMaterial=null
     fontObject=&BurbankSmall
     letterSpacing=0
     monospacedWidth=1
     size=16
     skewAmount=0
     typefaceFontName=Black
     outlineSettings {
      bApplyOutlineToDropShadows=false
      bMiteredCorners=false
      bSeparateFillAlpha=false
      outlineMaterial=null
      outlineSize=0
      outlineColor=<rgba 0 0 0 1>
     }
    }
    shadowOffset=<0 0>
   }
   slot {class=HorizontalBoxSlot, properties {horizontalAlignment=HAlign_Right, verticalAlignment=VAlign_Center}}
  }
  SizeBox_Separator_Left {
   widgetClassPath=&SizeBox
   bInherited=true
   parent=HorizontalBox_0
   widget=SizeBox_Separator_Left
   properties {bOverride_HeightOverride=true, bOverride_WidthOverride=true, heightOverride=12, widthOverride=12}
   slot {class=HorizontalBoxSlot, properties {horizontalAlignment=HAlign_Center, verticalAlignment=VAlign_Center, padding=<ltrb 5 0 5 0>}}
  }
  Image {
   widgetClassPath=&Image
   bIsVariable=true
   parent=SizeBox_Separator_Left
   widget=Image
   properties {
    brush {
     bIsDynamicallyLoaded=false
     drawAs=Image
     imageType=NoImage
     mirroring=NoMirror
     resourceName=None
     resourceObject=/LauncherSocial/circle
     tiling=NoTile
     imageSize=<24 24>
     margin=<ltrb 0 0 0 0>
     outlineSettings {
      bUseBrushTransparency=false
      roundingType=HalfHeightRadius
      width=0
      color {colorUseRule=UseColor_Specified, specifiedColor=<rgba 0 0 0 0>}
      cornerRadii=<0 0 0 0>
     }
     tintColor {colorUseRule=UseColor_Specified, specifiedColor=<rgba 1 1 1 1>}
     uVRegion {bIsValid=false, max=<0 0>, min=<0 0>}
    }
    colorAndOpacity=<rgba 0.0331 0.402 1 0.6>
   }
   slot {class=SizeBoxSlot}
  }
  Text_Region {
   widgetClassPath=&CommonTextBlock
   bInherited=true
   parent=HorizontalBox_0
   widget=Text_Region
   properties {
    justification=Right
    mobileFontSizeMultiplier=1.15
    text=Text_RegionText_Region
    colorAndOpacity {colorUseRule=UseColor_Specified, specifiedColor=<rgba 0.0331 0.402 1 0.6>}
    font {
     bForceMonospaced=false
     bMaterialIsStencil=false
     fontMaterial=null
     fontObject=&BurbankSmall
     letterSpacing=0
     monospacedWidth=1
     size=16
     skewAmount=0
     typefaceFontName=Black
     outlineSettings {
      bApplyOutlineToDropShadows=false
      bMiteredCorners=false
      bSeparateFillAlpha=false
      outlineMaterial=null
      outlineSize=0
      outlineColor=<rgba 0 0 0 1>
     }
    }
    shadowOffset=<0 0>
   }
   slot {class=HorizontalBoxSlot, properties {horizontalAlignment=HAlign_Right, verticalAlignment=VAlign_Center}}
  }
  SizeBox_Separator_Right {
   widgetClassPath=&SizeBox
   bInherited=true
   parent=HorizontalBox_0
   widget=SizeBox_Separator_Right
   properties {bOverride_HeightOverride=true, bOverride_WidthOverride=true, heightOverride=12, widthOverride=12}
   slot {class=HorizontalBoxSlot, properties {horizontalAlignment=HAlign_Center, verticalAlignment=VAlign_Center, padding=<ltrb 5 0 5 0>}}
  }
  Image_63 {
   widgetClassPath=&Image
   bIsVariable=true
   parent=SizeBox_Separator_Right
   widget=Image_63
   properties {
    brush {
     bIsDynamicallyLoaded=false
     drawAs=Image
     imageType=NoImage
     mirroring=NoMirror
     resourceName=None
     resourceObject=&circle
     tiling=NoTile
     imageSize=<24 24>
     margin=<ltrb 0 0 0 0>
     outlineSettings {
      bUseBrushTransparency=false
      roundingType=HalfHeightRadius
      width=0
      color {colorUseRule=UseColor_Specified, specifiedColor=<rgba 0 0 0 0>}
      cornerRadii=<0 0 0 0>
     }
     tintColor {colorUseRule=UseColor_Specified, specifiedColor=<rgba 1 1 1 1>}
     uVRegion {bIsValid=false, max=<0 0>, min=<0 0>}
    }
    colorAndOpacity=<rgba 0.0331 0.402 1 0.6>
   }
   slot {class=SizeBoxSlot}
  }
  Text_Language {
   widgetClassPath=&CommonTextBlock
   bInherited=true
   parent=HorizontalBox_0
   widget=Text_Language
   properties {
    mobileFontSizeMultiplier=1.15
    text=Text_LanguageText_Language
    colorAndOpacity {colorUseRule=UseColor_Specified, specifiedColor=<rgba 0.0331 0.402 1 0.6>}
    font {
     bForceMonospaced=false
     bMaterialIsStencil=false
     fontMaterial=null
     fontObject=&BurbankSmall
     letterSpacing=0
     monospacedWidth=1
     size=16
     skewAmount=0
     typefaceFontName=Black
     outlineSettings {
      bApplyOutlineToDropShadows=false
      bMiteredCorners=false
      bSeparateFillAlpha=false
      outlineMaterial=null
      outlineSize=0
      outlineColor=<rgba 0 0 0 1>
     }
    }
    shadowOffset=<0 0>
   }
   slot {class=HorizontalBoxSlot, properties {horizontalAlignment=HAlign_Left, verticalAlignment=VAlign_Center}}
  }
  SubPanel_Loading {widgetClassPath=/Script/FortniteUI.FortSocialProfileSubPanel_Loading, bInherited=true, widget=null}
 ]
}
