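# Widget Blueprint for an editable social tag list in the Fortnite sidebar UI, derived from the native SidebarSocialTagListEditable class.
# A centered WrapBox holds three tag widgets, each in a width-capped SizeBox, followed by an add-tag button that starts collapsed.
# The tag widgets and button are inherited variables, and the event graph has only unwired stub events, so this asset supplies layout only.
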WidgetBlueprint /Game/UI/Sidebar/Tags/SidebarSocialTagListEditable
graphs [EventGraph {nodes [
   Event PreConstruct()
   Event Construct()
   Event Tick()
  ]}]
variables []
widgets {
 info {inheritedWidgetCount=4, parentClass=/Script/FortniteUI.SidebarSocialTagListEditable, rootWidgetClass=&WrapBox, widgetCount=8}
 widgets [
  WrapBox_0 {widgetClassPath=&WrapBox, widget=WrapBox_0, properties {horizontalAlignment=HAlign_Center}}
  SizeBox_0 {
   widgetClassPath=&SizeBox
   parent=WrapBox_0
   widget=SizeBox_0
   properties {bOverride_MaxDesiredWidth=true, maxDesiredWidth=250}
   slot {class=WrapBoxSlot, properties {horizontalAlignment=HAlign_Left, verticalAlignment=VAlign_Center}}
  }
  SidebarSocialTagWidget_1 {
   widgetClassPath=&/SidebarSocialTagWidget
   bInherited=true
   bIsVariable=true
   parent=SizeBox_0
   widget=SidebarSocialTagWidget_1
   properties {tagCategory=Other, tagName="", visibility=Visible}
   slot {class=SizeBoxSlot}
  }
  SizeBox_1 {
   widgetClassPath=&SizeBox
   parent=WrapBox_0
   widget=SizeBox_1
   properties {bOverride_MaxDesiredWidth=true, maxDesiredWidth=250}
   slot {class=WrapBoxSlot, properties {horizontalAlignment=HAlign_Left, verticalAlignment=VAlign_Center}}
  }
  SidebarSocialTagWidget_2 {
   widgetClassPath=&SidebarSocialTagWidget
   bInherited=true
   bIsVariable=true
   parent=SizeBox_1
   widget=SidebarSocialTagWidget_2
   properties {tagCategory=Other, tagName="", visibility=Visible}
   slot {class=SizeBoxSlot, properties {padding=<ltrb 5 0 0 0>}}
  }
  SizeBox_2 {
   widgetClassPath=&SizeBox
   parent=WrapBox_0
   widget=SizeBox_2
   properties {bOverride_MaxDesiredWidth=true, maxDesiredWidth=250}
   slot {class=WrapBoxSlot, properties {horizontalAlignment=HAlign_Left, verticalAlignment=VAlign_Center}}
  }
  SidebarSocialTagWidget_3 {
   widgetClassPath=&SidebarSocialTagWidget
   bInherited=true
   bIsVariable=true
   parent=SizeBox_2
   widget=SidebarSocialTagWidget_3
   properties {tagCategory=Other, tagName="", visibility=Visible}
   slot {class=SizeBoxSlot, properties {padding=<ltrb 5 0 0 0>}}
  }
  AddButton {
   widgetClassPath=&/SidebarSocialAddTagWidget
   bInherited=true
   bIsVariable=true
   parent=WrapBox_0
   widget=AddButton
   properties {visibility=Collapsed}
   slot {class=WrapBoxSlot, properties {horizontalAlignment=HAlign_Center, verticalAlignment=VAlign_Center, padding=<ltrb 10 0 10 0>}}
  }
 ]
}
