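# Profile tags sub-panel widget for the social sidebar, built on the native FortSocialProfileTags class.
# A root horizontal box holds the inherited editable social tag list, padded to fill the space, with category customization shown.
# The inherited loading sub-panel slot has no widget assigned.
# The event graph holds only unwired default events, so behavior comes from the parent class and the child list widget.
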WidgetBlueprint /Game/UI/Sidebar/SidebarProfileSubPanelTags
graphs [EventGraph {nodes [
   Event PreConstruct()
   Event Construct()
   Event Tick()
  ]}]
variables []
widgets {
 info {inheritedWidgetCount=2, parentClass=/Script/FortniteUI.FortSocialProfileTags, rootWidgetClass=&HorizontalBox, widgetCount=3}
 widgets [
  HorizontalBox_0 {widgetClassPath=&HorizontalBox, widget=HorizontalBox_0}
  SidebarSocialTagList {
   widgetClassPath=/Game/UI/Sidebar/Tags/SidebarSocialTagListEditable
   bInherited=true
   bIsVariable=true
   parent=HorizontalBox_0
   widget=SidebarSocialTagList
   properties {bShowCategoryCustomization=true}
   slot {class=HorizontalBoxSlot, properties {padding=<ltrb 10 20 10 10>, size {sizeRule=Fill, value=1}}}
  }
  SubPanel_Loading {widgetClassPath=/Script/FortniteUI.FortSocialProfileSubPanel_Loading, bInherited=true, widget=null}
 ]
}
