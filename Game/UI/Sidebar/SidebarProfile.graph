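# Sidebar social profile panel widget, derived from FortSocialProfilePanel, stacking identity and back button above scrollable tags, presence, region, friend request, action and mutual-friend sub-panels.
# While the profile loads, a throbber shows and the presence/region border is collapsed.
# When loading finishes, OnContentLoaded fades out the throbber and expands the clipped border via desired-size scale while fading it in.
# On activation, the friend request panel shows only when the ClamOne option and friend-request flow are both true.
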WidgetBlueprint /Game/UI/Sidebar/SidebarProfile
animations [OnContentLoaded {
  binding_count=2
  length_seconds=0.2
  object=&self:OnContentLoaded
  sequence {
   bindings [
    Border_PresenceRegion {
     id=#1
     possessed_class=CommonBorder
     tracks [
      {
       type=MovieSceneDoubleVectorTrack
       display_name=Desired Size Scale
       sections [{
         type=MovieSceneDoubleVectorSection
         end_frame=12
         start_frame=0
         channels [
          Vector.X {type=MovieSceneScriptingDoubleChannel}
          Vector.Y {type=MovieSceneScriptingDoubleChannel, keys [[0,0],[12,1]]}
         ]
        }]
      }
      {
       type=MovieSceneFloatTrack
       display_name=Render Opacity
       sections [{type=MovieSceneFloatSection, end_frame=12, start_frame=0, channels [None {type=MovieSceneScriptingFloatChannel, keys [[6,0],[12,1]]}]}]
      }
      {
       type=MovieSceneByteTrack
       display_name=Visibility
       sections [{type=MovieSceneByteSection, end_frame=0, start_frame=0, channels [None {type=MovieSceneScriptingByteChannel, const [4,1]}]}]
      }
     ]
    }
    SidebarProfileLoading {
     id=#2
     possessed_class=Image
     tracks [
      {
       type=MovieSceneByteTrack
       display_name=Visibility
       sections [{type=MovieSceneByteSection, end_frame=12, start_frame=0, channels [None {type=MovieSceneScriptingByteChannel, keys [[0,4],[11,1]]}]}]
      }
      {
       type=MovieSceneFloatTrack
       display_name=Render Opacity
       sections [{type=MovieSceneFloatSection, end_frame=12, start_frame=0, channels [None {type=MovieSceneScriptingFloatChannel, keys [[0,1],[6,0]]}]}]
      }
     ]
    }
   ]
   display_rate {numerator=60}
  }
 }]
graphs [EventGraph {nodes [
   Event OnStartProfileLoading() {is=title K2Node_Event}
   > SetVisibility(self=Get SidebarProfileLoading() -> (SidebarProfileLoading:Image Object Reference), InVisibility=SelfHitTestInvisible)
   > SetVisibility(self=Get Border_PresenceRegion() -> (Border_PresenceRegion:Common Border Object Reference), InVisibility=Collapsed)
   Event OnFinishedProfileLoading() {is=title K2Node_Event}
   > PlayAnimation(InAnimation=Get OnContentLoaded() -> (OnContentLoaded:Widget Animation Object Reference))
   Event On Activated()
   n1 = GetIsClamOneEnabled!(self:Fort Runtime Options Object Reference=GetRuntimeOptions()) -> (ReturnValue:bool)
   n2 = GetIsFriendRequestFlow!(self:Fort Social Profile Panel Object Reference) -> (ReturnValue:bool)
   > SetVisibility(
    self=Get SubPanelFriendRequest() -> (SubPanelFriendRequest:Sidebar Profile Sub Panel Friend Request Object Reference)
    InVisibility=Select(Option 0=Collapsed, Option 1=HitTestInvisible, Index=AND Boolean(A=@n1, B=@n2))
   )
   Event PreConstruct()
   Event Construct()
   Event Tick()
  ]}]
variables []
widgets {
 info {inheritedWidgetCount=10, parentClass=/Script/FortniteUI.FortSocialProfilePanel, rootWidgetClass=&SafeZone, widgetCount=20}
 widgets [
  SafeZone_Content {widgetClassPath=&SafeZone, bIsVariable=true, widget=SafeZone_Content, properties {padLeft=false, padRight=false, padTop=false}}
  VerticalBox_0 {widgetClassPath=&VerticalBox, parent=SafeZone_Content, widget=VerticalBox_0, slot {class=SafeZoneSlot}}
  Overlay_0 {widgetClassPath=&Overlay, parent=VerticalBox_0, widget=Overlay_0, slot {class=VerticalBoxSlot}}
  SubPanelIdentity {
   widgetClassPath=&/SidebarProfileSubPanelIdentity
   bInherited=true
   bIsVariable=true
   parent=Overlay_0
   widget=SubPanelIdentity
   slot {class=OverlaySlot, properties {horizontalAlignment=HAlign_Fill}}
  }
  SidebarBackButton {
   widgetClassPath=/FortUILibrary/Foundation/Buttons/BackButtonSmall
   bInherited=true
   bIsVariable=true
   parent=Overlay_0
   widget=SidebarBackButton
   slot {class=OverlaySlot}
  }
  ScrollBox_ScrollablePanels {
   widgetClassPath=&ScrollBox
   bInherited=true
   parent=VerticalBox_0
   widget=ScrollBox_ScrollablePanels
   properties {widgetStyle {
     barThickness=9
     bottomShadowBrush {
      bIsDynamicallyLoaded=false
      drawAs=Box
      imageType=FullColor
      mirroring=NoMirror
      resourceName=../../../Engine/Content/Slate/Common/ScrollBoxShadowBottom.png
      resourceObject=null
      tiling=NoTile
      imageSize=<16 8>
      margin=<ltrb 0.5 0 0.5 1>
      outlineSettings {
       bUseBrushTransparency=false
       roundingType=HalfHeightRadius
       width=0
       color {colorUseRule=UseColor_Specified, specifiedColor=<rgba 0 0 0 0>}
       cornerRadii=<0 0 0 0>
      }
      tintColor {colorUseRule=UseColor_Specified, specifiedColor=<rgba 1 1 1 1>}
      uVRegion {bIsValid=false, max=<0 0>, min=<0 0>}
     }
     horizontalScrolledContentPadding=<ltrb 0 0 1 0>
     leftShadowBrush {
      bIsDynamicallyLoaded=false
      drawAs=Box
      imageType=FullColor
      mirroring=NoMirror
      resourceName=../../../Engine/Content/Slate/Common/ScrollBoxShadowLeft.png
      resourceObject=null
      tiling=NoTile
      imageSize=<8 16>
      margin=<ltrb 1 0.5 0 0.5>
      outlineSettings {
       bUseBrushTransparency=false
       roundingType=HalfHeightRadius
       width=0
       color {colorUseRule=UseColor_Specified, specifiedColor=<rgba 0 0 0 0>}
       cornerRadii=<0 0 0 0>
      }
      tintColor {colorUseRule=UseColor_Specified, specifiedColor=<rgba 1 1 1 1>}
      uVRegion {bIsValid=false, max=<0 0>, min=<0 0>}
     }
     rightShadowBrush {
      bIsDynamicallyLoaded=false
      drawAs=Box
      imageType=FullColor
      mirroring=NoMirror
      resourceName=../../../Engine/Content/Slate/Common/ScrollBoxShadowRight.png
      resourceObject=null
      tiling=NoTile
      imageSize=<8 16>
      margin=<ltrb 0 0.5 1 0.5>
      outlineSettings {
       bUseBrushTransparency=false
       roundingType=HalfHeightRadius
       width=0
       color {colorUseRule=UseColor_Specified, specifiedColor=<rgba 0 0 0 0>}
       cornerRadii=<0 0 0 0>
      }
      tintColor {colorUseRule=UseColor_Specified, specifiedColor=<rgba 1 1 1 1>}
      uVRegion {bIsValid=false, max=<0 0>, min=<0 0>}
     }
     topShadowBrush {
      bIsDynamicallyLoaded=false
      drawAs=NoDrawType
      imageType=FullColor
      mirroring=NoMirror
      resourceName=../../../Engine/Content/Slate/Common/ScrollBoxShadowTop.png
      resourceObject=null
      tiling=NoTile
      imageSize=<16 8>
      margin=<ltrb 0.5 1 0.5 0>
      outlineSettings {
       bUseBrushTransparency=false
       roundingType=HalfHeightRadius
       width=0
       color {colorUseRule=UseColor_Specified, specifiedColor=<rgba 0 0 0 0>}
       cornerRadii=<0 0 0 0>
      }
      tintColor {colorUseRule=UseColor_Specified, specifiedColor=<rgba 1 1 1 1>}
      uVRegion {bIsValid=false, max=<0 0>, min=<0 0>}
     }
     verticalScrolledContentPadding=<ltrb 0 0 0 1>
    }}
   slot {class=VerticalBoxSlot, properties {size {sizeRule=Fill, value=1}}}
  }
  VBox_Details {
   widgetClassPath=&VerticalBox
   parent=ScrollBox_ScrollablePanels
   widget=VBox_Details
   slot {class=ScrollBoxSlot, properties {padding=<ltrb 0 16 0 0>}}
  }
  SubPanelTags {
   widgetClassPath=&/SidebarProfileSubPanelTags
   bInherited=true
   bIsVariable=true
   parent=VBox_Details
   widget=SubPanelTags
   properties {visibility=Collapsed}
   slot {class=VerticalBoxSlot}
  }
  Overlay_PresenceRegion {
   widgetClassPath=&Overlay
   parent=VBox_Details
   widget=Overlay_PresenceRegion
   slot {class=VerticalBoxSlot, properties {padding=<ltrb 0 16 0 0>}}
  }
  SidebarProfileLoading {
   widgetClassPath=&Image
   bIsVariable=true
   parent=Overlay_PresenceRegion
   widget=SidebarProfileLoading
   properties {brush {
     bIsDynamicallyLoaded=false
     drawAs=Image
     imageType=NoImage
     mirroring=NoMirror
     resourceName=None
     resourceObject=/FortUILibrary/UIKit/Throbber/MI_UIKit_Monochrome_Throbber
     tiling=NoTile
     imageSize=<32 32>
     margin=<ltrb 0 0 0 0>
     outlineSettings {
      bUseBrushTransparency=false
      roundingType=HalfHeightRadius
      width=0
      color {colorUseRule=UseColor_Specified, specifiedColor=<rgba 0 0 0 0>}
      cornerRadii=<0 0 0 0>
     }
     tintColor {colorUseRule=UseColor_Specified, specifiedColor=<rgba 1 1 1 1>}
     uVRegion {bIsValid=false, max=<0 0>, min=<0 0>}
    }}
   slot {class=OverlaySlot, properties {horizontalAlignment=HAlign_Center}}
  }
  Border_PresenceRegion {
   widgetClassPath=&CommonBorder
   bIsVariable=true
   parent=Overlay_PresenceRegion
   widget=Border_PresenceRegion
   properties {clipping=ClipToBoundsAlways, desiredSizeScale=<1 0>, padding=<ltrb 0 0 0 0>}
   slot {class=OverlaySlot, properties {horizontalAlignment=HAlign_Fill, verticalAlignment=VAlign_Fill}}
  }
  VBox_PresenceRegion {
   widgetClassPath=&VerticalBox
   parent=Border_PresenceRegion
   widget=VBox_PresenceRegion
   slot {class=BorderSlot, properties {padding=<ltrb 0 0 0 0>}}
  }
  SubPanelMutualFriendsCount {
   widgetClassPath=&/SidebarProfileSubPanelMutualCount
   bInherited=true
   bIsVariable=true
   parent=VBox_PresenceRegion
   widget=SubPanelMutualFriendsCount
   slot {class=VerticalBoxSlot}
  }
  SubPanelPresence {
   widgetClassPath=&/SidebarProfileSubPanelPresence
   bInherited=true
   bIsVariable=true
   parent=VBox_PresenceRegion
   widget=SubPanelPresence
   properties {padding=<ltrb 0 0 0 0>}
   slot {class=VerticalBoxSlot}
  }
  SubPanelRegion {
   widgetClassPath=&/SidebarProfileSubPanelRegion
   bInherited=true
   bIsVariable=true
   parent=VBox_PresenceRegion
   widget=SubPanelRegion
   slot {class=VerticalBoxSlot}
  }
  SubPanelFriendRequest {
   widgetClassPath=&/SidebarProfileSubPanelFriendRequest
   bIsVariable=true
   parent=VBox_Details
   widget=SubPanelFriendRequest
   properties {visibility=Collapsed}
   slot {class=VerticalBoxSlot, properties {verticalAlignment=VAlign_Top, padding=<ltrb 0 60 0 0>}}
  }
  VBox_ActionsFriends {widgetClassPath=&VerticalBox, parent=VBox_Details, widget=VBox_ActionsFriends, slot {class=VerticalBoxSlot}}
  SubPanelActions {
   widgetClassPath=&/SideBarProfileSubPanelActions
   bInherited=true
   bIsVariable=true
   parent=VBox_ActionsFriends
   widget=SubPanelActions
   slot {class=VerticalBoxSlot}
  }
  SubPanelRebootRallyText {
   widgetClassPath=/Game/UI/Sidebar/RebootRally/SidebarProfileSubPanelRebootRallyEligible
   bInherited=true
   bIsVariable=true
   parent=VBox_ActionsFriends
   widget=SubPanelRebootRallyText
   properties {padding=<ltrb 0 0 0 0>}
   slot {class=VerticalBoxSlot, properties {padding=<ltrb 20 0 20 20>}}
  }
  SubPanelMutualFriends {
   widgetClassPath=&/SidebarProfileSubPanelMutualFriends
   bInherited=true
   bIsVariable=true
   parent=VBox_ActionsFriends
   widget=SubPanelMutualFriends
   slot {class=VerticalBoxSlot}
  }
 ]
}
