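# Sidebar button widget derived from FortMainMenuButton: a rounded rectangle with uppercase ticker-scrolling text, a named sub-slot, an icon and an input action glyph.
# Hover plays an animation that turns the blue background white, darkens the content and reveals the glyph; unhover reverses it at double speed.
# Input method changes switch the text wrapper's padding between gamepad and base values, starting at initialization.
# Glyph icon updates set its dynamic material to dark, without outline.
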WidgetBlueprint /Game/UI/Sidebar/SidebarRoundedRectangleButton
animations [Hovered {
  binding_count=3
  display_label=""
  length_seconds=0.2
  object=&self:Hovered
  sequence {bindings [
    TextWrapper {
     id=#1
     possessed_class=CommonBorder
     tracks [{
       type=MovieSceneColorTrack
       display_name=Content Color and Opacity
       sections [{
         type=MovieSceneColorSection
         end_frame=4
         start_frame=0
         channels [
          Color.R {type=MovieSceneScriptingFloatChannel, keys [[0,1],[4,0.0513]]}
          Color.G {type=MovieSceneScriptingFloatChannel, keys [[0,1],[4,0.0513]]}
          Color.B {type=MovieSceneScriptingFloatChannel, keys [[0,1],[4,0.0513]]}
          Color.A {type=MovieSceneScriptingFloatChannel, keys [[0,0.5],[4,1]]}
         ]
        }]
      }]
    }
    InputActionWidget {
     id=#2
     possessed_class=CommonActionWidget
     tracks [{
       type=MovieSceneFloatTrack
       display_name=Render Opacity
       sections [{type=MovieSceneFloatSection, end_frame=4, start_frame=0, channels [None {type=MovieSceneScriptingFloatChannel, keys [[0,0],[1,0],[2,1]]}]}]
      }]
    }
    ButtonBackground {
     id=#3
     possessed_class=Border
     tracks [{
       type=MovieSceneColorTrack
       display_name=Brush Color
       sections [{
         type=MovieSceneColorSection
         end_frame=4
         start_frame=0
         channels [
          Color.R {type=MovieSceneScriptingFloatChannel, keys [[0,0.0021],[4,1]]}
          Color.G {type=MovieSceneScriptingFloatChannel, keys [[0,0.0612],[4,1]]}
          Color.B {type=MovieSceneScriptingFloatChannel, keys [[0,0.5029],[4,1]]}
          Color.A {type=MovieSceneScriptingFloatChannel, const [1,2]}
         ]
        }]
      }]
    }
   ]}
 }]
graphs [
 EventGraph {nodes [
   "On Input Icon Updated (InputActionWidget)"() {is=title K2Node_ComponentBoundEvent}
   > n1 = GetIconDynamicMaterial!(execute, self:Common Action Widget Object Reference=Get InputActionWidget() -> (InputActionWidget:Common Action Widget Object Reference)) -> (then, ReturnValue:Material Instance Dynamic Object Reference)
   > SetScalarParameterValue(self=@n1, ParameterName=Use Dark, Value=1)
   > SetScalarParameterValue(self=@n1, ParameterName=Outline Opacity, Value=0)
   Event OnInitialized()
   n2 = Get CommonInputSubsystem()
   > Bind Event to On Input Method Changed(self=@n2, Delegate=Create Event())
   > HandleInputMethodChanged!(execute, self, bNewInputType:ECommonInputType Enum=GetCurrentInputType(self=@n2)) -> (then)
   Event On Hovered()
   > PlayAnimation(InAnimation=Get Hovered() -> (Hovered:Widget Animation Object Reference))
   Event On Unhovered()
   > PlayAnimation(InAnimation=Get Hovered(), PlayMode=Reverse, PlaybackSpeed=2)
  ]}
 HandleInputMethodChanged {nodes [
   n1 = HandleInputMethodChanged!() -> (then, bNewInputType:ECommonInputType Enum) {is=title K2Node_FunctionEntry}
   n2 = Get BasePadding() -> (BasePadding:Margin Structure)
   > Set Padding!(
    execute
    Padding:Margin Structure=Select!(
     MouseAndKeyboard:Margin Structure=@n2
     Gamepad:Margin Structure=Get GamepadPadding() -> (GamepadPadding:Margin Structure)
     Touch:Margin Structure=@n2
     Count:Margin Structure=@n2
     Index:ECommonInputType Enum=@n1
    ) -> (ReturnValue:Margin Structure)
    self:Border Slot Object Reference=SlotAsBorderSlot!(Widget:Widget Object Reference=Get TextWrapper() -> (TextWrapper:Common Border Object Reference)) -> (ReturnValue:Border Slot Object Reference)
   ) -> (then, Output_Get:Margin Structure)
  ]}
]
variables [ButtonMaterial {type=MaterialInterface, default=/Game/Athena/UI/Materials/M_UI_Circle_Blue}]
widgets {
 info {inheritedWidgetCount=4, parentClass=/Script/FortniteUI.FortMainMenuButton, rootWidgetClass=&SizeBox, widgetCount=10}
 widgets [
  SizeBox_0 {widgetClassPath=&SizeBox, widget=SizeBox_0, properties {bOverride_HeightOverride=true, heightOverride=72, minDesiredHeight=72}}
  Overlay_0 {widgetClassPath=&Overlay, parent=SizeBox_0, widget=Overlay_0, slot {class=SizeBoxSlot}}
  ButtonBackground {
   widgetClassPath=&Border
   bIsVariable=true
   parent=Overlay_0
   widget=ButtonBackground
   properties {
    bShowEffectWhenDisabled=false
    verticalAlignment=VAlign_Center
    visibility=SelfHitTestInvisible
    background {
     bIsDynamicallyLoaded=false
     drawAs=Box
     imageType=NoImage
     mirroring=NoMirror
     resourceName=None
     resourceObject=/Game/UI/Sidebar/Materials/MI_SidebarRoundedRectangleButton
     tiling=NoTile
     imageSize=<64 64>
     margin=<ltrb 0.5 0.5 0.5 0.5>
     outlineSettings {
      bUseBrushTransparency=false
      roundingType=HalfHeightRadius
      width=0
      color {colorUseRule=UseColor_Specified, specifiedColor=<rgba 0 0 0 0>}
      cornerRadii=<0 0 0 0>
     }
     tintColor {colorUseRule=UseColor_Specified, specifiedColor=<rgba 1 1 1 1>}
     uVRegion {bIsValid=false, max=<0 0>, min=<0 0>}
    }
    brushColor=<rgba 0.0021 0.0612 0.5029 1>
    padding=<ltrb 60 0 60 0>
   }
   slot {class=OverlaySlot, properties {horizontalAlignment=HAlign_Fill, verticalAlignment=VAlign_Fill, padding=<ltrb 8 8 8 8>}}
  }
  TextWrapper {
   widgetClassPath=&CommonBorder
   bIsVariable=true
   parent=ButtonBackground
   widget=TextWrapper
   properties {horizontalAlignment=HAlign_Center, verticalAlignment=VAlign_Center, contentColorAndOpacity=<rgba 1 1 1 0.5>, padding=<ltrb 0 0 0 0>}
   slot {class=BorderSlot, properties {verticalAlignment=VAlign_Center, padding=<ltrb 60 0 60 0>}}
  }
  HorizontalBox {
   widgetClassPath=&HorizontalBox
   parent=TextWrapper
   widget=HorizontalBox
   slot {class=BorderSlot, properties {horizontalAlignment=HAlign_Center, verticalAlignment=VAlign_Center, padding=<ltrb 0 0 0 0>}}
  }
  Text_Content {
   widgetClassPath=&CommonTextBlock
   bInherited=true
   parent=HorizontalBox
   widget=Text_Content
   properties {
    mobileFontSizeMultiplier=1.5
    scrollStyle=/Game/Athena/Prototype/Blueprints/Galileo/UI/TextScrollStyle-Base-Ticker.TextScrollStyle-Base-Ticker_C
    style=/FortUILibrary/Foundation/Buttons/ButtonStyles/Athena/TextStyle_BurbSmBk_15_White.TextStyle_BurbSmBk_15_White_C
    text=BUTTON TEXT
    textTransformPolicy=ToUpper
    font {
     bForceMonospaced=false
     bMaterialIsStencil=false
     fontMaterial=null
     fontObject=&BurbankSmall
     letterSpacing=0
     monospacedWidth=1
     size=15
     skewAmount=0
     typefaceFontName=Black
     outlineSettings {
      bApplyOutlineToDropShadows=false
      bMiteredCorners=false
      bSeparateFillAlpha=false
      outlineMaterial=null
      outlineSize=0
      outlineColor=<rgba 0 0 0 1>
     }
    }
    shadowOffset=<0 0>
   }
   slot {class=HorizontalBoxSlot, properties {size {sizeRule=Fill, value=1}}}
  }
  SubSlot {
   widgetClassPath=&NamedSlot
   bIsVariable=true
   parent=HorizontalBox
   widget=SubSlot
   properties {bExposeOnInstanceOnly=true}
   slot {class=HorizontalBoxSlot}
  }
  InputActionWidget {
   widgetClassPath=&CommonActionWidget
   bInherited=true
   bIsVariable=true
   parent=Overlay_0
   widget=InputActionWidget
   properties {renderOpacity=0}
   slot {class=OverlaySlot, properties {horizontalAlignment=HAlign_Right, verticalAlignment=VAlign_Center, padding=<ltrb 0 0 32 0>}}
  }
  Image_Icon {
   widgetClassPath=&Image
   bInherited=true
   parent=Overlay_0
   widget=Image_Icon
   properties {brush {
     bIsDynamicallyLoaded=false
     drawAs=NoDrawType
     imageType=NoImage
     mirroring=NoMirror
     resourceName=None
     resourceObject=null
     tiling=NoTile
     imageSize=<0 0>
     margin=<ltrb 0 0 0 0>
     outlineSettings {
      bUseBrushTransparency=false
      roundingType=HalfHeightRadius
      width=0
      color {colorUseRule=UseColor_Specified, specifiedColor=<rgba 0 0 0 0>}
      cornerRadii=<0 0 0 0>
     }
     tintColor {colorUseRule=UseColor_Specified, specifiedColor=<rgba 1 1 1 1>}
     uVRegion {bIsValid=false, max=<0 0>, min=<0 0>}
    }}
   slot {class=OverlaySlot, properties {verticalAlignment=VAlign_Center, padding=<ltrb 24 0 0 0>}}
  }
  ContentBorder {widgetClassPath=&Border, bInherited=true, widget=null, properties=~}
 ]
}
